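# Niagara drift trail cosmetic "Black Flame" for vehicles, drawing paintable flame ribbons that stream backward from the rear wheels.
# Ribbon emitters, including Aura_Ribbon and MainRibbon, spawn ribbon pairs at skeletal positions plus a user offset, with per-tire disabling by spawn group.
# Painted-variant overrides supply color-over-life curves per paint, scaled by user power settings when painted.
# Aura_Ribbon computes two opposed sine/cosine spiral positions, with spin mirrored per side, for its two renderers.
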NiagaraSystem /VehicleCosmetics/DriftTrail/DriftTrail_BlackFlame/FX/NS_DriftTrail_BlackFlame1
dependencies {
 usedDataInterfaces [&NiagaraDataInterfaceColorCurve, &NiagaraDataInterfaceCurve, &NiagaraDataInterfaceSkeletalMesh]
 usedDynamicInputs []
 usedModules [
  &self:SystemSpawnScript.SystemScriptSource.SystemScriptGraph.NiagaraNodeAssignment_0.SetVariables_#1
  &FN_UnpackVehicleFlags
  /VehicleCosmetics/FX/Niagara/Modules/VC_SetVehicleSpeedAttribs
  &SystemState
 ]
}
emitters {
 Aura_Ribbon {
  emitterUpdateScript [
   &EmitterState()
   &VC_RateQualityHandler(Epic Spawn Rate=30, Low Spawn Rate=30, Use Custom Rate?=true)
   &VC_SpawnRibbon_v1(Number of Ribbons=2, Spawn Time=-1, SpawnRate=$Emitter.Spawn Rate)
  ]
  particleSpawnScript [
   &VC_PaintedVariantsOverride(
    R1 Black:LinearColor=&ColorFromCurve(ColorCurve=DataInterface() {propertyValues {
       AlphaCurve {
        defaultValue=0
        keys [<at 0 0>, {arriveTangent=15.9916, interpMode=RCIM_Cubic, leaveTangent=15.9916, time=0.1339, value=8}, <at 0.6253 10>, <at 0.9771 0>]
       }
       BlueCurve {keys [<at 1 0.1094 RCIM_Cubic>]}
       GreenCurve {keys [<at 1 0.1094 RCIM_Cubic>]}
       RedCurve {keys [<at 1 0.1094 RCIM_Cubic>]}
      }})
    R1 BurntSienna:LinearColor=&ColorFromCurve(ColorCurve=DataInterface() {propertyValues {
       AlphaCurve {
        defaultValue=0
        keys [<at 0 0>, {arriveTangent=15.9916, interpMode=RCIM_Cubic, leaveTangent=15.9916, time=0.1756, value=8}, <at 0.6253 10>, <at 0.9771 0>]
       }
       BlueCurve {keys [<at 1 0.0686 RCIM_Cubic>]}
       GreenCurve {keys [<at 1 0.0858 RCIM_Cubic>]}
       RedCurve {keys [<at 1 0.1667 RCIM_Cubic>]}
      }})
    R1 Cobalt:LinearColor=&ColorFromCurve(ColorCurve=DataInterface() {propertyValues {
       AlphaCurve {
        defaultValue=0
        keys [<at 0 0>, {arriveTangent=15.9916, interpMode=RCIM_Cubic, leaveTangent=15.9916, time=0.1756, value=8}, <at 0.6253 10>, <at 0.9771 0>]
       }
       BlueCurve {keys [<at 1 0.3802 RCIM_Cubic>]}
       GreenCurve {keys [<at 1 0.0931 RCIM_Cubic>]}
       RedCurve {keys [<at 1 0.0931 RCIM_Cubic>]}
      }})
    R1 Crimson:LinearColor=&ColorFromCurve(ColorCurve=DataInterface() {propertyValues {
       AlphaCurve {
        defaultValue=0
        keys [<at 0 0>, {arriveTangent=15.9916, interpMode=RCIM_Cubic, leaveTangent=15.9916, time=0.1756, value=8}, <at 0.6253 10>, <at 0.9771 0>]
       }
       BlueCurve {keys [<at 1 0.0131 RCIM_Cubic>]}
       GreenCurve {keys [<at 1 0 RCIM_Cubic>]}
       RedCurve {keys [<at 1 0.2348 RCIM_Cubic>]}
      }})
    R1 DefaultColor=&ColorFromCurve(ColorCurve=DataInterface() {propertyValues {
       AlphaCurve {keys [<at 0 0>, {arriveTangent=15.9916, interpMode=RCIM_Cubic, leaveTangent=15.9916, time=0.1039, value=8}, <at 0.6253 10>, <at 0.9771 0>]}
       BlueCurve {keys [<at 0.6966 1>]}
       GreenCurve {keys [<at 0.6966 0.5>]}
       RedCurve {keys [<at 0.6966 0>]}
      }})
    R1 ForestGreen:LinearColor=&ColorFromCurve(ColorCurve=DataInterface() {propertyValues {
       AlphaCurve {
        defaultValue=0
        keys [<at 0 0>, {arriveTangent=15.9916, interpMode=RCIM_Cubic, leaveTangent=15.9916, time=0.1756, value=8}, <at 0.6253 10>, <at 0.9771 0>]
       }
       BlueCurve {keys [<at 1 0.0198 RCIM_Cubic>]}
       GreenCurve {keys [<at 1 0.2031 RCIM_Cubic>]}
       RedCurve {keys [<at 1 0 RCIM_Cubic>]}
      }})
    R1 Grey:LinearColor=&ColorFromCurve(ColorCurve=DataInterface() {propertyValues {
       AlphaCurve {
        defaultValue=0
        keys [<at 0 0>, {arriveTangent=15.9916, interpMode=RCIM_Cubic, leaveTangent=15.9916, time=0.1756, value=8}, <at 0.6253 10>, <at 0.9771 0>]
       }
       BlueCurve {keys [<at 1 0.3542 RCIM_Cubic>]}
       GreenCurve {keys [<at 1 0.3542 RCIM_Cubic>]}
       RedCurve {keys [<at 1 0.3542 RCIM_Cubic>]}
      }})
    R1 Lime:LinearColor=&ColorFromCurve(ColorCurve=DataInterface() {propertyValues {
       AlphaCurve {
        defaultValue=0
        keys [<at 0 0>, {arriveTangent=15.9916, interpMode=RCIM_Cubic, leaveTangent=15.9916, time=0.1756, value=8}, <at 0.6253 10>, <at 0.9771 0>]
       }
       BlueCurve {keys [<at 1 0.0865 RCIM_Cubic>]}
       GreenCurve {keys [<at 1 0.2865 RCIM_Cubic>]}
       RedCurve {keys [<at 1 0.1697 RCIM_Cubic>]}
      }})
    R1 Orange:LinearColor=&ColorFromCurve(ColorCurve=DataInterface() {propertyValues {
       AlphaCurve {
        defaultValue=0
        keys [<at 0 0>, {arriveTangent=15.9916, interpMode=RCIM_Cubic, leaveTangent=15.9916, time=0.1756, value=8}, <at 0.6253 10>, <at 0.9771 0>]
       }
       BlueCurve {keys [<at 1 0 RCIM_Cubic>]}
       GreenCurve {keys [<at 1 0.2961 RCIM_Cubic>]}
       RedCurve {keys [<at 1 0.9271 RCIM_Cubic>]}
      }})
    R1 Pink:LinearColor=&ColorFromCurve(ColorCurve=DataInterface() {propertyValues {
       AlphaCurve {
        defaultValue=0
        keys [<at 0 0>, {arriveTangent=15.9916, interpMode=RCIM_Cubic, leaveTangent=15.9916, time=0.1756, value=8}, <at 0.6253 10>, <at 0.9771 0>]
       }
       BlueCurve {keys [<at 1 0.2004 RCIM_Cubic>]}
       GreenCurve {keys [<at 1 0 RCIM_Cubic>]}
       RedCurve {keys [<at 1 0.2188 RCIM_Cubic>]}
      }})
    R1 Purple:LinearColor=&ColorFromCurve(ColorCurve=DataInterface() {propertyValues {
       AlphaCurve {
        defaultValue=0
        keys [<at 0 0>, {arriveTangent=15.9916, interpMode=RCIM_Cubic, leaveTangent=15.9916, time=0.1756, value=8}, <at 0.6253 10>, <at 0.9771 0>]
       }
       BlueCurve {keys [<at 1 0.375 RCIM_Cubic>]}
       GreenCurve {keys [<at 1 0 RCIM_Cubic>]}
       RedCurve {keys [<at 1 0.1484 RCIM_Cubic>]}
      }})
    R1 Saffron:LinearColor=&ColorFromCurve(ColorCurve=DataInterface() {propertyValues {
       AlphaCurve {
        defaultValue=0
        keys [<at 0 0>, {arriveTangent=15.9916, interpMode=RCIM_Cubic, leaveTangent=15.9916, time=0.1756, value=8}, <at 0.6253 10>, <at 0.9771 0>]
       }
       BlueCurve {keys [<at 1 0 RCIM_Cubic>]}
       GreenCurve {keys [<at 1 0.2395 RCIM_Cubic>]}
       RedCurve {keys [<at 1 0.2812 RCIM_Cubic>]}
      }})
    R1 SkyBlue:LinearColor=&ColorFromCurve(ColorCurve=DataInterface() {propertyValues {
       AlphaCurve {
        defaultValue=0
        keys [<at 0 0>, {arriveTangent=15.9916, interpMode=RCIM_Cubic, leaveTangent=15.9916, time=0.1756, value=8}, <at 0.6253 10>, <at 0.9771 0>]
       }
       BlueCurve {keys [<at 1 0.3802 RCIM_Cubic>]}
       GreenCurve {keys [<at 1 0.1905 RCIM_Cubic>]}
       RedCurve {keys [<at 1 0 RCIM_Cubic>]}
      }})
    R1 TitaniumWhite:LinearColor=&ColorFromCurve(ColorCurve=DataInterface() {propertyValues {
       AlphaCurve {
        defaultValue=0
        keys [<at 0 0>, {arriveTangent=15.9916, interpMode=RCIM_Cubic, leaveTangent=15.9916, time=0.1756, value=8}, <at 0.6253 10>, <at 0.9771 0>]
       }
       BlueCurve {keys [<at 1 0.6406 RCIM_Cubic>]}
       GreenCurve {keys [<at 1 0.6406 RCIM_Cubic>]}
       RedCurve {keys [<at 1 0.6406 RCIM_Cubic>]}
      }})
    Use R1 Painted Overrides=true
   ) {enabled=false}
   &VC_InitializeRibbon_v1(
    Color=$Particles.Color
    Custom Offset=true
    Enable Skeletal Position=true
    Facing Axis:ENiagaraOrientationAxis=::YAxis
    Lifetime=1
    Local Vector Offset=&AddVectorToPosition(Position=<25 0 -20>, Vector=Linked() {linkedVariable {name=User.offset, type=NiagaraPosition}})
    Offset Amount=1
    Ribbon Width=160
   )
   &AddVelocity(Velocity=<-2200 0 0>) {module=AddVelocity001}
   /SetVariables(Particles.FacingVert_Ribbon:Vector3f=<0 0 1>, Particles.SizeSecondAura_Ribbon=120) {isSetParameters=true}
   /SetVariables() {enabled=false, isSetParameters=true}
  ]
  particleUpdateScript [
   &VC_PaintedVariantsOverride(
    R1 Black:LinearColor=&ColorFromCurve(ColorCurve=DataInterface() {propertyValues {
       AlphaCurve {
        defaultValue=0
        keys [<at 0 0>, {arriveTangent=15.9916, interpMode=RCIM_Cubic, leaveTangent=15.9916, time=0.1339, value=8}, <at 0.6253 10>, <at 0.9771 0>]
       }
       BlueCurve {keys [<at 1 0.1094 RCIM_Cubic>]}
       GreenCurve {keys [<at 1 0.1094 RCIM_Cubic>]}
       RedCurve {keys [<at 1 0.1094 RCIM_Cubic>]}
      }})
    R1 BurntSienna:LinearColor=&ColorFromCurve(ColorCurve=DataInterface() {propertyValues {
       AlphaCurve {
        defaultValue=0
        keys [<at 0 0>, {arriveTangent=15.9916, interpMode=RCIM_Cubic, leaveTangent=15.9916, time=0.1756, value=8}, <at 0.6253 10>, <at 0.9771 0>]
       }
       BlueCurve {keys [<at 1 0.0686 RCIM_Cubic>]}
       GreenCurve {keys [<at 1 0.0858 RCIM_Cubic>]}
       RedCurve {keys [<at 1 0.1667 RCIM_Cubic>]}
      }})
    R1 Cobalt:LinearColor=&ColorFromCurve(ColorCurve=DataInterface() {propertyValues {
       AlphaCurve {
        defaultValue=0
        keys [<at 0 0>, {arriveTangent=15.9916, interpMode=RCIM_Cubic, leaveTangent=15.9916, time=0.1756, value=8}, <at 0.6253 10>, <at 0.9771 0>]
       }
       BlueCurve {keys [<at 1 0.3802 RCIM_Cubic>]}
       GreenCurve {keys [<at 1 0.0931 RCIM_Cubic>]}
       RedCurve {keys [<at 1 0.0931 RCIM_Cubic>]}
      }})
    R1 Crimson:LinearColor=&ColorFromCurve(ColorCurve=DataInterface() {propertyValues {
       AlphaCurve {
        defaultValue=0
        keys [<at 0 0>, {arriveTangent=15.9916, interpMode=RCIM_Cubic, leaveTangent=15.9916, time=0.1756, value=8}, <at 0.6253 10>, <at 0.9771 0>]
       }
       BlueCurve {keys [<at 1 0.0131 RCIM_Cubic>]}
       GreenCurve {keys [<at 1 0 RCIM_Cubic>]}
       RedCurve {keys [<at 1 0.2348 RCIM_Cubic>]}
      }})
    R1 DefaultColor=&ColorFromCurve(ColorCurve=DataInterface() {propertyValues {
       AlphaCurve {keys [<at 0 0>, {arriveTangent=15.9916, interpMode=RCIM_Cubic, leaveTangent=15.9916, time=0.1039, value=8}, <at 0.6253 10>, <at 0.9771 0>]}
       BlueCurve {keys [<at 0.6966 1>]}
       GreenCurve {keys [<at 0.6966 0.5>]}
       RedCurve {keys [<at 0.6966 0>]}
      }})
    R1 ForestGreen:LinearColor=&ColorFromCurve(ColorCurve=DataInterface() {propertyValues {
       AlphaCurve {
        defaultValue=0
        keys [<at 0 0>, {arriveTangent=15.9916, interpMode=RCIM_Cubic, leaveTangent=15.9916, time=0.1756, value=8}, <at 0.6253 10>, <at 0.9771 0>]
       }
       BlueCurve {keys [<at 1 0.0198 RCIM_Cubic>]}
       GreenCurve {keys [<at 1 0.2031 RCIM_Cubic>]}
       RedCurve {keys [<at 1 0 RCIM_Cubic>]}
      }})
    R1 Grey:LinearColor=&ColorFromCurve(ColorCurve=DataInterface() {propertyValues {
       AlphaCurve {
        defaultValue=0
        keys [<at 0 0>, {arriveTangent=15.9916, interpMode=RCIM_Cubic, leaveTangent=15.9916, time=0.1756, value=8}, <at 0.6253 10>, <at 0.9771 0>]
       }
       BlueCurve {keys [<at 1 0.3542 RCIM_Cubic>]}
       GreenCurve {keys [<at 1 0.3542 RCIM_Cubic>]}
       RedCurve {keys [<at 1 0.3542 RCIM_Cubic>]}
      }})
    R1 Lime:LinearColor=&ColorFromCurve(ColorCurve=DataInterface() {propertyValues {
       AlphaCurve {
        defaultValue=0
        keys [<at 0 0>, {arriveTangent=15.9916, interpMode=RCIM_Cubic, leaveTangent=15.9916, time=0.1756, value=8}, <at 0.6253 10>, <at 0.9771 0>]
       }
       BlueCurve {keys [<at 1 0.0865 RCIM_Cubic>]}
       GreenCurve {keys [<at 1 0.2865 RCIM_Cubic>]}
       RedCurve {keys [<at 1 0.1697 RCIM_Cubic>]}
      }})
    R1 Orange:LinearColor=&ColorFromCurve(ColorCurve=DataInterface() {propertyValues {
       AlphaCurve {
        defaultValue=0
        keys [<at 0 0>, {arriveTangent=15.9916, interpMode=RCIM_Cubic, leaveTangent=15.9916, time=0.1756, value=8}, <at 0.6253 10>, <at 0.9771 0>]
       }
       BlueCurve {keys [<at 1 0 RCIM_Cubic>]}
       GreenCurve {keys [<at 1 0.2961 RCIM_Cubic>]}
       RedCurve {keys [<at 1 0.9271 RCIM_Cubic>]}
      }})
    R1 Pink:LinearColor=&ColorFromCurve(ColorCurve=DataInterface() {propertyValues {
       AlphaCurve {
        defaultValue=0
        keys [<at 0 0>, {arriveTangent=15.9916, interpMode=RCIM_Cubic, leaveTangent=15.9916, time=0.1756, value=8}, <at 0.6253 10>, <at 0.9771 0>]
       }
       BlueCurve {keys [<at 1 0.2004 RCIM_Cubic>]}
       GreenCurve {keys [<at 1 0 RCIM_Cubic>]}
       RedCurve {keys [<at 1 0.2188 RCIM_Cubic>]}
      }})
    R1 Purple:LinearColor=&ColorFromCurve(ColorCurve=DataInterface() {propertyValues {
       AlphaCurve {
        defaultValue=0
        keys [<at 0 0>, {arriveTangent=15.9916, interpMode=RCIM_Cubic, leaveTangent=15.9916, time=0.1756, value=8}, <at 0.6253 10>, <at 0.9771 0>]
       }
       BlueCurve {keys [<at 1 0.375 RCIM_Cubic>]}
       GreenCurve {keys [<at 1 0 RCIM_Cubic>]}
       RedCurve {keys [<at 1 0.1484 RCIM_Cubic>]}
      }})
    R1 Saffron:LinearColor=&ColorFromCurve(ColorCurve=DataInterface() {propertyValues {
       AlphaCurve {defaultValue=0, keys [{arriveTangent=15.9916, interpMode=RCIM_Cubic, time=0.1756, value=8}, <at 0.6253 10>, <at 0.9771 0>]}
       BlueCurve {keys [<at 1 0 RCIM_Cubic>]}
       GreenCurve {keys [<at 1 0.2395 RCIM_Cubic>]}
       RedCurve {keys [<at 1 0.2812 RCIM_Cubic>]}
      }})
    R1 SkyBlue:LinearColor=&ColorFromCurve(ColorCurve=DataInterface() {propertyValues {
       AlphaCurve {
        defaultValue=0
        keys [<at 0 0>, {arriveTangent=15.9916, interpMode=RCIM_Cubic, leaveTangent=15.9916, time=0.1756, value=8}, <at 0.6253 10>, <at 0.9771 0>]
       }
       BlueCurve {keys [<at 1 0.3802 RCIM_Cubic>]}
       GreenCurve {keys [<at 1 0.1905 RCIM_Cubic>]}
       RedCurve {keys [<at 1 0 RCIM_Cubic>]}
      }})
    R1 TitaniumWhite:LinearColor=&ColorFromCurve(ColorCurve=DataInterface() {propertyValues {
       AlphaCurve {
        defaultValue=0
        keys [<at 0 0>, {arriveTangent=15.9916, interpMode=RCIM_Cubic, leaveTangent=15.9916, time=0.1756, value=8}, <at 0.6253 10>, <at 0.9771 0>]
       }
       BlueCurve {keys [<at 1 0.6406 RCIM_Cubic>]}
       GreenCurve {keys [<at 1 0.6406 RCIM_Cubic>]}
       RedCurve {keys [<at 1 0.6406 RCIM_Cubic>]}
      }})
    R2 Black:LinearColor=&ColorFromCurve(ColorCurve=DataInterface() {propertyValues {
       AlphaCurve {keys [<at 0 7>,<at 0.2232 1>,<at 0.3261 0.05>,<at 0.5495 0.2>,<at 0.7007 0.2>,<at 0.8474 0 RCIM_Cubic>]}
       BlueCurve {keys [<at 0 0.0194 RCIM_Cubic>,<at 0.1492 0.0333>,<at 0.2681 0.1104>,<at 0.4863 0.9274>,<at 0.6088 0.0305>]}
       GreenCurve {keys [<at 0 0.0194 RCIM_Cubic>,<at 0.1492 0.0333>,<at 0.2681 0.1104>,<at 0.4863 0.9274>,<at 0.6088 0.0305>]}
       RedCurve {keys [<at 0 0.0194 RCIM_Cubic>,<at 0.1492 0.0333>,<at 0.2681 0.1104>,<at 0.4863 0.9274>,<at 0.6088 0.0305>]}
      }})
    R2 BurntSienna:LinearColor=&ColorFromCurve(ColorCurve=DataInterface() {propertyValues {
       AlphaCurve {keys [<at 0 4>,<at 0.1907 1>,<at 0.3261 0.05>,<at 0.593 0.2>,<at 0.8474 0 RCIM_Cubic>]}
       BlueCurve {keys [<at 0 0.008 RCIM_Cubic>,<at 0.2839 0.0202>,<at 0.4863 0.2907>,<at 0.5956 0.008>]}
       GreenCurve {keys [<at 0 0.0203 RCIM_Cubic>,<at 0.2839 0.0323>,<at 0.4863 0.4394>,<at 0.5956 0.0203>]}
       RedCurve {keys [<at 0 0.0648 RCIM_Cubic>,<at 0.2839 0.0762>,<at 0.4863 0.9792>,<at 0.5956 0.0648>]}
      }})
    R2 Cobalt:LinearColor=&ColorFromCurve(ColorCurve=DataInterface() {propertyValues {
       AlphaCurve {keys [<at 0 7>,<at 0.1907 1>,<at 0.3261 0.05>,<at 0.5495 0.6>,<at 0.7007 0.2>,<at 0.8474 0 RCIM_Cubic>]}
       BlueCurve {keys [<at 0 0.6458 RCIM_Cubic>,<at 0.2575 0.9985>,<at 0.4863 1>,<at 0.6115 0.8177>]}
       GreenCurve {keys [<at 0 0.0119 RCIM_Cubic>,<at 0.2575 0.0367>,<at 0.4863 0.2997>,<at 0.6115 0.0527>]}
       RedCurve {keys [<at 0 0 RCIM_Cubic>,<at 0.2575 0.0186>,<at 0.4863 0.2865>,<at 0.6115 0.0383>]}
      }})
    R2 Crimson:LinearColor=&ColorFromCurve(ColorCurve=DataInterface() {propertyValues {
       AlphaCurve {keys [<at 0 7>,<at 0.1907 1>,<at 0.3261 0.05>,<at 0.5495 0.2>,<at 0.7007 0.2>,<at 0.8474 0 RCIM_Cubic>]}
       BlueCurve {keys [<at 0 0 RCIM_Cubic>,<at 0.1492 0.0104>,<at 0.4863 0.2995>,<at 0.6458 0.0894>]}
       GreenCurve {keys [<at 0 0 RCIM_Cubic>,<at 0.1492 0.0104>,<at 0.4863 0.2995>,<at 0.6458 0.0894>]}
       RedCurve {keys [<at 0 0.4323 RCIM_Cubic>,<at 0.1492 0.9992>,<at 0.4863 0.9274>,<at 0.6458 0.993>]}
      }})
    R2 DefaultColor:LinearColor=&ColorFromCurve(ColorCurve=DataInterface() {propertyValues {
       AlphaCurve {keys [<at 0 1>,<at 0.0914 0.2857>,<at 0.308 0.05>,<at 0.5592 0.2>,<at 0.8632 0>]}
       BlueCurve {keys [<at 0.0009 0>,<at 0.2668 -0.2829>,<at 0.5114 0.8849>,<at 0.619 1>,<at 0.682 0.383>]}
       GreenCurve {keys [<at 0.0009 0.0767>,<at 0.2742 -0.1994>,<at 0.5114 0.8849>,<at 0.619 1>,<at 0.682 0.2585>]}
       RedCurve {keys [<at 0.0009 3.2194>,<at 0.0016 3.2194>,<at 0.4479 0.8849>,<at 0.5114 0.8849>,<at 0.619 1>]}
      }})
    R2 ForestGreen:LinearColor=&ColorFromCurve(ColorCurve=DataInterface() {propertyValues {
       AlphaCurve {keys [<at 0 6>,<at 0.1907 1>,<at 0.3261 0.05>,<at 0.5495 0.2>,<at 0.7007 0.2>,<at 0.8474 0 RCIM_Cubic>]}
       BlueCurve {keys [<at 0 0.0172 RCIM_Cubic>,<at 0.2179 0.0322>,<at 0.4863 0.4656>,<at 0.6537 0.1027>]}
       GreenCurve {keys [<at 0 0.3438 RCIM_Cubic>,<at 0.2179 0.245>,<at 0.4863 0.875>,<at 0.6537 0.2987>]}
       RedCurve {keys [<at 0 0 RCIM_Cubic>,<at 0.2179 0.021>,<at 0.4863 0.4466>,<at 0.6537 0.0924>]}
      }})
    R2 Grey:LinearColor=&ColorFromCurve(ColorCurve=DataInterface() {propertyValues {
       AlphaCurve {keys [<at 0 4>,<at 0.1907 1>,<at 0.3261 0.05>,<at 0.5495 0.2>,<at 0.7007 0.2>,<at 0.8474 0 RCIM_Cubic>]}
       BlueCurve {keys [<at 0 0.0692 RCIM_Cubic>,<at 0.2285 0.1336>,<at 0.4863 0.9274>,<at 0.6405 0.132>]}
       GreenCurve {keys [<at 0 0.0781 RCIM_Cubic>,<at 0.2285 0.1336>,<at 0.4863 0.9274>,<at 0.6405 0.132>]}
       RedCurve {keys [<at 0 0.0767 RCIM_Cubic>,<at 0.2285 0.1406>,<at 0.4863 0.9274>,<at 0.6405 0.1354>]}
      }})
    R2 Lime:LinearColor=&ColorFromCurve(ColorCurve=DataInterface() {propertyValues {
       AlphaCurve {keys [<at 0 6>,<at 0.1907 1>,<at 0.3261 0.05>,<at 0.5495 0.2>,<at 0.8122 0 RCIM_Cubic>]}
       BlueCurve {keys [<at 0 0 RCIM_Cubic>,<at 0.2179 0.0459>,<at 0.4863 0.9274>,<at 0.6299 0.0724>]}
       GreenCurve {keys [<at 0 0.6875 RCIM_Cubic>,<at 0.2179 0.4635>,<at 0.4863 0.9274>,<at 0.6299 0.7708>]}
       RedCurve {keys [<at 0 0.3447 RCIM_Cubic>,<at 0.2179 0.2553>,<at 0.4863 0.9274>,<at 0.6299 0.4226>]}
      }})
    R2 Orange:LinearColor=&ColorFromCurve(ColorCurve=DataInterface() {propertyValues {
       AlphaCurve {keys [<at 0 7>,<at 0.1907 1>,<at 0.3261 0.05>,<at 0.5495 0.2>,<at 0.7007 0.2>,<at 0.8474 0 RCIM_Cubic>]}
       BlueCurve {keys [<at 0 0 RCIM_Cubic>,<at 0.2364 0.0598>,<at 0.4863 0.4263>,<at 0.6217 0.1191>]}
       GreenCurve {keys [<at 0 0.2834 RCIM_Cubic>,<at 0.2364 0.2792>,<at 0.4863 0.5719>,<at 0.6217 0.3121>]}
       RedCurve {keys [<at 0 0.8125 RCIM_Cubic>,<at 0.2364 0.6889>,<at 0.4863 0.8438>,<at 0.6217 0.6724>]}
      }})
    R2 Pink:LinearColor=&ColorFromCurve(ColorCurve=DataInterface() {propertyValues {
       AlphaCurve {keys [<at 0 7>,<at 0.1907 1>,<at 0.3261 0.05>,<at 0.5495 0.2>,<at 0.7007 0.2>,<at 0.8474 0 RCIM_Cubic>]}
       BlueCurve {keys [<at 0 0.3419 RCIM_Cubic>,<at 0.2707 0.3564>,<at 0.4863 0.7216>,<at 0.6141 0.5634>]}
       GreenCurve {keys [<at 0 0.0612 RCIM_Cubic>,<at 0.2707 0.1489>,<at 0.4863 0.5789>,<at 0.6141 0.2536>]}
       RedCurve {keys [<at 0 0.5029 RCIM_Cubic>,<at 0.2707 0.5614>,<at 0.4863 0.855>,<at 0.6141 0.8542>]}
      }})
    R2 Purple:LinearColor=&ColorFromCurve(ColorCurve=DataInterface() {propertyValues {
       AlphaCurve {keys [<at 0 7>,<at 0.1907 1>,<at 0.3261 0.05>,<at 0.5495 0.2>,<at 0.7007 0.2>,<at 0.8474 0 RCIM_Cubic>]}
       BlueCurve {keys [<at 0 0.8281 RCIM_Cubic>,<at 0.2919 0.6836>,<at 0.4863 0.8957>,<at 0.6141 0.9565>]}
       GreenCurve {keys [<at 0 0 RCIM_Cubic>,<at 0.2919 0.0563>,<at 0.4863 0.2791>,<at 0.6141 0.1097>]}
       RedCurve {keys [<at 0 0.1924 RCIM_Cubic>,<at 0.2919 0.3349>,<at 0.4863 0.5023>,<at 0.6141 0.3436>]}
      }})
    R2 Saffron:LinearColor=&ColorFromCurve(ColorCurve=DataInterface() {propertyValues {
       AlphaCurve {keys [<at 0 4>,<at 0.1907 1>,<at 0.3261 0.05>,<at 0.5495 0.2>,<at 0.8175 0 RCIM_Cubic>]}
       BlueCurve {keys [<at 0 0 RCIM_Cubic>,<at 0.2697 0.0046>,<at 0.4863 0.3188>,<at 0.6748 0.0798>]}
       GreenCurve {keys [<at 0 0.6105 RCIM_Cubic>,<at 0.2697 0.6881>,<at 0.4863 0.9274>,<at 0.6748 0.6052>]}
       RedCurve {keys [<at 0 0.7969 RCIM_Cubic>,<at 0.2697 0.8967>,<at 0.4863 0.8744>,<at 0.6748 0.7656>]}
      }})
    R2 SkyBlue:LinearColor=&ColorFromCurve(ColorCurve=DataInterface() {propertyValues {
       AlphaCurve {keys [<at 0 7>,<at 0.1907 1>,<at 0.3261 0.05>,<at 0.5495 0.2>,<at 0.8096 0 RCIM_Cubic>]}
       BlueCurve {keys [<at 0 0.5333 RCIM_Cubic>,<at 0.2496 0.5525>,<at 0.4863 0.9323>,<at 0.6141 0.9323>]}
       GreenCurve {keys [<at 0 0.3564 RCIM_Cubic>,<at 0.2496 0.3792>,<at 0.4863 0.8083>,<at 0.6141 0.6988>]}
       RedCurve {keys [<at 0 0 RCIM_Cubic>,<at 0.2496 0.0875>,<at 0.4863 0.5584>,<at 0.6141 0.2282>]}
      }})
    R2 TitaniumWhite:LinearColor=&ColorFromCurve(ColorCurve=DataInterface() {propertyValues {
       AlphaCurve {keys [<at 0 4>,<at 0.1907 1>,<at 0.3261 0.05>,<at 0.5495 0.2>,<at 0.7007 0.2>,<at 0.8175 0 RCIM_Cubic>]}
       BlueCurve {keys [<at 0 0.1846 RCIM_Cubic>,<at 0.2285 0.1336>,<at 0.4863 0.9274>,<at 0.6405 0.3248>]}
       GreenCurve {keys [<at 0 0.2083 RCIM_Cubic>,<at 0.2285 0.1336>,<at 0.4863 0.9274>,<at 0.6405 0.3248>]}
       RedCurve {keys [<at 0 0.2046 RCIM_Cubic>,<at 0.2285 0.1406>,<at 0.4863 0.9274>,<at 0.6405 0.3333>]}
      }})
    R3 Black:LinearColor=&ColorFromCurve(ColorCurve=DataInterface() {propertyValues {
       AlphaCurve {
        defaultValue=0
        keys [
         {tangentMode=RCTM_User, time=0.0027, value=0}
         {
          arriveTangent=13.1575
          arriveTangentWeight=0.0306
          interpMode=RCIM_Cubic
          leaveTangent=13.1575
          leaveTangentWeight=0.2107
          tangentMode=RCTM_User
          time=0.1591
          value=7
         }
         {tangentMode=RCTM_User, time=0.5992, value=7}
         <at 0.7913 0>
        ]
       }
       BlueCurve {keys [<at 0.6966 0.3043>]}
       GreenCurve {keys [<at 0.6966 0.3043>]}
       RedCurve {keys [<at 0.6966 0.3043>]}
      }})
    R3 BurntSienna:LinearColor=&ColorFromCurve(ColorCurve=DataInterface() {propertyValues {
       AlphaCurve {defaultValue=0, keys [{arriveTangent=13.1575, interpMode=RCIM_Cubic, time=0.0863, value=7}, <at 0.5992 7>, <at 0.7913 0>]}
       BlueCurve {keys [<at 0.6966 0.2729>]}
       GreenCurve {keys [<at 0.6966 0.3693>]}
       RedCurve {keys [<at 0.6966 0.6719>]}
      }})
    R3 Cobalt:LinearColor=&ColorFromCurve(ColorCurve=DataInterface() {propertyValues {
       AlphaCurve {
        defaultValue=0
        keys [
         {tangentMode=RCTM_User, time=0.01, value=0}
         {arriveTangent=13.1575, interpMode=RCIM_Cubic, leaveTangent=13.1575, tangentMode=RCTM_User, time=0.1591, value=7}
         {tangentMode=RCTM_User, time=0.5992, value=10}
         <at 0.7913 0>
        ]
       }
       BlueCurve {keys [<at 0.6966 2.5>]}
       GreenCurve {keys [<at 0.6966 0.2198>]}
       RedCurve {keys [<at 0.6966 0.125>]}
      }})
    R3 Crimson:LinearColor=&ColorFromCurve(ColorCurve=DataInterface() {propertyValues {
       AlphaCurve {
        defaultValue=0
        keys [
         {tangentMode=RCTM_User, time=0, value=0}
         {arriveTangent=13.1575, interpMode=RCIM_Cubic, leaveTangent=13.1575, tangentMode=RCTM_User, time=0.1591, value=7}
         {tangentMode=RCTM_User, time=0.5992, value=7}
         <at 0.7913 0>
        ]
       }
       BlueCurve {keys [<at 0.6966 0>]}
       GreenCurve {keys [<at 0.6966 0>]}
       RedCurve {keys [<at 0.6966 1>]}
      }})
    R3 DefaultColor:LinearColor=&ColorFromCurve(ColorCurve=DataInterface() {propertyValues {
       AlphaCurve {keys [
         {tangentMode=RCTM_User, time=0, value=0}
         {
          arriveTangent=13.1575
          arriveTangentWeight=0.0306
          interpMode=RCIM_Cubic
          leaveTangent=13.1575
          leaveTangentWeight=0.2538
          tangentMode=RCTM_User
          time=0.1591
          value=7
         }
         {tangentMode=RCTM_User, time=0.5992, value=7}
         <at 0.7913 0>
         <at 0.9206 0>
        ]}
       BlueCurve {keys [<at 0.4821 1>]}
       GreenCurve {keys [<at 0.4821 0.0657>]}
       RedCurve {keys [<at 0.4821 0>]}
      }})
    R3 ForestGreen:LinearColor=&ColorFromCurve(ColorCurve=DataInterface() {propertyValues {
       AlphaCurve {
        defaultValue=0
        keys [
         {tangentMode=RCTM_User, time=0, value=0}
         {arriveTangent=13.1575, interpMode=RCIM_Cubic, leaveTangent=13.1575, tangentMode=RCTM_User, time=0.1591, value=7}
         {tangentMode=RCTM_User, time=0.5992, value=7}
         <at 0.7913 0>
        ]
       }
       BlueCurve {keys [<at 0.6966 0.3387>]}
       GreenCurve {keys [<at 0.6966 1.5>]}
       RedCurve {keys [<at 0.6966 0>]}
      }})
    R3 Grey:LinearColor=&ColorFromCurve(ColorCurve=DataInterface() {propertyValues {
       AlphaCurve {
        defaultValue=0
        keys [
         {tangentMode=RCTM_User, time=0.0144, value=0}
         {arriveTangent=13.1575, interpMode=RCIM_Cubic, leaveTangent=13.1575, tangentMode=RCTM_User, time=0.1591, value=7}
         {tangentMode=RCTM_User, time=0.5992, value=7}
         <at 0.7913 0>
        ]
       }
       BlueCurve {keys [<at 0.6966 0.9375>]}
       GreenCurve {keys [<at 0.6966 0.9375>]}
       RedCurve {keys [<at 0.6966 0.9375>]}
      }})
    R3 Lime:LinearColor=&ColorFromCurve(ColorCurve=DataInterface() {propertyValues {
       AlphaCurve {
        defaultValue=0
        keys [
         {tangentMode=RCTM_User, time=0.0085, value=0}
         {arriveTangent=13.1575, interpMode=RCIM_Cubic, leaveTangent=13.1575, tangentMode=RCTM_User, time=0.1591, value=7}
         {tangentMode=RCTM_User, time=0.5992, value=7}
         <at 0.7913 0>
        ]
       }
       BlueCurve {keys [<at 0.6966 0>]}
       GreenCurve {keys [<at 0.6966 1>]}
       RedCurve {keys [<at 0.6966 0.5901>]}
      }})
    R3 Orange:LinearColor=&ColorFromCurve(ColorCurve=DataInterface() {propertyValues {
       AlphaCurve {
        defaultValue=0
        keys [
         {tangentMode=RCTM_User, time=0, value=0}
         {arriveTangent=13.1575, interpMode=RCIM_Cubic, leaveTangent=13.1575, tangentMode=RCTM_User, time=0.1591, value=7}
         {tangentMode=RCTM_User, time=0.5992, value=7}
         <at 0.7913 0>
        ]
       }
       BlueCurve {keys [<at 0.6966 0.2681>]}
       GreenCurve {keys [<at 0.6966 0.4367>]}
       RedCurve {keys [<at 0.6966 0.9531>]}
      }})
    R3 Pink:LinearColor=&ColorFromCurve(ColorCurve=DataInterface() {propertyValues {
       AlphaCurve {
        defaultValue=0
        keys [
         {tangentMode=RCTM_User, time=0.0129, value=0}
         {arriveTangent=13.1575, interpMode=RCIM_Cubic, leaveTangent=13.1575, tangentMode=RCTM_User, time=0.1591, value=7}
         {tangentMode=RCTM_User, time=0.5992, value=7}
         <at 0.7913 0>
        ]
       }
       BlueCurve {keys [<at 0.6966 1.2188>]}
       GreenCurve {keys [<at 0.6966 0.2158>]}
       RedCurve {keys [<at 0.6966 1.1488>]}
      }})
    R3 Purple:LinearColor=&ColorFromCurve(ColorCurve=DataInterface() {propertyValues {
       AlphaCurve {
        defaultValue=0
        keys [
         {tangentMode=RCTM_User, time=0.0114, value=0}
         {arriveTangent=13.1575, interpMode=RCIM_Cubic, leaveTangent=13.1575, tangentMode=RCTM_User, time=0.1591, value=7}
         {tangentMode=RCTM_User, time=0.5992, value=7}
         <at 0.7913 0>
        ]
       }
       BlueCurve {keys [<at 0.6966 1>]}
       GreenCurve {keys [<at 0.6966 0>]}
       RedCurve {keys [<at 0.6966 0.3627>]}
      }})
    R3 Saffron:LinearColor=&ColorFromCurve(ColorCurve=DataInterface() {propertyValues {
       AlphaCurve {
        defaultValue=0
        keys [
         <at 0 0>
         {arriveTangent=13.1575, interpMode=RCIM_Cubic, leaveTangent=13.1575, tangentMode=RCTM_User, time=0.1591, value=7}
         {tangentMode=RCTM_User, time=0.5992, value=7}
         <at 0.7913 0>
        ]
       }
       BlueCurve {keys [<at 0.6966 0>]}
       GreenCurve {keys [<at 0.6966 1.8157>]}
       RedCurve {keys [<at 0.6966 2>]}
      }})
    R3 SkyBlue:LinearColor=&ColorFromCurve(ColorCurve=DataInterface() {propertyValues {
       AlphaCurve {
        defaultValue=0
        keys [
         {tangentMode=RCTM_User, time=0, value=0}
         {arriveTangent=13.1575, interpMode=RCIM_Cubic, leaveTangent=13.1575, tangentMode=RCTM_User, time=0.1591, value=7}
         {tangentMode=RCTM_User, time=0.5992, value=7}
         <at 0.7913 0>
        ]
       }
       BlueCurve {keys [<at 0.6966 1.5>]}
       GreenCurve {keys [<at 0.6966 0.3855>]}
       RedCurve {keys [<at 0.6966 0>]}
      }})
    R3 TitaniumWhite:LinearColor=&ColorFromCurve(ColorCurve=DataInterface() {propertyValues {
       AlphaCurve {keys [
         {tangentMode=RCTM_User, time=0, value=0}
         {arriveTangent=13.1575, interpMode=RCIM_Cubic, leaveTangent=13.1575, tangentMode=RCTM_User, time=0.1591, value=7}
         {tangentMode=RCTM_User, time=0.5992, value=7}
         <at 0.7913 0>
        ]}
       BlueCurve {keys [<at 0.6966 1.8>]}
       GreenCurve {keys [<at 0.6966 1.8>]}
       RedCurve {keys [<at 0.6966 1.8>]}
      }})
    R4 BurntSienna:LinearColor=<rgba 0.0312 0.0096 0.0039 0.6>
    R4 Cobalt:LinearColor=<rgba 0 0.0115 0.6198 0.7>
    R4 Crimson:LinearColor=<rgba 0.2656 0 0.0072 0.7>
    R4 ForestGreen:LinearColor=<rgba 0 0.2552 0.0129 0.6>
    R4 Grey:LinearColor=<rgba 0.0417 0.0417 0.0417 0.7>
    R4 Lime:LinearColor=<rgba 0.1875 0.375 0 0.6>
    R4 Orange:LinearColor=<rgba 0.5521 0.1706 0 0.45>
    R4 Pink:LinearColor=<rgba 0.3906 0.0468 0.2656 0.5>
    R4 Purple:LinearColor=<rgba 0.1567 0 0.7552 0.5>
    R4 Saffron:LinearColor=<rgba 0.5573 0.4299 0 0.3>
    R4 SkyBlue:LinearColor=<rgba 0.0137 0.2957 0.4375 0.5>
    R4 TitaniumWhite:LinearColor=<rgba 0.224 0.224 0.224 0.6>
    R5 Black:LinearColor=<rgba 0.0469 0.0469 0.0469 0.6>
    R5 BurntSienna:LinearColor=<rgba 0.1042 0.0322 0.0128 0.7>
    R5 Cobalt:LinearColor=<rgba 0.0312 0.0401 1 0.6>
    R5 Crimson:LinearColor=<rgba 0.4635 0.0266 0.0384 0.6>
    R5 DefaultColor:LinearColor=<rgba 1 1 1 0.5>
    R5 ForestGreen:LinearColor=<rgba 0.0704 0.901 0.1124 0.4>
    R5 Grey:LinearColor=<rgba 0.1667 0.1667 0.1667 0.4>
    R5 Lime:LinearColor=<rgba 0.2938 0.5104 0.0771 0.25>
    R5 Orange:LinearColor=<rgba 0.8261 0.2993 0.0172 0.3>
    R5 Pink:LinearColor=<rgba 0.651 0.1763 0.4784 0.4>
    R5 Purple:LinearColor=<rgba 0.2704 0.1155 0.7917 0.3>
    R5 Saffron:LinearColor=<rgba 0.8906 0.6824 0 0.15>
    R5 SkyBlue:LinearColor=<rgba 0.0754 0.3822 0.5365 0.4>
    R5 TitaniumWhite:LinearColor=<rgba 0.5729 0.5729 0.5729 0.3>
    Use R1 Painted Overrides=true
    Use R2 Painted Overrides=true
    Use R3 Painted Overrides=true
    Use R4 Painted Overrides=true
    Use R5 Painted Overrides=true
   ) {module=VC_PaintedVariantsOverride001}
   &FN_DisableTireFX(Mask Index Attribute:NiagaraInt32=$Particles.SpawnGroup, Mask Index FL=2, Mask Index FR=3, Mask Index RL=0, Mask Index RR=1)
   &ParticleState()
   &SolveForcesAndVelocity(Clamp Velocity=false, Limit Acceleration=false)
   &VC_UpdateRibbon_v1(Enable Skeletal Position=true)
   &ScaleRibbonWidth(Ribbon Width Scale=&FloatFromCurve(FloatCurve=DataInterface() {propertyValues {Curve {keys [
         {interpMode=RCIM_Cubic, tangentMode=RCTM_Break, time=0, value=1}
         {interpMode=RCIM_Cubic, tangentMode=RCTM_Break, time=0.3, value=1}
         {interpMode=RCIM_Cubic, tangentMode=RCTM_Break, time=0.8, value=0}
        ]}}}))
   /SetVariables(Particles.WPO_Aura:NiagaraPosition=&AddVectorToPosition(
     Position=$Particles.Position
     Vector=&TransformVector(
      Vector=&MakeVector(
       X=0
       Y=&Cosine(
        Normalized Angle=&Add_Float(
         A=&Multiply_Float(A=$Particles.NormalizedAge, B=-1)
         B=&Multiply_Float(A=$Emitter.LoopedAge, B=1)
        )
        Period=&MaskFloatBySpawnGroup(Masked Float=-1.04, Passthrough Float=1.04)
        Scale=&FloatFromCurve(
         FloatCurve=DataInterface() {propertyValues {Curve {
            defaultValue=1
            keys [{interpMode=RCIM_Cubic, tangentMode=RCTM_User, time=0, value=1}, {interpMode=RCIM_Cubic, tangentMode=RCTM_User, time=0.75, value=0}]
           }}}
         Scale Curve=10
        )
        Bias=0
       )
       Z=&Sine(
        Normalized Angle=&Add_Float(
         A=&Multiply_Float(A=$Particles.NormalizedAge, B=-1)
         B=&Multiply_Float(A=$Emitter.LoopedAge, B=1)
        )
        Period=&MaskFloatBySpawnGroup(Masked Float=-1.04, Passthrough Float=1.04)
        Scale=&FloatFromCurve(
         FloatCurve=DataInterface() {propertyValues {Curve {
            defaultValue=1
            keys [{interpMode=RCIM_Cubic, tangentMode=RCTM_User, time=0, value=1}, {interpMode=RCIM_Cubic, tangentMode=RCTM_User, time=0.75, value=0}]
           }}}
         Scale Curve=10
        )
        Bias=0
       )
      )
      Destination Space=::Simulation
      Apply Scale=true
     )
    )) {isSetParameters=true}
   /SetVariables(Particles.WPO_second:NiagaraPosition=&AddVectorToPosition(
     Position=$Particles.Position
     Vector=&TransformVector(
      Vector=&MakeVector(
       X=0
       Y=&Cosine(
        Normalized Angle=&Add_Float(
         A=&Multiply_Float(A=$Particles.NormalizedAge, B=-1)
         B=&Multiply_Float(A=$Emitter.LoopedAge, B=1)
        )
        Period=&MaskFloatBySpawnGroup(Masked Float=-1.04, Passthrough Float=1.04)
        Scale=&FloatFromCurve(
         FloatCurve=DataInterface() {propertyValues {Curve {
            defaultValue=1
            keys [{interpMode=RCIM_Cubic, tangentMode=RCTM_User, time=0, value=1}, {interpMode=RCIM_Cubic, tangentMode=RCTM_User, time=0.75, value=0}]
           }}}
         Scale Curve=-10
        )
        Bias=0
       )
       Z=&Sine(
        Normalized Angle=&Add_Float(
         A=&Multiply_Float(A=$Particles.NormalizedAge, B=-1)
         B=&Multiply_Float(A=$Emitter.LoopedAge, B=1)
        )
        Period=&MaskFloatBySpawnGroup(Masked Float=-1.04, Passthrough Float=1.04)
        Scale=&FloatFromCurve(
         FloatCurve=DataInterface() {propertyValues {Curve {
            defaultValue=1
            keys [{interpMode=RCIM_Cubic, tangentMode=RCTM_User, time=0, value=1}, {interpMode=RCIM_Cubic, tangentMode=RCTM_User, time=0.75, value=0}]
           }}}
         Scale Curve=-10
        )
        Bias=0
       )
      )
      Destination Space=::Simulation
      Apply Scale=true
     )
    )) {isSetParameters=true}
   /SetVariables(
    Particles.R1_Color_Scaled_Main:LinearColor=&MultiplyLinearColor(
     Linear Color=$Particles.R1 Color
     Scale Factor=&MakeVector4FromVectorAndFloat(
      "Vector (RGB)"=&VectorFromFloat(Value=&FloatFromCurve(
        FloatCurve=DataInterface() {propertyValues {Curve {keys [
            {arriveTangent=1.0866, interpMode=RCIM_Cubic, leaveTangent=0.2704, tangentMode=RCTM_Break, time=0, value=0}
            {arriveTangent=0.5459, interpMode=RCIM_Cubic, leaveTangent=0.5459, time=0.08, value=0.5}
            <at 0.9158 0.5>
            <at 1 0>
           ]}}}
        Scale Curve=&MakeCustomFloatFromBool(Bool=$User.bIsPainted, True Float=$User.Power_EmissivePaintedR4, False Float=2)
       ))
      "Float (Alpha)"=&FloatFromCurve(
       FloatCurve=DataInterface() {propertyValues {Curve {keys [
           {arriveTangent=1.0866, interpMode=RCIM_Cubic, time=0.0064, value=0}
           {arriveTangent=1.3524, interpMode=RCIM_Cubic, leaveTangent=1.3524, time=0.066, value=0.6}
           {arriveTangent=-1.1236, interpMode=RCIM_Cubic, leaveTangent=-1.1236, time=0.5, value=0.6676}
           <at 0.6 0 RCIM_Cubic>
          ]}}}
       Scale Curve=&MakeCustomFloatFromBool(Bool=$User.bIsPainted, True Float=$User.Power_AlphaPaintedR4, False Float=0.5)
      )
     )
    )
    Particles.R3_Color_Scaled_Second:LinearColor=&MultiplyLinearColor(
     Linear Color=$Particles.R3 Color
     Scale Factor=&MakeVector4FromVectorAndFloat(
      "Vector (RGB)"=&VectorFromFloat(Value=&FloatFromCurve(
        FloatCurve=DataInterface() {propertyValues {Curve {keys [{arriveTangent=1.0866, interpMode=RCIM_Cubic, leaveTangent=0.2704, tangentMode=RCTM_Break, time=0, value=0.5}, <at 0.7 0.5>, <at 1 0>]}}}
        Scale Curve=&MakeCustomFloatFromBool(Bool=$User.bIsPainted, True Float=$User.Power_EmissivePaintedR4, False Float=2)
       ))
      "Float (Alpha)"=&FloatFromCurve(
       FloatCurve=DataInterface() {propertyValues {Curve {keys [
           {arriveTangent=1.0866, interpMode=RCIM_Cubic, time=0.0064, value=-0.02}
           {arriveTangent=1.393, interpMode=RCIM_Cubic, leaveTangent=1.393, time=0.066, value=0.63}
           {arriveTangent=-1.1797, interpMode=RCIM_Cubic, leaveTangent=-1.1797, time=0.5, value=0.6676}
           <at 0.6 0 RCIM_Cubic>
          ]}}}
       Scale Curve=&MakeCustomFloatFromBool(Bool=$User.bIsPainted, True Float=$User.Power_AlphaPaintedR4, False Float=1)
      )
     )
    )
   ) {isSetParameters=true}
  ]
  rendererClasses [&NiagaraRibbonRendererProperties]
  renderers [
   {rendererClass=&NiagaraRibbonRendererProperties}
   {rendererClass=&NiagaraRibbonRendererProperties, rendererIndex=1}
  ]
 }
 MainRibbon {
  emitterUpdateScript [
   &EmitterState()
   &VC_RateQualityHandler(Epic Spawn Rate=30, Low Spawn Rate=30, Use Custom Rate?=true)
   &VC_SpawnRibbon_v1(Number of Ribbons=2, Spawn Time=-1, SpawnRate=$Emitter.Spawn Rate)
  ]
  particleSpawnScript [
   &VC_PaintedVariantsOverride(Use R1 Painted Overrides=true) {enabled=false}
   &VC_InitializeRibbon_v1(
    Color=$Particles.Color
    Custom Offset=true
    Enable Skeletal Position=true
    Facing Axis:ENiagaraOrientationAxis=::YAxis
    Lifetime=1
    Local Vector Offset=&AddVectorToPosition(Position=<25 0 -20>, Vector=Linked() {linkedVariable {name=User.offset, type=NiagaraPosition}})
    Offset Amount=1
    Ribbon Width=60
   )
   &AddVelocity(Velocity=<-2200 0 0>) {module=AddVelocity001}
  ]
  particleUpdateScript [
   &VC_PaintedVariantsOverride(
    R1 Black:LinearColor=&ColorFromCurve(ColorCurve=DataInterface() {propertyValues {
       AlphaCurve {
        defaultValue=7
        keys [
         <at 0.2819 7>
         <at 0.6302 1>
         <at 0.7212 1>
         {arriveTangent=-6.7193, interpMode=RCIM_Cubic, leaveTangent=-6.7193, time=0.7822, value=1}
         <at 0.87 0>
        ]
       }
       BlueCurve {keys [<at 1 0.1094 RCIM_Cubic>]}
       GreenCurve {keys [<at 1 0.1094 RCIM_Cubic>]}
       RedCurve {keys [<at 1 0.1094 RCIM_Cubic>]}
      }})
    R1 BurntSienna:LinearColor=&ColorFromCurve(ColorCurve=DataInterface() {propertyValues {
       AlphaCurve {
        defaultValue=7
        keys [
         <at 0.2819 7>
         <at 0.6302 1>
         <at 0.7212 1>
         {arriveTangent=-6.7193, interpMode=RCIM_Cubic, leaveTangent=-6.7193, time=0.7822, value=1}
         <at 0.87 0>
        ]
       }
       BlueCurve {keys [<at 1 0.0825 RCIM_Cubic>]}
       GreenCurve {keys [<at 1 0.0972 RCIM_Cubic>]}
       RedCurve {keys [<at 1 0.1667 RCIM_Cubic>]}
      }})
    R1 Cobalt:LinearColor=&ColorFromCurve(ColorCurve=DataInterface() {propertyValues {
       AlphaCurve {
        defaultValue=7
        keys [
         <at 0.2819 7>
         <at 0.6302 1>
         <at 0.7212 1>
         {arriveTangent=-6.7193, interpMode=RCIM_Cubic, leaveTangent=-6.7193, time=0.7822, value=1}
         <at 0.87 0>
        ]
       }
       BlueCurve {keys [<at 1 0.5 RCIM_Cubic>]}
       GreenCurve {keys [<at 1 0 RCIM_Cubic>]}
       RedCurve {keys [<at 1 0 RCIM_Cubic>]}
      }})
    R1 Crimson:LinearColor=&ColorFromCurve(ColorCurve=DataInterface() {propertyValues {
       AlphaCurve {
        defaultValue=7
        keys [
         <at 0.2819 7>
         <at 0.6302 1>
         <at 0.7212 1>
         {arriveTangent=-6.7193, interpMode=RCIM_Cubic, leaveTangent=-6.7193, time=0.7822, value=1}
         <at 0.87 0>
        ]
       }
       BlueCurve {keys [<at 1 0.0131 RCIM_Cubic>]}
       GreenCurve {keys [<at 1 0 RCIM_Cubic>]}
       RedCurve {keys [<at 1 0.2348 RCIM_Cubic>]}
      }})
    R1 DefaultColor=&ColorFromCurve(ColorCurve=DataInterface() {propertyValues {
       AlphaCurve {keys [<at 0.2819 7>, <at 0.6302 1>, {interpMode=RCIM_Cubic, leaveTangent=-3.2831, time=0.7822, value=1}]}
       BlueCurve {keys [<at 1 0.15 RCIM_Cubic>]}
       GreenCurve {keys [<at 1 0.0003 RCIM_Cubic>]}
       RedCurve {keys [<at 1 0.0529 RCIM_Cubic>]}
      }})
    R1 ForestGreen:LinearColor=&ColorFromCurve(ColorCurve=DataInterface() {propertyValues {
       AlphaCurve {
        defaultValue=7
        keys [
         <at 0.2819 7>
         <at 0.6302 1>
         <at 0.7212 1>
         {arriveTangent=-6.7193, interpMode=RCIM_Cubic, leaveTangent=-6.7193, time=0.7822, value=1}
         <at 0.87 0>
        ]
       }
       BlueCurve {keys [<at 0.6169 0.0198 RCIM_Cubic>]}
       GreenCurve {keys [<at 0.6169 0.2031 RCIM_Cubic>]}
       RedCurve {keys [<at 0.6169 0 RCIM_Cubic>]}
      }})
    R1 Grey:LinearColor=&ColorFromCurve(ColorCurve=DataInterface() {propertyValues {
       AlphaCurve {
        defaultValue=1
        keys [
         <at 0.2819 7>
         <at 0.6302 1>
         <at 0.7212 1>
         {arriveTangent=-6.7193, interpMode=RCIM_Cubic, leaveTangent=-6.7193, time=0.7822, value=1}
         <at 0.87 0>
        ]
       }
       BlueCurve {keys [<at 1 0.1979 RCIM_Cubic>]}
       GreenCurve {keys [<at 1 0.1979 RCIM_Cubic>]}
       RedCurve {keys [<at 1 0.1979 RCIM_Cubic>]}
      }})
    R1 Lime:LinearColor=&ColorFromCurve(ColorCurve=DataInterface() {propertyValues {
       AlphaCurve {
        defaultValue=7
        keys [
         <at 0.2819 7>
         <at 0.6302 1>
         <at 0.7212 1>
         {arriveTangent=-6.7193, interpMode=RCIM_Cubic, leaveTangent=-6.7193, time=0.7822, value=1}
         <at 0.87 0>
        ]
       }
       BlueCurve {keys [<at 1 0.0865 RCIM_Cubic>]}
       GreenCurve {keys [<at 1 0.2865 RCIM_Cubic>]}
       RedCurve {keys [<at 1 0.1697 RCIM_Cubic>]}
      }})
    R1 Orange:LinearColor=&ColorFromCurve(ColorCurve=DataInterface() {propertyValues {
       AlphaCurve {
        defaultValue=7
        keys [
         <at 0.2819 7>
         <at 0.6302 1>
         <at 0.7212 1>
         {arriveTangent=-6.7193, interpMode=RCIM_Cubic, leaveTangent=-6.7193, time=0.7822, value=1}
         <at 0.87 0>
        ]
       }
       BlueCurve {keys [<at 1 0.0622 RCIM_Cubic>]}
       GreenCurve {keys [<at 1 0.1169 RCIM_Cubic>]}
       RedCurve {keys [<at 1 0.3073 RCIM_Cubic>]}
      }})
    R1 Pink:LinearColor=&ColorFromCurve(ColorCurve=DataInterface() {propertyValues {
       AlphaCurve {
        defaultValue=7
        keys [
         <at 0.2819 7>
         <at 0.6302 1>
         <at 0.7212 1>
         {arriveTangent=-6.7193, interpMode=RCIM_Cubic, leaveTangent=-6.7193, time=0.7822, value=1}
         <at 0.87 0>
        ]
       }
       BlueCurve {keys [<at 1 0.2004 RCIM_Cubic>]}
       GreenCurve {keys [<at 1 0 RCIM_Cubic>]}
       RedCurve {keys [<at 1 0.2188 RCIM_Cubic>]}
      }})
    R1 Purple:LinearColor=&ColorFromCurve(ColorCurve=DataInterface() {propertyValues {
       AlphaCurve {
        defaultValue=7
        keys [
         <at 0.2819 7>
         <at 0.6302 1>
         <at 0.7212 1>
         {arriveTangent=-6.7193, interpMode=RCIM_Cubic, leaveTangent=-6.7193, time=0.7822, value=1}
         <at 0.87 0>
        ]
       }
       BlueCurve {keys [<at 1 0.375 RCIM_Cubic>]}
       GreenCurve {keys [<at 1 0 RCIM_Cubic>]}
       RedCurve {keys [<at 1 0.1484 RCIM_Cubic>]}
      }})
    R1 Saffron:LinearColor=&ColorFromCurve(ColorCurve=DataInterface() {propertyValues {
       AlphaCurve {
        defaultValue=7
        keys [
         <at 0.2819 7>
         <at 0.6302 1>
         <at 0.7212 1>
         {arriveTangent=-6.7193, interpMode=RCIM_Cubic, leaveTangent=-6.7193, time=0.7822, value=1}
         <at 0.87 0>
        ]
       }
       BlueCurve {keys [<at 1 0.0278 RCIM_Cubic>]}
       GreenCurve {keys [<at 1 0.2596 RCIM_Cubic>]}
       RedCurve {keys [<at 1 0.2812 RCIM_Cubic>]}
      }})
    R1 SkyBlue:LinearColor=&ColorFromCurve(ColorCurve=DataInterface() {propertyValues {
       AlphaCurve {
        defaultValue=7
        keys [
         <at 0.2819 7>
         <at 0.6302 1>
         <at 0.7212 1>
         {arriveTangent=-6.7193, interpMode=RCIM_Cubic, leaveTangent=-6.7193, time=0.7822, value=1}
         <at 0.87 0>
        ]
       }
       BlueCurve {keys [<at 1 0.375 RCIM_Cubic>]}
       GreenCurve {keys [<at 1 0.1761 RCIM_Cubic>]}
       RedCurve {keys [<at 1 0.0547 RCIM_Cubic>]}
      }})
    R1 TitaniumWhite:LinearColor=&ColorFromCurve(ColorCurve=DataInterface() {propertyValues {
       AlphaCurve {
        defaultValue=2
        keys [
         <at 0.2819 7>
         <at 0.6302 1>
         <at 0.7212 1>
         {arriveTangent=-6.7193, interpMode=RCIM_Cubic, leaveTangent=-6.7193, time=0.7822, value=1}
         <at 0.87 0>
        ]
       }
       BlueCurve {keys [<at 0.6118 0.3802 RCIM_Cubic>,<at 1 0.0677>]}
       GreenCurve {keys [<at 0.6118 0.3802 RCIM_Cubic>,<at 1 0.0677>]}
       RedCurve {keys [<at 0.6118 0.3802 RCIM_Cubic>,<at 1 0.0677>]}
      }})
    R2 Black:LinearColor=&ColorFromCurve(ColorCurve=DataInterface() {propertyValues {
       AlphaCurve {keys [<at 0 7>,<at 0.2232 1>,<at 0.3261 0.05>,<at 0.5495 0.2>,<at 0.7007 0.2>,<at 0.8474 0 RCIM_Cubic>]}
       BlueCurve {keys [<at 0 0.0194 RCIM_Cubic>,<at 0.1492 0.0333>,<at 0.2681 0.1104>,<at 0.4863 0.9274>,<at 0.6088 0.0305>]}
       GreenCurve {keys [<at 0 0.0194 RCIM_Cubic>,<at 0.1492 0.0333>,<at 0.2681 0.1104>,<at 0.4863 0.9274>,<at 0.6088 0.0305>]}
       RedCurve {keys [<at 0 0.0194 RCIM_Cubic>,<at 0.1492 0.0333>,<at 0.2681 0.1104>,<at 0.4863 0.9274>,<at 0.6088 0.0305>]}
      }})
    R2 BurntSienna:LinearColor=&ColorFromCurve(ColorCurve=DataInterface() {propertyValues {
       AlphaCurve {keys [<at 0 4>,<at 0.1907 1>,<at 0.3261 0.05>,<at 0.593 0.2>,<at 0.8474 0 RCIM_Cubic>]}
       BlueCurve {keys [<at 0 0.008 RCIM_Cubic>,<at 0.2839 0.0202>,<at 0.4863 0.2907>,<at 0.5956 0.008>]}
       GreenCurve {keys [<at 0 0.0203 RCIM_Cubic>,<at 0.2839 0.0323>,<at 0.4863 0.4394>,<at 0.5956 0.0203>]}
       RedCurve {keys [<at 0 0.0648 RCIM_Cubic>,<at 0.2839 0.0762>,<at 0.4863 0.9792>,<at 0.5956 0.0648>]}
      }})
    R2 Cobalt:LinearColor=&ColorFromCurve(ColorCurve=DataInterface() {propertyValues {
       AlphaCurve {keys [<at 0 7>,<at 0.1907 1>,<at 0.3261 0.05>,<at 0.5495 0.6>,<at 0.7007 0.2>,<at 0.8474 0 RCIM_Cubic>]}
       BlueCurve {keys [<at 0 0.6458 RCIM_Cubic>,<at 0.2575 0.9985>,<at 0.4863 1>,<at 0.6115 0.8177>]}
       GreenCurve {keys [<at 0 0.0119 RCIM_Cubic>,<at 0.2575 0.0367>,<at 0.4863 0.2997>,<at 0.6115 0.0527>]}
       RedCurve {keys [<at 0 0 RCIM_Cubic>,<at 0.2575 0.0186>,<at 0.4863 0.2865>,<at 0.6115 0.0383>]}
      }})
    R2 Crimson:LinearColor=&ColorFromCurve(ColorCurve=DataInterface() {propertyValues {
       AlphaCurve {keys [<at 0 7>,<at 0.1907 1>,<at 0.3261 0.05>,<at 0.5495 0.2>,<at 0.7007 0.2>,<at 0.8474 0 RCIM_Cubic>]}
       BlueCurve {keys [<at 0 0 RCIM_Cubic>,<at 0.1492 0.0104>,<at 0.4863 0.2995>,<at 0.6458 0.0894>]}
       GreenCurve {keys [<at 0 0 RCIM_Cubic>,<at 0.1492 0.0104>,<at 0.4863 0.2995>,<at 0.6458 0.0894>]}
       RedCurve {keys [<at 0 0.4323 RCIM_Cubic>,<at 0.1492 0.9992>,<at 0.4863 0.9274>,<at 0.6458 0.993>]}
      }})
    R2 DefaultColor:LinearColor=&ColorFromCurve(ColorCurve=DataInterface() {propertyValues {
       AlphaCurve {keys [<at 0 1>,<at 0.0914 0.2857>,<at 0.308 0.05>,<at 0.5592 0.2>,<at 0.8632 0>]}
       BlueCurve {keys [<at 0.0009 0>,<at 0.2668 -0.2829>,<at 0.5114 0.8849>,<at 0.619 1>,<at 0.682 0.383>]}
       GreenCurve {keys [<at 0.0009 0.0767>,<at 0.2742 -0.1994>,<at 0.5114 0.8849>,<at 0.619 1>,<at 0.682 0.2585>]}
       RedCurve {keys [<at 0.0009 3.2194>,<at 0.0016 3.2194>,<at 0.4479 0.8849>,<at 0.5114 0.8849>,<at 0.619 1>]}
      }})
    R2 ForestGreen:LinearColor=&ColorFromCurve(ColorCurve=DataInterface() {propertyValues {
       AlphaCurve {keys [<at 0 6>,<at 0.1907 1>,<at 0.3261 0.05>,<at 0.5495 0.2>,<at 0.7007 0.2>,<at 0.8474 0 RCIM_Cubic>]}
       BlueCurve {keys [<at 0 0.0172 RCIM_Cubic>,<at 0.2179 0.0322>,<at 0.4863 0.4656>,<at 0.6537 0.1027>]}
       GreenCurve {keys [<at 0 0.3438 RCIM_Cubic>,<at 0.2179 0.245>,<at 0.4863 0.875>,<at 0.6537 0.2987>]}
       RedCurve {keys [<at 0 0 RCIM_Cubic>,<at 0.2179 0.021>,<at 0.4863 0.4466>,<at 0.6537 0.0924>]}
      }})
    R2 Grey:LinearColor=&ColorFromCurve(ColorCurve=DataInterface() {propertyValues {
       AlphaCurve {keys [<at 0 4>,<at 0.1907 1>,<at 0.3261 0.05>,<at 0.5495 0.2>,<at 0.7007 0.2>,<at 0.8474 0 RCIM_Cubic>]}
       BlueCurve {keys [<at 0 0.0692 RCIM_Cubic>,<at 0.2285 0.1336>,<at 0.4863 0.9274>,<at 0.6405 0.132>]}
       GreenCurve {keys [<at 0 0.0781 RCIM_Cubic>,<at 0.2285 0.1336>,<at 0.4863 0.9274>,<at 0.6405 0.132>]}
       RedCurve {keys [<at 0 0.0767 RCIM_Cubic>,<at 0.2285 0.1406>,<at 0.4863 0.9274>,<at 0.6405 0.1354>]}
      }})
    R2 Lime:LinearColor=&ColorFromCurve(ColorCurve=DataInterface() {propertyValues {
       AlphaCurve {keys [<at 0 6>,<at 0.1907 1>,<at 0.3261 0.05>,<at 0.5495 0.2>,<at 0.8122 0 RCIM_Cubic>]}
       BlueCurve {keys [<at 0 0 RCIM_Cubic>,<at 0.2179 0.0459>,<at 0.4863 0.9274>,<at 0.6299 0.0724>]}
       GreenCurve {keys [<at 0 0.6875 RCIM_Cubic>,<at 0.2179 0.4635>,<at 0.4863 0.9274>,<at 0.6299 0.7708>]}
       RedCurve {keys [<at 0 0.3447 RCIM_Cubic>,<at 0.2179 0.2553>,<at 0.4863 0.9274>,<at 0.6299 0.4226>]}
      }})
    R2 Orange:LinearColor=&ColorFromCurve(ColorCurve=DataInterface() {propertyValues {
       AlphaCurve {keys [<at 0 7>,<at 0.1907 1>,<at 0.3261 0.05>,<at 0.5495 0.2>,<at 0.7007 0.2>,<at 0.8474 0 RCIM_Cubic>]}
       BlueCurve {keys [<at 0 0 RCIM_Cubic>,<at 0.2364 0.0598>,<at 0.4863 0.4263>,<at 0.6217 0.1191>]}
       GreenCurve {keys [<at 0 0.2834 RCIM_Cubic>,<at 0.2364 0.2792>,<at 0.4863 0.5719>,<at 0.6217 0.3121>]}
       RedCurve {keys [<at 0 0.8125 RCIM_Cubic>,<at 0.2364 0.6889>,<at 0.4863 0.8438>,<at 0.6217 0.6724>]}
      }})
    R2 Pink:LinearColor=&ColorFromCurve(ColorCurve=DataInterface() {propertyValues {
       AlphaCurve {keys [<at 0 7>,<at 0.1907 1>,<at 0.3261 0.05>,<at 0.5495 0.2>,<at 0.7007 0.2>,<at 0.8474 0 RCIM_Cubic>]}
       BlueCurve {keys [<at 0 0.3419 RCIM_Cubic>,<at 0.2707 0.3564>,<at 0.4863 0.7216>,<at 0.6141 0.5634>]}
       GreenCurve {keys [<at 0 0.0612 RCIM_Cubic>,<at 0.2707 0.1489>,<at 0.4863 0.5789>,<at 0.6141 0.2536>]}
       RedCurve {keys [<at 0 0.5029 RCIM_Cubic>,<at 0.2707 0.5614>,<at 0.4863 0.855>,<at 0.6141 0.8542>]}
      }})
    R2 Purple:LinearColor=&ColorFromCurve(ColorCurve=DataInterface() {propertyValues {
       AlphaCurve {keys [<at 0 7>,<at 0.1907 1>,<at 0.3261 0.05>,<at 0.5495 0.2>,<at 0.7007 0.2>,<at 0.8474 0 RCIM_Cubic>]}
       BlueCurve {keys [<at 0 0.8281 RCIM_Cubic>,<at 0.2919 0.6836>,<at 0.4863 0.8957>,<at 0.6141 0.9565>]}
       GreenCurve {keys [<at 0 0 RCIM_Cubic>,<at 0.2919 0.0563>,<at 0.4863 0.2791>,<at 0.6141 0.1097>]}
       RedCurve {keys [<at 0 0.1924 RCIM_Cubic>,<at 0.2919 0.3349>,<at 0.4863 0.5023>,<at 0.6141 0.3436>]}
      }})
    R2 Saffron:LinearColor=&ColorFromCurve(ColorCurve=DataInterface() {propertyValues {
       AlphaCurve {keys [<at 0 4>,<at 0.1907 1>,<at 0.3261 0.05>,<at 0.5495 0.2>,<at 0.8175 0 RCIM_Cubic>]}
       BlueCurve {keys [<at 0 0 RCIM_Cubic>,<at 0.2697 0.0046>,<at 0.4863 0.3188>,<at 0.6748 0.0798>]}
       GreenCurve {keys [<at 0 0.6105 RCIM_Cubic>,<at 0.2697 0.6881>,<at 0.4863 0.9274>,<at 0.6748 0.6052>]}
       RedCurve {keys [<at 0 0.7969 RCIM_Cubic>,<at 0.2697 0.8967>,<at 0.4863 0.8744>,<at 0.6748 0.7656>]}
      }})
    R2 SkyBlue:LinearColor=&ColorFromCurve(ColorCurve=DataInterface() {propertyValues {
       AlphaCurve {keys [<at 0 7>,<at 0.1907 1>,<at 0.3261 0.05>,<at 0.5495 0.2>,<at 0.8096 0 RCIM_Cubic>]}
       BlueCurve {keys [<at 0 0.5333 RCIM_Cubic>,<at 0.2496 0.5525>,<at 0.4863 0.9323>,<at 0.6141 0.9323>]}
       GreenCurve {keys [<at 0 0.3564 RCIM_Cubic>,<at 0.2496 0.3792>,<at 0.4863 0.8083>,<at 0.6141 0.6988>]}
       RedCurve {keys [<at 0 0 RCIM_Cubic>,<at 0.2496 0.0875>,<at 0.4863 0.5584>,<at 0.6141 0.2282>]}
      }})
    R2 TitaniumWhite:LinearColor=&ColorFromCurve(ColorCurve=DataInterface() {propertyValues {
       AlphaCurve {keys [<at 0 4>,<at 0.1907 1>,<at 0.3261 0.05>,<at 0.5495 0.2>,<at 0.7007 0.2>,<at 0.8175 0 RCIM_Cubic>]}
       BlueCurve {keys [<at 0 0.1846 RCIM_Cubic>,<at 0.2285 0.1336>,<at 0.4863 0.9274>,<at 0.6405 0.3248>]}
       GreenCurve {keys [<at 0 0.2083 RCIM_Cubic>,<at 0.2285 0.1336>,<at 0.4863 0.9274>,<at 0.6405 0.3248>]}
       RedCurve {keys [<at 0 0.2046 RCIM_Cubic>,<at 0.2285 0.1406>,<at 0.4863 0.9274>,<at 0.6405 0.3333>]}
      }})
    R3 Black:LinearColor=&ColorFromCurve(ColorCurve=DataInterface() {propertyValues {
       AlphaCurve {keys [{tangentMode=RCTM_User, time=0.8086, value=3}, {tangentMode=RCTM_User, time=0.8837, value=0}, {tangentMode=RCTM_User, time=1, value=0}]}
       BlueCurve {keys [<at 0.6966 0.25>]}
       GreenCurve {keys [<at 0.6966 0.25>]}
       RedCurve {keys [<at 0.6966 0.25>]}
      }})
    R3 BurntSienna:LinearColor=&ColorFromCurve(ColorCurve=DataInterface() {propertyValues {
       AlphaCurve {
        defaultValue=1
        keys [{tangentMode=RCTM_User, time=0.8086, value=3}, {tangentMode=RCTM_User, time=0.8837, value=0}, {tangentMode=RCTM_User, time=1, value=0}]
       }
       BlueCurve {keys [<at 0.6966 0.147>]}
       GreenCurve {keys [<at 0.6966 0.2197>]}
       RedCurve {keys [<at 0.6966 0.4479>]}
      }})
    R3 Cobalt:LinearColor=&ColorFromCurve(ColorCurve=DataInterface() {propertyValues {
       AlphaCurve {
        defaultValue=3
        keys [{tangentMode=RCTM_User, time=0.8086, value=1}, {tangentMode=RCTM_User, time=0.8837, value=0}, {tangentMode=RCTM_User, time=1, value=0}]
       }
       BlueCurve {keys [<at 0.6966 1.3>]}
       GreenCurve {keys [<at 0.6966 0.1083>]}
       RedCurve {keys [<at 0.6966 0>]}
      }})
    R3 Crimson:LinearColor=&ColorFromCurve(ColorCurve=DataInterface() {propertyValues {
       AlphaCurve {
        defaultValue=3
        keys [{tangentMode=RCTM_User, time=0.8086, value=3}, {tangentMode=RCTM_User, time=0.8837, value=0}, {tangentMode=RCTM_User, time=1, value=0}]
       }
       BlueCurve {keys [<at 0.6966 0>]}
       GreenCurve {keys [<at 0.6966 0>]}
       RedCurve {keys [<at 0.6966 1>]}
      }})
    R3 DefaultColor:LinearColor=&ColorFromCurve(ColorCurve=DataInterface() {propertyValues {
       AlphaCurve {keys [<at 0.8086 3>,<at 0.8837 0>,<at 1 0>]}
       BlueCurve {keys [<at 0.6966 1>]}
       GreenCurve {keys [<at 0.6966 0.2>]}
       RedCurve {keys [<at 0.6966 0>]}
      }})
    R3 ForestGreen:LinearColor=&ColorFromCurve(ColorCurve=DataInterface() {propertyValues {
       AlphaCurve {
        defaultValue=3
        keys [{tangentMode=RCTM_User, time=0.8086, value=3}, {tangentMode=RCTM_User, time=0.8837, value=0}, {tangentMode=RCTM_User, time=1, value=0}]
       }
       BlueCurve {keys [<at 0.6966 0.2593>]}
       GreenCurve {keys [<at 0.6966 0.6562>]}
       RedCurve {keys [<at 0.6966 0.1436>]}
      }})
    R3 Grey:LinearColor=&ColorFromCurve(ColorCurve=DataInterface() {propertyValues {
       AlphaCurve {
        defaultValue=3
        keys [{tangentMode=RCTM_User, time=0.8086, value=3}, {tangentMode=RCTM_User, time=0.8837, value=0}, {tangentMode=RCTM_User, time=1, value=0}]
       }
       BlueCurve {keys [<at 0.6966 0.3646>]}
       GreenCurve {keys [<at 0.6966 0.3646>]}
       RedCurve {keys [<at 0.6966 0.3646>]}
      }})
    R3 Lime:LinearColor=&ColorFromCurve(ColorCurve=DataInterface() {propertyValues {
       AlphaCurve {
        defaultValue=3
        keys [{tangentMode=RCTM_User, time=0.8086, value=3}, {tangentMode=RCTM_User, time=0.8837, value=0}, {tangentMode=RCTM_User, time=1, value=0}]
       }
       BlueCurve {keys [<at 0.6966 0>]}
       GreenCurve {keys [<at 0.6966 1>]}
       RedCurve {keys [<at 0.6966 0.5901>]}
      }})
    R3 Orange:LinearColor=&ColorFromCurve(ColorCurve=DataInterface() {propertyValues {
       AlphaCurve {
        defaultValue=3
        keys [{tangentMode=RCTM_User, time=0.8086, value=3}, {tangentMode=RCTM_User, time=0.8837, value=0}, {tangentMode=RCTM_User, time=1, value=0}]
       }
       BlueCurve {keys [<at 0.6966 0.1367>]}
       GreenCurve {keys [<at 0.6966 0.2734>]}
       RedCurve {keys [<at 0.6966 0.7292>]}
      }})
    R3 Pink:LinearColor=&ColorFromCurve(ColorCurve=DataInterface() {propertyValues {
       AlphaCurve {
        defaultValue=3
        keys [{tangentMode=RCTM_User, time=0.8086, value=3}, {tangentMode=RCTM_User, time=0.8837, value=0}, {tangentMode=RCTM_User, time=1, value=0}]
       }
       BlueCurve {keys [<at 0.6966 0.6855>]}
       GreenCurve {keys [<at 0.6966 0.1214>]}
       RedCurve {keys [<at 0.6966 0.6462>]}
      }})
    R3 Purple:LinearColor=&ColorFromCurve(ColorCurve=DataInterface() {propertyValues {
       AlphaCurve {
        defaultValue=3
        keys [{tangentMode=RCTM_User, time=0.8086, value=3}, {tangentMode=RCTM_User, time=0.8837, value=0}, {tangentMode=RCTM_User, time=1, value=0}]
       }
       BlueCurve {keys [<at 0.6966 0.8906>]}
       GreenCurve {keys [<at 0.6966 0>]}
       RedCurve {keys [<at 0.6966 0.2841>]}
      }})
    R3 Saffron:LinearColor=&ColorFromCurve(ColorCurve=DataInterface() {propertyValues {
       AlphaCurve {
        defaultValue=3
        keys [{tangentMode=RCTM_User, time=0.8086, value=3}, {tangentMode=RCTM_User, time=0.8837, value=0}, {tangentMode=RCTM_User, time=1, value=0}]
       }
       BlueCurve {keys [<at 0.6966 0.199>]}
       GreenCurve {keys [<at 0.6966 1.2571>]}
       RedCurve {keys [<at 0.6966 1.3646>]}
      }})
    R3 SkyBlue:LinearColor=&ColorFromCurve(ColorCurve=DataInterface() {propertyValues {
       AlphaCurve {
        defaultValue=3
        keys [{tangentMode=RCTM_User, time=0.8086, value=3}, {tangentMode=RCTM_User, time=0.8837, value=0}, {tangentMode=RCTM_User, time=1, value=0}]
       }
       BlueCurve {keys [<at 0.6966 0.9375>]}
       GreenCurve {keys [<at 0.6966 0.4441>]}
       RedCurve {keys [<at 0.6966 0>]}
      }})
    R3 TitaniumWhite:LinearColor=&ColorFromCurve(ColorCurve=DataInterface() {propertyValues {
       AlphaCurve {keys [<at 0.4952 1>,<at 0.9629 0>]}
       BlueCurve {keys [<at 0.6966 2.5>]}
       GreenCurve {keys [<at 0.6966 2.5>]}
       RedCurve {keys [<at 0.6966 2.5>]}
      }})
    R4 Black:LinearColor=&ColorFromCurve(ColorCurve=DataInterface() {propertyValues {
       AlphaCurve {keys [<at 0.0115 0.9906>,<at 0.3154 1>,<at 0.5062 0>]}
       BlueCurve {keys [<at 0.3014 0>]}
       GreenCurve {keys [<at 0.3014 0>]}
       RedCurve {keys [<at 0.3014 0>]}
      }})
    R4 BurntSienna:LinearColor=&ColorFromCurve(ColorCurve=DataInterface() {propertyValues {
       AlphaCurve {keys [<at 0.0115 0.9906>,<at 0.3154 1>,<at 0.5062 0>]}
       BlueCurve {keys [<at 0.3014 0>]}
       GreenCurve {keys [<at 0.3014 0>]}
       RedCurve {keys [<at 0.3014 0>]}
      }})
    R4 Cobalt:LinearColor=&ColorFromCurve(ColorCurve=DataInterface() {propertyValues {
       AlphaCurve {keys [<at 0.0115 0.9906>,<at 0.3154 1>,<at 0.5062 0>]}
       BlueCurve {keys [<at 0.3014 0>]}
       GreenCurve {keys [<at 0.3014 0>]}
       RedCurve {keys [<at 0.3014 0>]}
      }})
    R4 Crimson:LinearColor=&ColorFromCurve(ColorCurve=DataInterface() {propertyValues {
       AlphaCurve {keys [<at 0.0115 0.9906>,<at 0.3154 1>,<at 0.5062 0>]}
       BlueCurve {keys [<at 0.0688 0.069>,<at 0.2065 0.0676>,<at 0.3014 0>]}
       GreenCurve {keys [<at 0.0688 0.069>,<at 0.2065 0.0676>,<at 0.3014 0>]}
       RedCurve {keys [<at 0.0688 0.25>,<at 0.2065 0.2448>,<at 0.3014 0>]}
      }})
    R4 DefaultColor:LinearColor=&ColorFromCurve(ColorCurve=DataInterface() {propertyValues {
       AlphaCurve {keys [<at 0.0115 0.9906>,<at 0.3154 1>,<at 0.5062 0>]}
       BlueCurve {keys [<at 0.1589 0.1979>,<at 0.3466 0>,<at 0.3491 0>]}
       GreenCurve {keys [<at 0.1589 0.0166>,<at 0.3466 0>,<at 0.3491 0>]}
       RedCurve {keys [<at 0.1589 0.105>,<at 0.3466 0>,<at 0.3491 0>]}
      }})
    R4 ForestGreen:LinearColor=&ColorFromCurve(ColorCurve=DataInterface() {propertyValues {
       AlphaCurve {keys [<at 0.0115 0.9906>,<at 0.3154 1>,<at 0.5062 0>]}
       BlueCurve {keys [<at 0.229 0.0107>,<at 0.4942 0>]}
       GreenCurve {keys [<at 0.229 0.2031>,<at 0.4942 0>]}
       RedCurve {keys [<at 0.229 0>,<at 0.4942 0>]}
      }})
    R4 Grey:LinearColor=&ColorFromCurve(ColorCurve=DataInterface() {propertyValues {
       AlphaCurve {keys [<at 0.0115 0.9906>,<at 0.3154 1>,<at 0.5062 0>]}
       BlueCurve {keys [<at 0.1914 0.1875>,<at 0.3014 0>]}
       GreenCurve {keys [<at 0.1914 0.1875>,<at 0.3014 0>]}
       RedCurve {keys [<at 0.1914 0.1875>,<at 0.3014 0>]}
      }})
    R4 Lime:LinearColor=&ColorFromCurve(ColorCurve=DataInterface() {propertyValues {
       AlphaCurve {keys [<at 0.0115 0.9906>,<at 0.3154 1>,<at 0.5062 0>]}
       BlueCurve {keys [<at 0.3014 0.0328>,<at 0.4842 0.0068>]}
       GreenCurve {keys [<at 0.3014 0.151>,<at 0.4842 0.0312>]}
       RedCurve {keys [<at 0.3014 0.0925>,<at 0.4842 0.0191>]}
      }})
    R4 Orange:LinearColor=&ColorFromCurve(ColorCurve=DataInterface() {propertyValues {
       AlphaCurve {keys [<at 0.0115 0.9906>,<at 0.3154 1>,<at 0.5062 0>]}
       BlueCurve {keys [<at 0.2565 0.0204>,<at 0.4842 0>]}
       GreenCurve {keys [<at 0.2565 0.0617>,<at 0.4842 0>]}
       RedCurve {keys [<at 0.2565 0.276>,<at 0.4842 0>]}
      }})
    R4 Pink:LinearColor=&ColorFromCurve(ColorCurve=DataInterface() {propertyValues {
       AlphaCurve {keys [<at 0.0115 0.9906>,<at 0.3154 1>,<at 0.5062 0>]}
       BlueCurve {keys [<at 0.279 0.6169>,<at 0.4817 0>]}
       GreenCurve {keys [<at 0.279 0.087>,<at 0.4817 0>]}
       RedCurve {keys [<at 0.279 0.6198>,<at 0.4817 0>]}
      }})
    R4 Purple:LinearColor=&ColorFromCurve(ColorCurve=DataInterface() {propertyValues {
       AlphaCurve {keys [<at 0.0115 0.9906>,<at 0.3154 1>,<at 0.5062 0>]}
       BlueCurve {keys [<at 0.3014 0>]}
       GreenCurve {keys [<at 0.3014 0>]}
       RedCurve {keys [<at 0.3014 0>]}
      }})
    R4 Saffron:LinearColor=&ColorFromCurve(ColorCurve=DataInterface() {propertyValues {
       AlphaCurve {keys [<at 0.0115 0.9906>,<at 0.3154 1>,<at 0.5062 0>]}
       BlueCurve {keys [<at 0.3014 0>]}
       GreenCurve {keys [<at 0.3014 0>]}
       RedCurve {keys [<at 0.3014 0>]}
      }})
    R4 SkyBlue:LinearColor=&ColorFromCurve(ColorCurve=DataInterface() {propertyValues {
       AlphaCurve {keys [<at 0.0115 0.9906>,<at 0.3154 1>,<at 0.5062 0>]}
       BlueCurve {keys [<at 0.3014 0>]}
       GreenCurve {keys [<at 0.3014 0>]}
       RedCurve {keys [<at 0.3014 0>]}
      }})
    R4 TitaniumWhite:LinearColor=&ColorFromCurve(ColorCurve=DataInterface() {propertyValues {
       AlphaCurve {keys [<at 0.0115 0.9906>,<at 0.1682 1>,<at 0.3697 0>]}
       BlueCurve {keys [<at 0.3509 0.2708>,<at 0.4792 0.0521>]}
       GreenCurve {keys [<at 0.3509 0.2708>,<at 0.4792 0.0521>]}
       RedCurve {keys [<at 0.3509 0.2708>,<at 0.4792 0.0521>]}
      }})
    R5 Black:LinearColor=<rgba 0.0469 0.0469 0.0469 0.6>
    R5 BurntSienna:LinearColor=<rgba 0.1042 0.0322 0.0128 0.7>
    R5 Cobalt:LinearColor=<rgba 0.0312 0.0401 1 0.6>
    R5 Crimson:LinearColor=<rgba 0.4635 0.0266 0.0384 0.6>
    R5 DefaultColor:LinearColor=<rgba 1 1 1 0.5>
    R5 ForestGreen:LinearColor=<rgba 0.0704 0.901 0.1124 0.4>
    R5 Grey:LinearColor=<rgba 0.1667 0.1667 0.1667 0.4>
    R5 Lime:LinearColor=<rgba 0.2938 0.5104 0.0771 0.25>
    R5 Orange:LinearColor=<rgba 0.8261 0.2993 0.0172 0.3>
    R5 Pink:LinearColor=<rgba 0.651 0.1763 0.4784 0.4>
    R5 Purple:LinearColor=<rgba 0.2704 0.1155 0.7917 0.3>
    R5 Saffron:LinearColor=<rgba 0.8906 0.6824 0 0.15>
    R5 SkyBlue:LinearColor=<rgba 0.0754 0.3822 0.5365 0.4>
    R5 TitaniumWhite:LinearColor=<rgba 0.5729 0.5729 0.5729 0.3>
    Use R1 Painted Overrides=true
    Use R2 Painted Overrides=true
    Use R3 Painted Overrides=true
    Use R4 Painted Overrides=true
    Use R5 Painted Overrides=true
   ) {module=VC_PaintedVariantsOverride002}
   &FN_DisableTireFX(Mask Index Attribute:NiagaraInt32=$Particles.SpawnGroup, Mask Index FL=2, Mask Index FR=3, Mask Index RL=0, Mask Index RR=1)
   &ParticleState()
   &SolveForcesAndVelocity(Clamp Velocity=false, Limit Acceleration=false)
   &VC_UpdateRibbon_v1(Enable Skeletal Position=true)
   /SetVariables(Particles.WPO:NiagaraPosition=&AddVectorToPosition(
     Position=$Particles.Position
     Vector=&TransformVector(
      Vector=&MakeVector(
       X=0
       Y=&Cosine(
        Normalized Angle=&Add_Float(
         A=&Multiply_Float(A=$Particles.NormalizedAge, B=-1)
         B=&Multiply_Float(A=$Emitter.LoopedAge, B=1)
        )
        Period=&MaskFloatBySpawnGroup(Masked Float=-1.04, Passthrough Float=1.04)
        Scale=&FloatFromCurve(
         FloatCurve=DataInterface() {propertyValues {Curve {
            defaultValue=1
            keys [{interpMode=RCIM_Cubic, tangentMode=RCTM_User, time=0, value=1}, {interpMode=RCIM_Cubic, tangentMode=RCTM_User, time=0.75, value=0}]
           }}}
         Scale Curve=8.5
        )
        Bias=0
       )
       Z=&Sine(
        Normalized Angle=&Add_Float(
         A=&Multiply_Float(A=$Particles.NormalizedAge, B=-1)
         B=&Multiply_Float(A=$Emitter.LoopedAge, B=1)
        )
        Period=&MaskFloatBySpawnGroup(Masked Float=-1.04, Passthrough Float=1.04)
        Scale=&FloatFromCurve(
         FloatCurve=DataInterface() {propertyValues {Curve {
            defaultValue=1
            keys [{interpMode=RCIM_Cubic, tangentMode=RCTM_User, time=0, value=1}, {interpMode=RCIM_Cubic, tangentMode=RCTM_User, time=0.75, value=0}]
           }}}
         Scale Curve=8.5
        )
        Bias=0
       )
      )
      Destination Space=::Simulation
      Apply Scale=true
     )
    )) {isSetParameters=true, module=SetVariables_#2}
   /SetVariables(
    Particles.R1_Color_Scaled:LinearColor=&MultiplyLinearColor(
     Linear Color=$Particles.R1 Color
     Scale Factor=&MakeVector4FromVectorAndFloat(
      "Vector (RGB)"=&VectorFromFloat(Value=&FloatFromCurve(
        FloatCurve=DataInterface() {propertyValues {Curve {keys [
            {arriveTangent=-13.0287, interpMode=RCIM_Cubic, time=0.2667, value=0.9059}
            {arriveTangent=-1.5185, tangentMode=RCTM_User, time=0.4953, value=0.8118}
            <at 0.6886 0>
            <at 0.7212 1>
            {arriveTangent=-6.7757, interpMode=RCIM_Cubic, leaveTangent=-6.7757, time=0.7822, value=1}
            <at 0.8688 0>
           ]}}}
        Scale Curve=&MakeCustomFloatFromBool(Bool=$User.bIsPainted, True Float=$User.Power_EmissivePaintedR4, False Float=2.5)
       ))
      "Float (Alpha)"=&FloatFromCurve(
       FloatCurve=DataInterface() {propertyValues {Curve {keys [
           <at -0.0383 3.4294 RCIM_Cubic>
           {arriveTangent=-0.2628, interpMode=RCIM_Cubic, leaveTangent=-0.2628, time=0.1518, value=2.4766}
           {arriveTangent=-0.2383, interpMode=RCIM_Cubic, leaveTangent=-0.2383, tangentMode=RCTM_User, time=0.3301, value=3.3326}
           {arriveTangent=-41.3513, interpMode=RCIM_Cubic, leaveTangent=-30.5159, tangentMode=RCTM_Break, time=0.5258, value=3.3174}
           {arriveTangent=-9.9908, interpMode=RCIM_Cubic, leaveTangent=-9.9908, time=0.6771, value=1.2857}
           <at 0.8608 -0.0297 RCIM_Cubic>
          ]}}}
       Scale Curve=&MakeCustomFloatFromBool(Bool=$User.bIsPainted, True Float=$User.Power_AlphaPaintedR4, False Float=0.5)
      )
     )
    )
    Particles.R3_Color_Scaled:LinearColor=&MultiplyLinearColor(
     Linear Color=$Particles.R3 Color
     Scale Factor=&MakeVector4FromVectorAndFloat(
      "Vector (RGB)"=&VectorFromFloat(Value=&FloatFromCurve(
        FloatCurve=DataInterface() {propertyValues {Curve {keys [<at 0.3572 0.8681>,<at 0.49 0.675>,<at 0.6121 0.1761>,<at 0.7769 0>]}}}
        Scale Curve=&MakeCustomFloatFromBool(Bool=$User.bIsPainted, True Float=$User.Power_EmissivePaintedR4, False Float=1.8)
       ))
      "Float (Alpha)"=&FloatFromCurve(
       FloatCurve=DataInterface() {propertyValues {Curve {keys [<at 0.2665 1>,<at 0.4676 0.5559>,<at 0.5805 0.7676>,<at 0.7 0>]}}}
       Scale Curve=&MakeCustomFloatFromBool(Bool=$User.bIsPainted, True Float=$User.Power_AlphaPaintedR4, False Float=0.5)
      )
     )
    )
    Particles.R4_Color_Scaled:LinearColor=&MultiplyLinearColor(
     Linear Color=$Particles.R4 Color
     Scale Factor=&MakeVector4FromVectorAndFloat(
      "Vector (RGB)"=&VectorFromFloat(Value=&FloatFromCurve(
        FloatCurve=DataInterface() {propertyValues {Curve {keys [<at 0.0107 1>,<at 0.0753 1>,<at 0.3572 1>,<at 0.49 1>,<at 0.5253 1>,<at 0.6121 1>,<at 0.7769 0>]}}}
        Scale Curve=&MakeCustomFloatFromBool(Bool=$User.bIsPainted, True Float=$User.Power_EmissivePaintedR4, False Float=0.6)
       ))
      "Float (Alpha)"=&FloatFromCurve(
       FloatCurve=DataInterface() {propertyValues {Curve {keys [<at 0.106 0>,<at 0.1223 1.0029>,<at 0.2665 1>,<at 0.4464 0.9028>,<at 0.5244 0.8384>,<at 0.6688 0>]}}}
       Scale Curve=&MakeCustomFloatFromBool(Bool=$User.bIsPainted, True Float=$User.Power_AlphaPaintedR4, False Float=1.2)
      )
     )
    )
   ) {isSetParameters=true, module=SetVariables_#3}
   /SetVariables(
    Particles.Ribbon_BlackColor:LinearColor=<rgba 0 0 0 1>
    Particles.WaveScaleCurve:NiagaraFloat=&FloatFromCurve(
     FloatCurve=DataInterface() {propertyValues {Curve {keys [
         {arriveTangent=39.5227, interpMode=RCIM_Cubic, leaveTangent=39.5227, tangentMode=RCTM_User, time=0.0092, value=0}
         {
          arriveTangent=0.7219
          arriveTangentWeight=0.0795
          interpMode=RCIM_Cubic
          leaveTangent=0.1806
          leaveTangentWeight=0.0696
          tangentMode=RCTM_Break
          tangentWeightMode=RCTWM_WeightedBoth
          time=0.125
          value=1.25
         }
         {arriveTangent=0.1988, interpMode=RCIM_Cubic, leaveTangent=0.1988, tangentMode=RCTM_User, time=0.45, value=0.625}
        ]}}}
     Scale Curve=10
    )
   ) {enabled=false, isSetParameters=true}
   /SetVariables(Particles.DynamicMP_R2:Vector4f=&MakeVector4(
     X=&FloatFromCurve(FloatCurve=DataInterface() {propertyValues {Curve {
         defaultValue=1
         keys [
          {tangentMode=RCTM_User, time=0, value=0.5}
          {arriveTangent=-0.0633, interpMode=RCIM_Cubic, leaveTangent=-0.0633, tangentMode=RCTM_User, time=0.65, value=0.5}
          {arriveTangent=-0.0699, interpMode=RCIM_Cubic, leaveTangent=-0.0699, tangentMode=RCTM_User, time=0.75, value=-2}
         ]
        }}})
     Y=0
     Z=0
     W=0
    )) {isSetParameters=true, module=SetVariables_#4}
   &ScaleRibbonWidth(Ribbon Width Scale=&FloatFromCurve(FloatCurve=DataInterface() {propertyValues {Curve {keys [{interpMode=RCIM_Cubic, tangentMode=RCTM_Break, time=0, value=1}, {interpMode=RCIM_Cubic, tangentMode=RCTM_Break, time=1, value=0}]}}}))
   &DynamicMaterialParameters(
    Index 0 Param 1=&FloatFromCurve(
     FloatCurve=DataInterface() {propertyValues {Curve {
        defaultValue=1
        keys [
         {
          arriveTangent=0.0405
          arriveTangentWeight=0.1835
          interpMode=RCIM_Cubic
          leaveTangent=0.0405
          leaveTangentWeight=0.1835
          tangentMode=RCTM_User
          time=0
          value=0
         }
         {arriveTangent=2.8994, interpMode=RCIM_Cubic, leaveTangent=2.8994, time=0.2297, value=0.2011}
         {arriveTangent=3.1289, interpMode=RCIM_Cubic, leaveTangent=3.1289, tangentMode=RCTM_User, time=0.2915, value=0.8451}
         {
          arriveTangent=0.0201
          arriveTangentWeight=0.1834
          interpMode=RCIM_Cubic
          leaveTangent=0.0201
          leaveTangentWeight=0.1834
          tangentMode=RCTM_User
          time=0.5
          value=1
         }
        ]
       }}}
     Scale Curve=0.8
    )
    Module.Index0Float0Write=true
    Module.Index0Float1Write=true
    Module.Index0Float2Write=true
    Module.Index0Float3Write=true
   )
   /SetVariables(Particles.PurpleRibbon:NiagaraFloat=&Multiply_Float(
     A=60
     B=&FloatFromCurve(FloatCurve=DataInterface() {propertyValues {Curve {keys [{interpMode=RCIM_Cubic, tangentMode=RCTM_Break, time=0, value=1}, {interpMode=RCIM_Cubic, tangentMode=RCTM_Break, time=1, value=0}]}}})
    )) {isSetParameters=true}
  ]
  rendererClasses [&NiagaraRibbonRendererProperties]
  renderers [
   {rendererClass=&NiagaraRibbonRendererProperties}
   {rendererClass=&NiagaraRibbonRendererProperties, rendererIndex=1}
   {rendererClass=&NiagaraRibbonRendererProperties, rendererIndex=2}
  ]
 }
 MainRibbon001 {
  emitterUpdateScript [
   &EmitterState()
   &VC_RateQualityHandler(Epic Spawn Rate=30, Low Spawn Rate=30, Use Custom Rate?=true)
   &VC_SpawnRibbon_v1(Number of Ribbons=2, Spawn Time=-1, SpawnRate=$Emitter.Spawn Rate)
  ]
  particleSpawnScript [
   &VC_PaintedVariantsOverride(Use R1 Painted Overrides=true) {enabled=false}
   &VC_InitializeRibbon_v1(
    Color=$Particles.Color
    Custom Offset=true
    Enable Skeletal Position=true
    Facing Axis:ENiagaraOrientationAxis=::YAxis
    Lifetime=1
    Local Vector Offset=&AddVectorToPosition(Position=<25 0 -20>, Vector=Linked() {linkedVariable {name=User.offset, type=NiagaraPosition}})
    Offset Amount=1
    Ribbon Width=60
   )
   &AddVelocity(Velocity=<-2200 0 0>) {module=AddVelocity001}
  ]
  particleUpdateScript [
   &VC_PaintedVariantsOverride(
    R1 Black:LinearColor=&ColorFromCurve(ColorCurve=DataInterface() {propertyValues {
       AlphaCurve {
        defaultValue=7
        keys [
         <at 0.2819 7>
         <at 0.6302 1>
         <at 0.7212 1>
         {arriveTangent=-6.7193, interpMode=RCIM_Cubic, leaveTangent=-6.7193, time=0.7822, value=1}
         <at 0.87 0>
        ]
       }
       BlueCurve {keys [<at 1 0.1094 RCIM_Cubic>]}
       GreenCurve {keys [<at 1 0.1094 RCIM_Cubic>]}
       RedCurve {keys [<at 1 0.1094 RCIM_Cubic>]}
      }})
    R1 BurntSienna:LinearColor=&ColorFromCurve(ColorCurve=DataInterface() {propertyValues {
       AlphaCurve {
        defaultValue=7
        keys [
         <at 0.2819 7>
         <at 0.6302 1>
         <at 0.7212 1>
         {arriveTangent=-6.7193, interpMode=RCIM_Cubic, leaveTangent=-6.7193, time=0.7822, value=1}
         <at 0.87 0>
        ]
       }
       BlueCurve {keys [<at 1 0.0825 RCIM_Cubic>]}
       GreenCurve {keys [<at 1 0.0972 RCIM_Cubic>]}
       RedCurve {keys [<at 1 0.1667 RCIM_Cubic>]}
      }})
    R1 Cobalt:LinearColor=&ColorFromCurve(ColorCurve=DataInterface() {propertyValues {
       AlphaCurve {
        defaultValue=7
        keys [
         <at 0.2819 7>
         <at 0.6302 1>
         <at 0.7212 1>
         {arriveTangent=-6.7193, interpMode=RCIM_Cubic, leaveTangent=-6.7193, time=0.7822, value=1}
         <at 0.87 0>
        ]
       }
       BlueCurve {keys [<at 1 0.5 RCIM_Cubic>]}
       GreenCurve {keys [<at 1 0 RCIM_Cubic>]}
       RedCurve {keys [<at 1 0 RCIM_Cubic>]}
      }})
    R1 Crimson:LinearColor=&ColorFromCurve(ColorCurve=DataInterface() {propertyValues {
       AlphaCurve {
        defaultValue=7
        keys [
         <at 0.2819 7>
         <at 0.6302 1>
         <at 0.7212 1>
         {arriveTangent=-6.7193, interpMode=RCIM_Cubic, leaveTangent=-6.7193, time=0.7822, value=1}
         <at 0.87 0>
        ]
       }
       BlueCurve {keys [<at 1 0.0131 RCIM_Cubic>]}
       GreenCurve {keys [<at 1 0 RCIM_Cubic>]}
       RedCurve {keys [<at 1 0.2348 RCIM_Cubic>]}
      }})
    R1 DefaultColor=&ColorFromCurve(ColorCurve=DataInterface() {propertyValues {
       AlphaCurve {keys [<at 0.2819 7>, <at 0.4296 1>, {interpMode=RCIM_Cubic, leaveTangent=-3.2831, time=1, value=0}]}
       BlueCurve {keys [<at 0.6627 0.15 RCIM_Cubic>]}
       GreenCurve {keys [<at 0.6627 0.0003 RCIM_Cubic>]}
       RedCurve {keys [<at 0.6627 0.0529 RCIM_Cubic>]}
      }})
    R1 ForestGreen:LinearColor=&ColorFromCurve(ColorCurve=DataInterface() {propertyValues {
       AlphaCurve {
        defaultValue=7
        keys [
         <at 0.2819 7>
         <at 0.6302 1>
         <at 0.7212 1>
         {arriveTangent=-6.7193, interpMode=RCIM_Cubic, leaveTangent=-6.7193, time=0.7822, value=1}
         <at 0.87 0>
        ]
       }
       BlueCurve {keys [<at 0.6169 0.0198 RCIM_Cubic>]}
       GreenCurve {keys [<at 0.6169 0.2031 RCIM_Cubic>]}
       RedCurve {keys [<at 0.6169 0 RCIM_Cubic>]}
      }})
    R1 Grey:LinearColor=&ColorFromCurve(ColorCurve=DataInterface() {propertyValues {
       AlphaCurve {
        defaultValue=1
        keys [
         <at 0.2819 7>
         <at 0.6302 1>
         <at 0.7212 1>
         {arriveTangent=-6.7193, interpMode=RCIM_Cubic, leaveTangent=-6.7193, time=0.7822, value=1}
         <at 0.87 0>
        ]
       }
       BlueCurve {keys [<at 1 0.1979 RCIM_Cubic>]}
       GreenCurve {keys [<at 1 0.1979 RCIM_Cubic>]}
       RedCurve {keys [<at 1 0.1979 RCIM_Cubic>]}
      }})
    R1 Lime:LinearColor=&ColorFromCurve(ColorCurve=DataInterface() {propertyValues {
       AlphaCurve {
        defaultValue=7
        keys [
         <at 0.2819 7>
         <at 0.6302 1>
         <at 0.7212 1>
         {arriveTangent=-6.7193, interpMode=RCIM_Cubic, leaveTangent=-6.7193, time=0.7822, value=1}
         <at 0.87 0>
        ]
       }
       BlueCurve {keys [<at 1 0.0865 RCIM_Cubic>]}
       GreenCurve {keys [<at 1 0.2865 RCIM_Cubic>]}
       RedCurve {keys [<at 1 0.1697 RCIM_Cubic>]}
      }})
    R1 Orange:LinearColor=&ColorFromCurve(ColorCurve=DataInterface() {propertyValues {
       AlphaCurve {
        defaultValue=7
        keys [
         <at 0.2819 7>
         <at 0.6302 1>
         <at 0.7212 1>
         {arriveTangent=-6.7193, interpMode=RCIM_Cubic, leaveTangent=-6.7193, time=0.7822, value=1}
         <at 0.87 0>
        ]
       }
       BlueCurve {keys [<at 1 0.0622 RCIM_Cubic>]}
       GreenCurve {keys [<at 1 0.1169 RCIM_Cubic>]}
       RedCurve {keys [<at 1 0.3073 RCIM_Cubic>]}
      }})
    R1 Pink:LinearColor=&ColorFromCurve(ColorCurve=DataInterface() {propertyValues {
       AlphaCurve {
        defaultValue=7
        keys [
         <at 0.2819 7>
         <at 0.6302 1>
         <at 0.7212 1>
         {arriveTangent=-6.7193, interpMode=RCIM_Cubic, leaveTangent=-6.7193, time=0.7822, value=1}
         <at 0.87 0>
        ]
       }
       BlueCurve {keys [<at 1 0.2004 RCIM_Cubic>]}
       GreenCurve {keys [<at 1 0 RCIM_Cubic>]}
       RedCurve {keys [<at 1 0.2188 RCIM_Cubic>]}
      }})
    R1 Purple:LinearColor=&ColorFromCurve(ColorCurve=DataInterface() {propertyValues {
       AlphaCurve {
        defaultValue=7
        keys [
         <at 0.2819 7>
         <at 0.6302 1>
         <at 0.7212 1>
         {arriveTangent=-6.7193, interpMode=RCIM_Cubic, leaveTangent=-6.7193, time=0.7822, value=1}
         <at 0.87 0>
        ]
       }
       BlueCurve {keys [<at 1 0.375 RCIM_Cubic>]}
       GreenCurve {keys [<at 1 0 RCIM_Cubic>]}
       RedCurve {keys [<at 1 0.1484 RCIM_Cubic>]}
      }})
    R1 Saffron:LinearColor=&ColorFromCurve(ColorCurve=DataInterface() {propertyValues {
       AlphaCurve {
        defaultValue=7
        keys [
         <at 0.2819 7>
         <at 0.6302 1>
         <at 0.7212 1>
         {arriveTangent=-6.7193, interpMode=RCIM_Cubic, leaveTangent=-6.7193, time=0.7822, value=1}
         <at 0.87 0>
        ]
       }
       BlueCurve {keys [<at 1 0.0278 RCIM_Cubic>]}
       GreenCurve {keys [<at 1 0.2596 RCIM_Cubic>]}
       RedCurve {keys [<at 1 0.2812 RCIM_Cubic>]}
      }})
    R1 SkyBlue:LinearColor=&ColorFromCurve(ColorCurve=DataInterface() {propertyValues {
       AlphaCurve {
        defaultValue=7
        keys [
         <at 0.2819 7>
         <at 0.6302 1>
         <at 0.7212 1>
         {arriveTangent=-6.7193, interpMode=RCIM_Cubic, leaveTangent=-6.7193, time=0.7822, value=1}
         <at 0.87 0>
        ]
       }
       BlueCurve {keys [<at 1 0.375 RCIM_Cubic>]}
       GreenCurve {keys [<at 1 0.1761 RCIM_Cubic>]}
       RedCurve {keys [<at 1 0.0547 RCIM_Cubic>]}
      }})
    R1 TitaniumWhite:LinearColor=&ColorFromCurve(ColorCurve=DataInterface() {propertyValues {
       AlphaCurve {
        defaultValue=2
        keys [
         <at 0.2819 7>
         <at 0.6302 1>
         <at 0.7212 1>
         {arriveTangent=-6.7193, interpMode=RCIM_Cubic, leaveTangent=-6.7193, time=0.7822, value=1}
         <at 0.87 0>
        ]
       }
       BlueCurve {keys [<at 0.6118 0.3802 RCIM_Cubic>,<at 1 0.0677>]}
       GreenCurve {keys [<at 0.6118 0.3802 RCIM_Cubic>,<at 1 0.0677>]}
       RedCurve {keys [<at 0.6118 0.3802 RCIM_Cubic>,<at 1 0.0677>]}
      }})
    R2 Black:LinearColor=&ColorFromCurve(ColorCurve=DataInterface() {propertyValues {
       AlphaCurve {keys [<at 0 7>,<at 0.2232 1>,<at 0.3261 0.05>,<at 0.5495 0.2>,<at 0.7007 0.2>,<at 0.8474 0 RCIM_Cubic>]}
       BlueCurve {keys [<at 0 0.0194 RCIM_Cubic>,<at 0.1492 0.0333>,<at 0.2681 0.1104>,<at 0.4863 0.9274>,<at 0.6088 0.0305>]}
       GreenCurve {keys [<at 0 0.0194 RCIM_Cubic>,<at 0.1492 0.0333>,<at 0.2681 0.1104>,<at 0.4863 0.9274>,<at 0.6088 0.0305>]}
       RedCurve {keys [<at 0 0.0194 RCIM_Cubic>,<at 0.1492 0.0333>,<at 0.2681 0.1104>,<at 0.4863 0.9274>,<at 0.6088 0.0305>]}
      }})
    R2 BurntSienna:LinearColor=&ColorFromCurve(ColorCurve=DataInterface() {propertyValues {
       AlphaCurve {keys [<at 0 4>,<at 0.1907 1>,<at 0.3261 0.05>,<at 0.593 0.2>,<at 0.8474 0 RCIM_Cubic>]}
       BlueCurve {keys [<at 0 0.008 RCIM_Cubic>,<at 0.2839 0.0202>,<at 0.4863 0.2907>,<at 0.5956 0.008>]}
       GreenCurve {keys [<at 0 0.0203 RCIM_Cubic>,<at 0.2839 0.0323>,<at 0.4863 0.4394>,<at 0.5956 0.0203>]}
       RedCurve {keys [<at 0 0.0648 RCIM_Cubic>,<at 0.2839 0.0762>,<at 0.4863 0.9792>,<at 0.5956 0.0648>]}
      }})
    R2 Cobalt:LinearColor=&ColorFromCurve(ColorCurve=DataInterface() {propertyValues {
       AlphaCurve {keys [<at 0 7>,<at 0.1907 1>,<at 0.3261 0.05>,<at 0.5495 0.6>,<at 0.7007 0.2>,<at 0.8474 0 RCIM_Cubic>]}
       BlueCurve {keys [<at 0 0.6458 RCIM_Cubic>,<at 0.2575 0.9985>,<at 0.4863 1>,<at 0.6115 0.8177>]}
       GreenCurve {keys [<at 0 0.0119 RCIM_Cubic>,<at 0.2575 0.0367>,<at 0.4863 0.2997>,<at 0.6115 0.0527>]}
       RedCurve {keys [<at 0 0 RCIM_Cubic>,<at 0.2575 0.0186>,<at 0.4863 0.2865>,<at 0.6115 0.0383>]}
      }})
    R2 Crimson:LinearColor=&ColorFromCurve(ColorCurve=DataInterface() {propertyValues {
       AlphaCurve {keys [<at 0 7>,<at 0.1907 1>,<at 0.3261 0.05>,<at 0.5495 0.2>,<at 0.7007 0.2>,<at 0.8474 0 RCIM_Cubic>]}
       BlueCurve {keys [<at 0 0 RCIM_Cubic>,<at 0.1492 0.0104>,<at 0.4863 0.2995>,<at 0.6458 0.0894>]}
       GreenCurve {keys [<at 0 0 RCIM_Cubic>,<at 0.1492 0.0104>,<at 0.4863 0.2995>,<at 0.6458 0.0894>]}
       RedCurve {keys [<at 0 0.4323 RCIM_Cubic>,<at 0.1492 0.9992>,<at 0.4863 0.9274>,<at 0.6458 0.993>]}
      }})
    R2 DefaultColor:LinearColor=&ColorFromCurve(ColorCurve=DataInterface() {propertyValues {
       AlphaCurve {keys [<at 0 1>,<at 0.0914 0.2857>,<at 0.308 0.05>,<at 0.5592 0.2>,<at 0.8632 0>]}
       BlueCurve {keys [<at 0.0009 0>,<at 0.2668 -0.2829>,<at 0.5114 0.8849>,<at 0.619 1>,<at 0.682 0.383>]}
       GreenCurve {keys [<at 0.0009 0.0767>,<at 0.2742 -0.1994>,<at 0.5114 0.8849>,<at 0.619 1>,<at 0.682 0.2585>]}
       RedCurve {keys [<at 0.0009 3.2194>,<at 0.0016 3.2194>,<at 0.4479 0.8849>,<at 0.5114 0.8849>,<at 0.619 1>]}
      }})
    R2 ForestGreen:LinearColor=&ColorFromCurve(ColorCurve=DataInterface() {propertyValues {
       AlphaCurve {keys [<at 0 6>,<at 0.1907 1>,<at 0.3261 0.05>,<at 0.5495 0.2>,<at 0.7007 0.2>,<at 0.8474 0 RCIM_Cubic>]}
       BlueCurve {keys [<at 0 0.0172 RCIM_Cubic>,<at 0.2179 0.0322>,<at 0.4863 0.4656>,<at 0.6537 0.1027>]}
       GreenCurve {keys [<at 0 0.3438 RCIM_Cubic>,<at 0.2179 0.245>,<at 0.4863 0.875>,<at 0.6537 0.2987>]}
       RedCurve {keys [<at 0 0 RCIM_Cubic>,<at 0.2179 0.021>,<at 0.4863 0.4466>,<at 0.6537 0.0924>]}
      }})
    R2 Grey:LinearColor=&ColorFromCurve(ColorCurve=DataInterface() {propertyValues {
       AlphaCurve {keys [<at 0 4>,<at 0.1907 1>,<at 0.3261 0.05>,<at 0.5495 0.2>,<at 0.7007 0.2>,<at 0.8474 0 RCIM_Cubic>]}
       BlueCurve {keys [<at 0 0.0692 RCIM_Cubic>,<at 0.2285 0.1336>,<at 0.4863 0.9274>,<at 0.6405 0.132>]}
       GreenCurve {keys [<at 0 0.0781 RCIM_Cubic>,<at 0.2285 0.1336>,<at 0.4863 0.9274>,<at 0.6405 0.132>]}
       RedCurve {keys [<at 0 0.0767 RCIM_Cubic>,<at 0.2285 0.1406>,<at 0.4863 0.9274>,<at 0.6405 0.1354>]}
      }})
    R2 Lime:LinearColor=&ColorFromCurve(ColorCurve=DataInterface() {propertyValues {
       AlphaCurve {keys [<at 0 6>,<at 0.1907 1>,<at 0.3261 0.05>,<at 0.5495 0.2>,<at 0.8122 0 RCIM_Cubic>]}
       BlueCurve {keys [<at 0 0 RCIM_Cubic>,<at 0.2179 0.0459>,<at 0.4863 0.9274>,<at 0.6299 0.0724>]}
       GreenCurve {keys [<at 0 0.6875 RCIM_Cubic>,<at 0.2179 0.4635>,<at 0.4863 0.9274>,<at 0.6299 0.7708>]}
       RedCurve {keys [<at 0 0.3447 RCIM_Cubic>,<at 0.2179 0.2553>,<at 0.4863 0.9274>,<at 0.6299 0.4226>]}
      }})
    R2 Orange:LinearColor=&ColorFromCurve(ColorCurve=DataInterface() {propertyValues {
       AlphaCurve {keys [<at 0 7>,<at 0.1907 1>,<at 0.3261 0.05>,<at 0.5495 0.2>,<at 0.7007 0.2>,<at 0.8474 0 RCIM_Cubic>]}
       BlueCurve {keys [<at 0 0 RCIM_Cubic>,<at 0.2364 0.0598>,<at 0.4863 0.4263>,<at 0.6217 0.1191>]}
       GreenCurve {keys [<at 0 0.2834 RCIM_Cubic>,<at 0.2364 0.2792>,<at 0.4863 0.5719>,<at 0.6217 0.3121>]}
       RedCurve {keys [<at 0 0.8125 RCIM_Cubic>,<at 0.2364 0.6889>,<at 0.4863 0.8438>,<at 0.6217 0.6724>]}
      }})
    R2 Pink:LinearColor=&ColorFromCurve(ColorCurve=DataInterface() {propertyValues {
       AlphaCurve {keys [<at 0 7>,<at 0.1907 1>,<at 0.3261 0.05>,<at 0.5495 0.2>,<at 0.7007 0.2>,<at 0.8474 0 RCIM_Cubic>]}
       BlueCurve {keys [<at 0 0.3419 RCIM_Cubic>,<at 0.2707 0.3564>,<at 0.4863 0.7216>,<at 0.6141 0.5634>]}
       GreenCurve {keys [<at 0 0.0612 RCIM_Cubic>,<at 0.2707 0.1489>,<at 0.4863 0.5789>,<at 0.6141 0.2536>]}
       RedCurve {keys [<at 0 0.5029 RCIM_Cubic>,<at 0.2707 0.5614>,<at 0.4863 0.855>,<at 0.6141 0.8542>]}
      }})
    R2 Purple:LinearColor=&ColorFromCurve(ColorCurve=DataInterface() {propertyValues {
       AlphaCurve {keys [<at 0 7>,<at 0.1907 1>,<at 0.3261 0.05>,<at 0.5495 0.2>,<at 0.7007 0.2>,<at 0.8474 0 RCIM_Cubic>]}
       BlueCurve {keys [<at 0 0.8281 RCIM_Cubic>,<at 0.2919 0.6836>,<at 0.4863 0.8957>,<at 0.6141 0.9565>]}
       GreenCurve {keys [<at 0 0 RCIM_Cubic>,<at 0.2919 0.0563>,<at 0.4863 0.2791>,<at 0.6141 0.1097>]}
       RedCurve {keys [<at 0 0.1924 RCIM_Cubic>,<at 0.2919 0.3349>,<at 0.4863 0.5023>,<at 0.6141 0.3436>]}
      }})
    R2 Saffron:LinearColor=&ColorFromCurve(ColorCurve=DataInterface() {propertyValues {
       AlphaCurve {keys [<at 0 4>,<at 0.1907 1>,<at 0.3261 0.05>,<at 0.5495 0.2>,<at 0.8175 0 RCIM_Cubic>]}
       BlueCurve {keys [<at 0 0 RCIM_Cubic>,<at 0.2697 0.0046>,<at 0.4863 0.3188>,<at 0.6748 0.0798>]}
       GreenCurve {keys [<at 0 0.6105 RCIM_Cubic>,<at 0.2697 0.6881>,<at 0.4863 0.9274>,<at 0.6748 0.6052>]}
       RedCurve {keys [<at 0 0.7969 RCIM_Cubic>,<at 0.2697 0.8967>,<at 0.4863 0.8744>,<at 0.6748 0.7656>]}
      }})
    R2 SkyBlue:LinearColor=&ColorFromCurve(ColorCurve=DataInterface() {propertyValues {
       AlphaCurve {keys [<at 0 7>,<at 0.1907 1>,<at 0.3261 0.05>,<at 0.5495 0.2>,<at 0.8096 0 RCIM_Cubic>]}
       BlueCurve {keys [<at 0 0.5333 RCIM_Cubic>,<at 0.2496 0.5525>,<at 0.4863 0.9323>,<at 0.6141 0.9323>]}
       GreenCurve {keys [<at 0 0.3564 RCIM_Cubic>,<at 0.2496 0.3792>,<at 0.4863 0.8083>,<at 0.6141 0.6988>]}
       RedCurve {keys [<at 0 0 RCIM_Cubic>,<at 0.2496 0.0875>,<at 0.4863 0.5584>,<at 0.6141 0.2282>]}
      }})
    R2 TitaniumWhite:LinearColor=&ColorFromCurve(ColorCurve=DataInterface() {propertyValues {
       AlphaCurve {keys [<at 0 4>,<at 0.1907 1>,<at 0.3261 0.05>,<at 0.5495 0.2>,<at 0.7007 0.2>,<at 0.8175 0 RCIM_Cubic>]}
       BlueCurve {keys [<at 0 0.1846 RCIM_Cubic>,<at 0.2285 0.1336>,<at 0.4863 0.9274>,<at 0.6405 0.3248>]}
       GreenCurve {keys [<at 0 0.2083 RCIM_Cubic>,<at 0.2285 0.1336>,<at 0.4863 0.9274>,<at 0.6405 0.3248>]}
       RedCurve {keys [<at 0 0.2046 RCIM_Cubic>,<at 0.2285 0.1406>,<at 0.4863 0.9274>,<at 0.6405 0.3333>]}
      }})
    R3 Black:LinearColor=&ColorFromCurve(ColorCurve=DataInterface() {propertyValues {
       AlphaCurve {keys [{tangentMode=RCTM_User, time=0.8086, value=3}, {tangentMode=RCTM_User, time=0.8837, value=0}, {tangentMode=RCTM_User, time=1, value=0}]}
       BlueCurve {keys [<at 0.6966 0.25>]}
       GreenCurve {keys [<at 0.6966 0.25>]}
       RedCurve {keys [<at 0.6966 0.25>]}
      }})
    R3 BurntSienna:LinearColor=&ColorFromCurve(ColorCurve=DataInterface() {propertyValues {
       AlphaCurve {
        defaultValue=1
        keys [{tangentMode=RCTM_User, time=0.8086, value=3}, {tangentMode=RCTM_User, time=0.8837, value=0}, {tangentMode=RCTM_User, time=1, value=0}]
       }
       BlueCurve {keys [<at 0.6966 0.147>]}
       GreenCurve {keys [<at 0.6966 0.2197>]}
       RedCurve {keys [<at 0.6966 0.4479>]}
      }})
    R3 Cobalt:LinearColor=&ColorFromCurve(ColorCurve=DataInterface() {propertyValues {
       AlphaCurve {
        defaultValue=3
        keys [{tangentMode=RCTM_User, time=0.8086, value=1}, {tangentMode=RCTM_User, time=0.8837, value=0}, {tangentMode=RCTM_User, time=1, value=0}]
       }
       BlueCurve {keys [<at 0.6966 1.3>]}
       GreenCurve {keys [<at 0.6966 0.1083>]}
       RedCurve {keys [<at 0.6966 0>]}
      }})
    R3 Crimson:LinearColor=&ColorFromCurve(ColorCurve=DataInterface() {propertyValues {
       AlphaCurve {
        defaultValue=3
        keys [{tangentMode=RCTM_User, time=0.8086, value=3}, {tangentMode=RCTM_User, time=0.8837, value=0}, {tangentMode=RCTM_User, time=1, value=0}]
       }
       BlueCurve {keys [<at 0.6966 0>]}
       GreenCurve {keys [<at 0.6966 0>]}
       RedCurve {keys [<at 0.6966 1>]}
      }})
    R3 DefaultColor:LinearColor=&ColorFromCurve(ColorCurve=DataInterface() {propertyValues {
       AlphaCurve {keys [<at 0.8086 3>,<at 0.8837 0>,<at 1 0>]}
       BlueCurve {keys [<at 0.6966 1>]}
       GreenCurve {keys [<at 0.6966 0.2>]}
       RedCurve {keys [<at 0.6966 0>]}
      }})
    R3 ForestGreen:LinearColor=&ColorFromCurve(ColorCurve=DataInterface() {propertyValues {
       AlphaCurve {
        defaultValue=3
        keys [{tangentMode=RCTM_User, time=0.8086, value=3}, {tangentMode=RCTM_User, time=0.8837, value=0}, {tangentMode=RCTM_User, time=1, value=0}]
       }
       BlueCurve {keys [<at 0.6966 0.2593>]}
       GreenCurve {keys [<at 0.6966 0.6562>]}
       RedCurve {keys [<at 0.6966 0.1436>]}
      }})
    R3 Grey:LinearColor=&ColorFromCurve(ColorCurve=DataInterface() {propertyValues {
       AlphaCurve {
        defaultValue=3
        keys [{tangentMode=RCTM_User, time=0.8086, value=3}, {tangentMode=RCTM_User, time=0.8837, value=0}, {tangentMode=RCTM_User, time=1, value=0}]
       }
       BlueCurve {keys [<at 0.6966 0.3646>]}
       GreenCurve {keys [<at 0.6966 0.3646>]}
       RedCurve {keys [<at 0.6966 0.3646>]}
      }})
    R3 Lime:LinearColor=&ColorFromCurve(ColorCurve=DataInterface() {propertyValues {
       AlphaCurve {
        defaultValue=3
        keys [{tangentMode=RCTM_User, time=0.8086, value=3}, {tangentMode=RCTM_User, time=0.8837, value=0}, {tangentMode=RCTM_User, time=1, value=0}]
       }
       BlueCurve {keys [<at 0.6966 0>]}
       GreenCurve {keys [<at 0.6966 1>]}
       RedCurve {keys [<at 0.6966 0.5901>]}
      }})
    R3 Orange:LinearColor=&ColorFromCurve(ColorCurve=DataInterface() {propertyValues {
       AlphaCurve {
        defaultValue=3
        keys [{tangentMode=RCTM_User, time=0.8086, value=3}, {tangentMode=RCTM_User, time=0.8837, value=0}, {tangentMode=RCTM_User, time=1, value=0}]
       }
       BlueCurve {keys [<at 0.6966 0.1367>]}
       GreenCurve {keys [<at 0.6966 0.2734>]}
       RedCurve {keys [<at 0.6966 0.7292>]}
      }})
    R3 Pink:LinearColor=&ColorFromCurve(ColorCurve=DataInterface() {propertyValues {
       AlphaCurve {
        defaultValue=3
        keys [{tangentMode=RCTM_User, time=0.8086, value=3}, {tangentMode=RCTM_User, time=0.8837, value=0}, {tangentMode=RCTM_User, time=1, value=0}]
       }
       BlueCurve {keys [<at 0.6966 0.6855>]}
       GreenCurve {keys [<at 0.6966 0.1214>]}
       RedCurve {keys [<at 0.6966 0.6462>]}
      }})
    R3 Purple:LinearColor=&ColorFromCurve(ColorCurve=DataInterface() {propertyValues {
       AlphaCurve {
        defaultValue=3
        keys [{tangentMode=RCTM_User, time=0.8086, value=3}, {tangentMode=RCTM_User, time=0.8837, value=0}, {tangentMode=RCTM_User, time=1, value=0}]
       }
       BlueCurve {keys [<at 0.6966 0.8906>]}
       GreenCurve {keys [<at 0.6966 0>]}
       RedCurve {keys [<at 0.6966 0.2841>]}
      }})
    R3 Saffron:LinearColor=&ColorFromCurve(ColorCurve=DataInterface() {propertyValues {
       AlphaCurve {
        defaultValue=3
        keys [{tangentMode=RCTM_User, time=0.8086, value=3}, {tangentMode=RCTM_User, time=0.8837, value=0}, {tangentMode=RCTM_User, time=1, value=0}]
       }
       BlueCurve {keys [<at 0.6966 0.199>]}
       GreenCurve {keys [<at 0.6966 1.2571>]}
       RedCurve {keys [<at 0.6966 1.3646>]}
      }})
    R3 SkyBlue:LinearColor=&ColorFromCurve(ColorCurve=DataInterface() {propertyValues {
       AlphaCurve {
        defaultValue=3
        keys [{tangentMode=RCTM_User, time=0.8086, value=3}, {tangentMode=RCTM_User, time=0.8837, value=0}, {tangentMode=RCTM_User, time=1, value=0}]
       }
       BlueCurve {keys [<at 0.6966 0.9375>]}
       GreenCurve {keys [<at 0.6966 0.4441>]}
       RedCurve {keys [<at 0.6966 0>]}
      }})
    R3 TitaniumWhite:LinearColor=&ColorFromCurve(ColorCurve=DataInterface() {propertyValues {
       AlphaCurve {keys [<at 0.4952 1>,<at 0.9629 0>]}
       BlueCurve {keys [<at 0.6966 2.5>]}
       GreenCurve {keys [<at 0.6966 2.5>]}
       RedCurve {keys [<at 0.6966 2.5>]}
      }})
    R4 Black:LinearColor=&ColorFromCurve(ColorCurve=DataInterface() {propertyValues {
       AlphaCurve {keys [<at 0.0115 0.9906>,<at 0.3154 1>,<at 0.5062 0>]}
       BlueCurve {keys [<at 0.3014 0>]}
       GreenCurve {keys [<at 0.3014 0>]}
       RedCurve {keys [<at 0.3014 0>]}
      }})
    R4 BurntSienna:LinearColor=&ColorFromCurve(ColorCurve=DataInterface() {propertyValues {
       AlphaCurve {keys [<at 0.0115 0.9906>,<at 0.3154 1>,<at 0.5062 0>]}
       BlueCurve {keys [<at 0.3014 0>]}
       GreenCurve {keys [<at 0.3014 0>]}
       RedCurve {keys [<at 0.3014 0>]}
      }})
    R4 Cobalt:LinearColor=&ColorFromCurve(ColorCurve=DataInterface() {propertyValues {
       AlphaCurve {keys [<at 0.0115 0.9906>,<at 0.3154 1>,<at 0.5062 0>]}
       BlueCurve {keys [<at 0.3014 0>]}
       GreenCurve {keys [<at 0.3014 0>]}
       RedCurve {keys [<at 0.3014 0>]}
      }})
    R4 Crimson:LinearColor=&ColorFromCurve(ColorCurve=DataInterface() {propertyValues {
       AlphaCurve {keys [<at 0.0115 0.9906>,<at 0.3154 1>,<at 0.5062 0>]}
       BlueCurve {keys [<at 0.0688 0.069>,<at 0.2065 0.0676>,<at 0.3014 0>]}
       GreenCurve {keys [<at 0.0688 0.069>,<at 0.2065 0.0676>,<at 0.3014 0>]}
       RedCurve {keys [<at 0.0688 0.25>,<at 0.2065 0.2448>,<at 0.3014 0>]}
      }})
    R4 DefaultColor:LinearColor=&ColorFromCurve(ColorCurve=DataInterface() {propertyValues {
       AlphaCurve {keys [<at 0.0115 0.9906>,<at 0.3154 1>,<at 0.5062 0>]}
       BlueCurve {keys [<at 0.1589 0.1979>,<at 0.3466 0>,<at 0.3491 0>]}
       GreenCurve {keys [<at 0.1589 0.0166>,<at 0.3466 0>,<at 0.3491 0>]}
       RedCurve {keys [<at 0.1589 0.105>,<at 0.3466 0>,<at 0.3491 0>]}
      }})
    R4 ForestGreen:LinearColor=&ColorFromCurve(ColorCurve=DataInterface() {propertyValues {
       AlphaCurve {keys [<at 0.0115 0.9906>,<at 0.3154 1>,<at 0.5062 0>]}
       BlueCurve {keys [<at 0.229 0.0107>,<at 0.4942 0>]}
       GreenCurve {keys [<at 0.229 0.2031>,<at 0.4942 0>]}
       RedCurve {keys [<at 0.229 0>,<at 0.4942 0>]}
      }})
    R4 Grey:LinearColor=&ColorFromCurve(ColorCurve=DataInterface() {propertyValues {
       AlphaCurve {keys [<at 0.0115 0.9906>,<at 0.3154 1>,<at 0.5062 0>]}
       BlueCurve {keys [<at 0.1914 0.1875>,<at 0.3014 0>]}
       GreenCurve {keys [<at 0.1914 0.1875>,<at 0.3014 0>]}
       RedCurve {keys [<at 0.1914 0.1875>,<at 0.3014 0>]}
      }})
    R4 Lime:LinearColor=&ColorFromCurve(ColorCurve=DataInterface() {propertyValues {
       AlphaCurve {keys [<at 0.0115 0.9906>,<at 0.3154 1>,<at 0.5062 0>]}
       BlueCurve {keys [<at 0.3014 0.0328>,<at 0.4842 0.0068>]}
       GreenCurve {keys [<at 0.3014 0.151>,<at 0.4842 0.0312>]}
       RedCurve {keys [<at 0.3014 0.0925>,<at 0.4842 0.0191>]}
      }})
    R4 Orange:LinearColor=&ColorFromCurve(ColorCurve=DataInterface() {propertyValues {
       AlphaCurve {keys [<at 0.0115 0.9906>,<at 0.3154 1>,<at 0.5062 0>]}
       BlueCurve {keys [<at 0.2565 0.0204>,<at 0.4842 0>]}
       GreenCurve {keys [<at 0.2565 0.0617>,<at 0.4842 0>]}
       RedCurve {keys [<at 0.2565 0.276>,<at 0.4842 0>]}
      }})
    R4 Pink:LinearColor=&ColorFromCurve(ColorCurve=DataInterface() {propertyValues {
       AlphaCurve {keys [<at 0.0115 0.9906>,<at 0.3154 1>,<at 0.5062 0>]}
       BlueCurve {keys [<at 0.279 0.6169>,<at 0.4817 0>]}
       GreenCurve {keys [<at 0.279 0.087>,<at 0.4817 0>]}
       RedCurve {keys [<at 0.279 0.6198>,<at 0.4817 0>]}
      }})
    R4 Purple:LinearColor=&ColorFromCurve(ColorCurve=DataInterface() {propertyValues {
       AlphaCurve {keys [<at 0.0115 0.9906>,<at 0.3154 1>,<at 0.5062 0>]}
       BlueCurve {keys [<at 0.3014 0>]}
       GreenCurve {keys [<at 0.3014 0>]}
       RedCurve {keys [<at 0.3014 0>]}
      }})
    R4 Saffron:LinearColor=&ColorFromCurve(ColorCurve=DataInterface() {propertyValues {
       AlphaCurve {keys [<at 0.0115 0.9906>,<at 0.3154 1>,<at 0.5062 0>]}
       BlueCurve {keys [<at 0.3014 0>]}
       GreenCurve {keys [<at 0.3014 0>]}
       RedCurve {keys [<at 0.3014 0>]}
      }})
    R4 SkyBlue:LinearColor=&ColorFromCurve(ColorCurve=DataInterface() {propertyValues {
       AlphaCurve {keys [<at 0.0115 0.9906>,<at 0.3154 1>,<at 0.5062 0>]}
       BlueCurve {keys [<at 0.3014 0>]}
       GreenCurve {keys [<at 0.3014 0>]}
       RedCurve {keys [<at 0.3014 0>]}
      }})
    R4 TitaniumWhite:LinearColor=&ColorFromCurve(ColorCurve=DataInterface() {propertyValues {
       AlphaCurve {keys [<at 0.0115 0.9906>,<at 0.1682 1>,<at 0.3697 0>]}
       BlueCurve {keys [<at 0.3509 0.2708>,<at 0.4792 0.0521>]}
       GreenCurve {keys [<at 0.3509 0.2708>,<at 0.4792 0.0521>]}
       RedCurve {keys [<at 0.3509 0.2708>,<at 0.4792 0.0521>]}
      }})
    R5 Black:LinearColor=<rgba 0.0469 0.0469 0.0469 0.6>
    R5 BurntSienna:LinearColor=<rgba 0.1042 0.0322 0.0128 0.7>
    R5 Cobalt:LinearColor=<rgba 0.0312 0.0401 1 0.6>
    R5 Crimson:LinearColor=<rgba 0.4635 0.0266 0.0384 0.6>
    R5 DefaultColor:LinearColor=<rgba 1 1 1 0.5>
    R5 ForestGreen:LinearColor=<rgba 0.0704 0.901 0.1124 0.4>
    R5 Grey:LinearColor=<rgba 0.1667 0.1667 0.1667 0.4>
    R5 Lime:LinearColor=<rgba 0.2938 0.5104 0.0771 0.25>
    R5 Orange:LinearColor=<rgba 0.8261 0.2993 0.0172 0.3>
    R5 Pink:LinearColor=<rgba 0.651 0.1763 0.4784 0.4>
    R5 Purple:LinearColor=<rgba 0.2704 0.1155 0.7917 0.3>
    R5 Saffron:LinearColor=<rgba 0.8906 0.6824 0 0.15>
    R5 SkyBlue:LinearColor=<rgba 0.0754 0.3822 0.5365 0.4>
    R5 TitaniumWhite:LinearColor=<rgba 0.5729 0.5729 0.5729 0.3>
    Use R1 Painted Overrides=true
    Use R2 Painted Overrides=true
    Use R3 Painted Overrides=true
    Use R4 Painted Overrides=true
    Use R5 Painted Overrides=true
   ) {module=VC_PaintedVariantsOverride002}
   &FN_DisableTireFX(Mask Index Attribute:NiagaraInt32=$Particles.SpawnGroup, Mask Index FL=2, Mask Index FR=3, Mask Index RL=0, Mask Index RR=1)
   &ParticleState()
   &SolveForcesAndVelocity(Clamp Velocity=false, Limit Acceleration=false)
   &VC_UpdateRibbon_v1(Enable Skeletal Position=true)
   /SetVariables(Particles.WPO:NiagaraPosition=&AddVectorToPosition(
     Position=$Particles.Position
     Vector=&TransformVector(
      Vector=&MakeVector(
       X=0
       Y=&Cosine(
        Normalized Angle=&Add_Float(
         A=&Multiply_Float(A=$Particles.NormalizedAge, B=-1)
         B=&Multiply_Float(A=$Emitter.LoopedAge, B=1)
        )
        Period=&MaskFloatBySpawnGroup(Masked Float=-1.04, Passthrough Float=1.04)
        Scale=&FloatFromCurve(
         FloatCurve=DataInterface() {propertyValues {Curve {
            defaultValue=1
            keys [{interpMode=RCIM_Cubic, tangentMode=RCTM_User, time=0, value=1}, {interpMode=RCIM_Cubic, tangentMode=RCTM_User, time=0.75, value=0}]
           }}}
         Scale Curve=-8.5
        )
        Bias=0
       )
       Z=&Sine(
        Normalized Angle=&Add_Float(
         A=&Multiply_Float(A=$Particles.NormalizedAge, B=-1)
         B=&Multiply_Float(A=$Emitter.LoopedAge, B=1)
        )
        Period=&MaskFloatBySpawnGroup(Masked Float=-1.04, Passthrough Float=1.04)
        Scale=&FloatFromCurve(
         FloatCurve=DataInterface() {propertyValues {Curve {
            defaultValue=1
            keys [{interpMode=RCIM_Cubic, tangentMode=RCTM_User, time=0, value=1}, {interpMode=RCIM_Cubic, tangentMode=RCTM_User, time=0.75, value=0}]
           }}}
         Scale Curve=-8.5
        )
        Bias=0
       )
      )
      Destination Space=::Simulation
      Apply Scale=true
     )
    )) {isSetParameters=true, module=SetVariables_#2}
   /SetVariables(
    Particles.R1_Color_Scaled:LinearColor=&MultiplyLinearColor(
     Linear Color=$Particles.R1 Color
     Scale Factor=&MakeVector4FromVectorAndFloat(
      "Vector (RGB)"=&VectorFromFloat(Value=&FloatFromCurve(
        FloatCurve=DataInterface() {propertyValues {Curve {keys [
            {arriveTangent=-13.0287, interpMode=RCIM_Cubic, time=0.2667, value=0.9059}
            {arriveTangent=-1.5185, tangentMode=RCTM_User, time=0.4953, value=0.8118}
            <at 0.6886 0>
            <at 0.7212 1>
            {arriveTangent=-6.7757, interpMode=RCIM_Cubic, leaveTangent=-6.7757, time=0.7822, value=1}
            <at 0.8688 0>
           ]}}}
        Scale Curve=&MakeCustomFloatFromBool(Bool=$User.bIsPainted, True Float=$User.Power_EmissivePaintedR4, False Float=2.5)
       ))
      "Float (Alpha)"=&FloatFromCurve(
       FloatCurve=DataInterface() {propertyValues {Curve {keys [
           <at -0.0383 3.4294 RCIM_Cubic>
           {arriveTangent=-3.3027, interpMode=RCIM_Cubic, leaveTangent=-3.3027, time=0.1518, value=2.4766}
           {interpMode=RCIM_Cubic, leaveTangent=-5.0104, time=1, value=0}
          ]}}}
       Scale Curve=&MakeCustomFloatFromBool(Bool=$User.bIsPainted, True Float=$User.Power_AlphaPaintedR4, False Float=1)
      )
     )
    )
    Particles.R3_Color_Scaled:LinearColor=&MultiplyLinearColor(
     Linear Color=$Particles.R3 Color
     Scale Factor=&MakeVector4FromVectorAndFloat(
      "Vector (RGB)"=&VectorFromFloat(Value=&FloatFromCurve(
        FloatCurve=DataInterface() {propertyValues {Curve {keys [<at 0.3572 0.8681>,<at 0.49 0.675>,<at 0.6121 0.1761>,<at 0.7769 0>]}}}
        Scale Curve=&MakeCustomFloatFromBool(Bool=$User.bIsPainted, True Float=$User.Power_EmissivePaintedR4, False Float=1.8)
       ))
      "Float (Alpha)"=&FloatFromCurve(
       FloatCurve=DataInterface() {propertyValues {Curve {keys [
           <at -0.0383 3.4294 RCIM_Cubic>
           {arriveTangent=-3.3027, interpMode=RCIM_Cubic, leaveTangent=-3.3027, time=0.1518, value=2.4766}
           {interpMode=RCIM_Cubic, leaveTangent=-5.0104, time=1, value=0}
          ]}}}
       Scale Curve=&MakeCustomFloatFromBool(Bool=$User.bIsPainted, True Float=$User.Power_AlphaPaintedR4, False Float=0.5)
      )
     )
    )
    Particles.R4_Color_Scaled:LinearColor=&MultiplyLinearColor(
     Linear Color=$Particles.R4 Color
     Scale Factor=&MakeVector4FromVectorAndFloat(
      "Vector (RGB)"=&VectorFromFloat(Value=&FloatFromCurve(
        FloatCurve=DataInterface() {propertyValues {Curve {keys [<at 0.0107 1>,<at 0.0753 1>,<at 0.3572 1>,<at 0.49 1>,<at 0.5253 1>,<at 0.6121 1>,<at 0.7769 0>]}}}
        Scale Curve=&MakeCustomFloatFromBool(Bool=$User.bIsPainted, True Float=$User.Power_EmissivePaintedR4, False Float=0.6)
       ))
      "Float (Alpha)"=&FloatFromCurve(
       FloatCurve=DataInterface() {propertyValues {Curve {keys [<at 0.106 0>,<at 0.1223 1.0029>,<at 0.2665 1>,<at 0.4464 0.9028>,<at 0.5244 0.8384>,<at 0.6688 0>]}}}
       Scale Curve=&MakeCustomFloatFromBool(Bool=$User.bIsPainted, True Float=$User.Power_AlphaPaintedR4, False Float=1.2)
      )
     )
    )
   ) {isSetParameters=true, module=SetVariables_#3}
   /SetVariables(Particles.DynamicMP_R2:Vector4f=&MakeVector4(
     X=&FloatFromCurve(FloatCurve=DataInterface() {propertyValues {Curve {
         defaultValue=1
         keys [
          {tangentMode=RCTM_User, time=0, value=0.5}
          {arriveTangent=-0.0633, interpMode=RCIM_Cubic, leaveTangent=-0.0633, tangentMode=RCTM_User, time=0.65, value=0.5}
          {arriveTangent=-0.0699, interpMode=RCIM_Cubic, leaveTangent=-0.0699, tangentMode=RCTM_User, time=0.75, value=-2}
         ]
        }}})
     Y=0
     Z=0
     W=0
    )) {isSetParameters=true, module=SetVariables_#4}
   &ScaleRibbonWidth(Ribbon Width Scale=&FloatFromCurve(FloatCurve=DataInterface() {propertyValues {Curve {keys [{interpMode=RCIM_Cubic, tangentMode=RCTM_Break, time=0, value=1}, {interpMode=RCIM_Cubic, tangentMode=RCTM_Break, time=1, value=0}]}}}))
   &DynamicMaterialParameters(
    Index 0 Param 1=&FloatFromCurve(
     FloatCurve=DataInterface() {propertyValues {Curve {
        defaultValue=1
        keys [
         {
          arriveTangent=0.0405
          arriveTangentWeight=0.1835
          interpMode=RCIM_Cubic
          leaveTangent=0.0405
          leaveTangentWeight=0.1835
          tangentMode=RCTM_User
          time=0
          value=0
         }
         {arriveTangent=2.8994, interpMode=RCIM_Cubic, leaveTangent=2.8994, time=0.2297, value=0.2011}
         {arriveTangent=3.1289, interpMode=RCIM_Cubic, leaveTangent=3.1289, tangentMode=RCTM_User, time=0.2915, value=0.8451}
         {
          arriveTangent=0.0201
          arriveTangentWeight=0.1834
          interpMode=RCIM_Cubic
          leaveTangent=0.0201
          leaveTangentWeight=0.1834
          tangentMode=RCTM_User
          time=0.5
          value=1
         }
        ]
       }}}
     Scale Curve=0.8
    )
    Module.Index0Float0Write=true
    Module.Index0Float1Write=true
    Module.Index0Float2Write=true
    Module.Index0Float3Write=true
   )
   /SetVariables(Particles.PurpleRibbon:NiagaraFloat=&Multiply_Float(
     A=60
     B=&FloatFromCurve(FloatCurve=DataInterface() {propertyValues {Curve {keys [{interpMode=RCIM_Cubic, tangentMode=RCTM_Break, time=0, value=1}, {interpMode=RCIM_Cubic, tangentMode=RCTM_Break, time=1, value=0}]}}})
    )) {isSetParameters=true}
  ]
  rendererClasses [&NiagaraRibbonRendererProperties]
  renderers [
   {rendererClass=&NiagaraRibbonRendererProperties}
   {rendererClass=&NiagaraRibbonRendererProperties, rendererIndex=1}
   {rendererClass=&NiagaraRibbonRendererProperties, rendererIndex=2}
  ]
 }
 VerticalFlames {
  emitterUpdateScript [
   &EmitterState(
    Inactive Response=::NewEnumerator1
    Life Cycle Mode=::NewEnumerator1
    Loop Behavior=::NewEnumerator1
    Loop Duration Mode=::NewEnumerator1
    UseLoopDelay=false
   )
   &SpawnBurst_Instantaneous(Spawn Count=2, Use Spawn Probability=false, UseLoopCountLimit=false)
  ]
  particleSpawnScript [
   &VC_PaintedVariantsOverride() {enabled=false}
   &InitializeParticle(
    Color Mode=::NewEnumerator0
    Mesh Scale Mode=::NewEnumerator4
    Sprite Rotation Mode=::NewEnumerator2
    UsePositionOffset=false
    Write Material Random=true
   )
   &PartitionParticles()
   &SkeletalMeshLocation(
    Bone / Socket Index=&Add_Integer(A=$Particles.PartitionParticles.Partition, B=4)
    Bone / Socket Sampling Mode=::NewEnumerator5
    Interpolate Mesh Sampling=false
    MaskSpawnGroup=false
    Mesh Sampling Type=::NewEnumerator6
    Skeletal Mesh=$System.SkeletalMesh
   )
   /SetVariables(Particles.Size_Mesh:Vector3f=&Multiply_Vector(A=<0.9 0.9 0.9>, B=$Particles.Scale)) {isSetParameters=true}
  ]
  particleUpdateScript [
   &VC_PaintedVariantsOverride(
    R1 Black:LinearColor=&ColorFromCurve(ColorCurve=DataInterface() {propertyValues {
       AlphaCurve {keys [<at 0.2819 7>]}
       BlueCurve {keys [<at 1 0.1094 RCIM_Cubic>]}
       GreenCurve {keys [<at 1 0.1094 RCIM_Cubic>]}
       RedCurve {keys [<at 1 0.1094 RCIM_Cubic>]}
      }})
    R1 BurntSienna:LinearColor=&ColorFromCurve(ColorCurve=DataInterface() {propertyValues {
       AlphaCurve {keys [<at 0.2819 7>]}
       BlueCurve {keys [<at 1 0.0686 RCIM_Cubic>]}
       GreenCurve {keys [<at 1 0.0858 RCIM_Cubic>]}
       RedCurve {keys [<at 1 0.1667 RCIM_Cubic>]}
      }})
    R1 Cobalt:LinearColor=&ColorFromCurve(ColorCurve=DataInterface() {propertyValues {
       AlphaCurve {defaultValue=7, keys [<at 0.2819 7>, <at 0.6302 1>, <at 0.7212 1>, {interpMode=RCIM_Cubic, leaveTangent=-6.7193, time=0.7822, value=1}]}
       BlueCurve {keys [<at 1 0.5 RCIM_Cubic>]}
       GreenCurve {keys [<at 1 0 RCIM_Cubic>]}
       RedCurve {keys [<at 1 0 RCIM_Cubic>]}
      }})
    R1 Crimson:LinearColor=&ColorFromCurve(ColorCurve=DataInterface() {propertyValues {
       AlphaCurve {keys [<at 0.2819 7>, {interpMode=RCIM_Cubic, leaveTangent=-11.9263, time=0.7822, value=1}]}
       BlueCurve {keys [<at 1 0.0131 RCIM_Cubic>]}
       GreenCurve {keys [<at 1 0 RCIM_Cubic>]}
       RedCurve {keys [<at 1 0.2348 RCIM_Cubic>]}
      }})
    R1 DefaultColor=&ColorFromCurve(ColorCurve=DataInterface() {propertyValues {
       AlphaCurve {keys [<at 0.2819 7>]}
       BlueCurve {keys [<at 1 0.25 RCIM_Cubic>]}
       GreenCurve {keys [<at 1 0 RCIM_Cubic>]}
       RedCurve {keys [<at 1 0.1318 RCIM_Cubic>]}
      }})
    R1 ForestGreen:LinearColor=&ColorFromCurve(ColorCurve=DataInterface() {propertyValues {
       AlphaCurve {keys [<at 0.2819 7>, <at 0.7212 1>, {interpMode=RCIM_Cubic, leaveTangent=-6.7757, time=0.7822, value=1}]}
       BlueCurve {keys [<at 1 0.0198 RCIM_Cubic>]}
       GreenCurve {keys [<at 1 0.2031 RCIM_Cubic>]}
       RedCurve {keys [<at 1 0 RCIM_Cubic>]}
      }})
    R1 Grey:LinearColor=&ColorFromCurve(ColorCurve=DataInterface() {propertyValues {
       AlphaCurve {keys [<at 0.2819 1>, <at 0.7212 1>, {interpMode=RCIM_Cubic, leaveTangent=-6.7757, time=0.7822, value=1}]}
       BlueCurve {keys [<at 1 0.3542 RCIM_Cubic>]}
       GreenCurve {keys [<at 1 0.3542 RCIM_Cubic>]}
       RedCurve {keys [<at 1 0.3542 RCIM_Cubic>]}
      }})
    R1 Lime:LinearColor=&ColorFromCurve(ColorCurve=DataInterface() {propertyValues {
       AlphaCurve {keys [<at 0.2819 7>, <at 0.7212 1>, {interpMode=RCIM_Cubic, leaveTangent=-6.7757, time=0.7822, value=1}]}
       BlueCurve {keys [<at 1 0.0865 RCIM_Cubic>]}
       GreenCurve {keys [<at 1 0.2865 RCIM_Cubic>]}
       RedCurve {keys [<at 1 0.1697 RCIM_Cubic>]}
      }})
    R1 Orange:LinearColor=&ColorFromCurve(ColorCurve=DataInterface() {propertyValues {
       AlphaCurve {keys [<at 0.2819 7>, <at 0.7212 1>, {interpMode=RCIM_Cubic, leaveTangent=-6.7757, time=0.7822, value=1}]}
       BlueCurve {keys [<at 1 0.0688 RCIM_Cubic>]}
       GreenCurve {keys [<at 1 0.1085 RCIM_Cubic>]}
       RedCurve {keys [<at 1 0.3478 RCIM_Cubic>]}
      }})
    R1 Pink:LinearColor=&ColorFromCurve(ColorCurve=DataInterface() {propertyValues {
       AlphaCurve {keys [<at 0.2819 7>, <at 0.7212 1>, {interpMode=RCIM_Cubic, leaveTangent=-6.7757, time=0.7822, value=1}]}
       BlueCurve {keys [<at 1 0.2004 RCIM_Cubic>]}
       GreenCurve {keys [<at 1 0 RCIM_Cubic>]}
       RedCurve {keys [<at 1 0.2188 RCIM_Cubic>]}
      }})
    R1 Purple:LinearColor=&ColorFromCurve(ColorCurve=DataInterface() {propertyValues {
       AlphaCurve {keys [<at 0.2819 7>, <at 0.7212 1>, {interpMode=RCIM_Cubic, leaveTangent=-6.7757, time=0.7822, value=1}]}
       BlueCurve {keys [<at 1 0.375 RCIM_Cubic>]}
       GreenCurve {keys [<at 1 0 RCIM_Cubic>]}
       RedCurve {keys [<at 1 0.1484 RCIM_Cubic>]}
      }})
    R1 Saffron:LinearColor=&ColorFromCurve(ColorCurve=DataInterface() {propertyValues {
       AlphaCurve {keys [<at 0.2819 7>, <at 0.7212 1>, {interpMode=RCIM_Cubic, leaveTangent=-6.7757, time=0.7822, value=1}]}
       BlueCurve {keys [<at 1 0 RCIM_Cubic>]}
       GreenCurve {keys [<at 1 0.2395 RCIM_Cubic>]}
       RedCurve {keys [<at 1 0.2812 RCIM_Cubic>]}
      }})
    R1 SkyBlue:LinearColor=&ColorFromCurve(ColorCurve=DataInterface() {propertyValues {
       AlphaCurve {keys [<at 0.2819 7>, <at 0.7212 1>, {interpMode=RCIM_Cubic, leaveTangent=-6.7757, time=0.7822, value=1}]}
       BlueCurve {keys [<at 1 0.3802 RCIM_Cubic>]}
       GreenCurve {keys [<at 1 0.1905 RCIM_Cubic>]}
       RedCurve {keys [<at 1 0 RCIM_Cubic>]}
      }})
    R1 TitaniumWhite:LinearColor=&ColorFromCurve(ColorCurve=DataInterface() {propertyValues {
       AlphaCurve {keys [<at 0.2819 2>, <at 0.7212 1>, {interpMode=RCIM_Cubic, leaveTangent=-6.7757, time=0.7822, value=1}]}
       BlueCurve {keys [<at 0.1153 0.3802 RCIM_Cubic>,<at 0.479 0>]}
       GreenCurve {keys [<at 0.1153 0.3802 RCIM_Cubic>,<at 0.479 0>]}
       RedCurve {keys [<at 0.1153 0.3802 RCIM_Cubic>,<at 0.479 0>]}
      }})
    R2 Black:LinearColor=&ColorFromCurve(ColorCurve=DataInterface() {propertyValues {
       AlphaCurve {keys [<at 0 7>,<at 0.2232 1>,<at 0.3261 0.05>,<at 0.5495 0.2>,<at 0.7007 0.2>]}
       BlueCurve {keys [<at 0 0.0194 RCIM_Cubic>,<at 0.1492 0.0333>,<at 0.2681 0.1104>,<at 0.4863 0.9274>,<at 0.6088 0.0305>]}
       GreenCurve {keys [<at 0 0.0194 RCIM_Cubic>,<at 0.1492 0.0333>,<at 0.2681 0.1104>,<at 0.4863 0.9274>,<at 0.6088 0.0305>]}
       RedCurve {keys [<at 0 0.0194 RCIM_Cubic>,<at 0.1492 0.0333>,<at 0.2681 0.1104>,<at 0.4863 0.9274>,<at 0.6088 0.0305>]}
      }})
    R2 BurntSienna:LinearColor=&ColorFromCurve(ColorCurve=DataInterface() {propertyValues {
       AlphaCurve {keys [<at 0 4>,<at 0.1907 1>,<at 0.3261 0.05>,<at 0.593 0.2>]}
       BlueCurve {keys [<at 0 0.008 RCIM_Cubic>,<at 0.2839 0.0202>,<at 0.4863 0.2907>,<at 0.5956 0.008>]}
       GreenCurve {keys [<at 0 0.0203 RCIM_Cubic>,<at 0.2839 0.0323>,<at 0.4863 0.4394>,<at 0.5956 0.0203>]}
       RedCurve {keys [<at 0 0.0648 RCIM_Cubic>,<at 0.2839 0.0762>,<at 0.4863 0.9792>,<at 0.5956 0.0648>]}
      }})
    R2 Cobalt:LinearColor=&ColorFromCurve(ColorCurve=DataInterface() {propertyValues {
       AlphaCurve {keys [<at 0 7>,<at 0.1907 1>,<at 0.3261 0.05>,<at 0.5495 0.6>,<at 0.7007 0.2>]}
       BlueCurve {keys [<at 0 0.6458 RCIM_Cubic>,<at 0.2575 0.9985>,<at 0.4863 1>,<at 0.6115 0.8177>]}
       GreenCurve {keys [<at 0 0.0119 RCIM_Cubic>,<at 0.2575 0.0367>,<at 0.4863 0.2997>,<at 0.6115 0.0527>]}
       RedCurve {keys [<at 0 0 RCIM_Cubic>,<at 0.2575 0.0186>,<at 0.4863 0.2865>,<at 0.6115 0.0383>]}
      }})
    R2 Crimson:LinearColor=&ColorFromCurve(ColorCurve=DataInterface() {propertyValues {
       AlphaCurve {keys [<at 0 7>,<at 0.1907 1>,<at 0.3261 0.05>,<at 0.5495 0.2>,<at 0.7007 0.2>]}
       BlueCurve {keys [<at 0 0 RCIM_Cubic>,<at 0.1492 0.0104>,<at 0.4863 0.2995>,<at 0.6458 0.0894>]}
       GreenCurve {keys [<at 0 0 RCIM_Cubic>,<at 0.1492 0.0104>,<at 0.4863 0.2995>,<at 0.6458 0.0894>]}
       RedCurve {keys [<at 0 0.4323 RCIM_Cubic>,<at 0.1492 0.9992>,<at 0.4863 0.9274>,<at 0.6458 0.993>]}
      }})
    R2 DefaultColor:LinearColor=&ColorFromCurve(ColorCurve=DataInterface() {propertyValues {
       AlphaCurve {keys [<at 0 1>,<at 0.0914 0.2857>,<at 0.308 0.05>,<at 0.5592 0.2>]}
       BlueCurve {keys [<at 0.0009 0>,<at 0.2668 -0.2829>,<at 0.5114 0.8849>,<at 0.619 1>,<at 0.682 0.383>]}
       GreenCurve {keys [<at 0.0009 0.0767>,<at 0.2742 -0.1994>,<at 0.5114 0.8849>,<at 0.619 1>,<at 0.682 0.2585>]}
       RedCurve {keys [<at 0.0009 3.2194>,<at 0.0016 3.2194>,<at 0.4479 0.8849>,<at 0.5114 0.8849>,<at 0.619 1>]}
      }})
    R2 ForestGreen:LinearColor=&ColorFromCurve(ColorCurve=DataInterface() {propertyValues {
       AlphaCurve {keys [<at 0 6>,<at 0.1907 1>,<at 0.3261 0.05>,<at 0.5495 0.2>,<at 0.7007 0.2>]}
       BlueCurve {keys [<at 0 0.0172 RCIM_Cubic>,<at 0.2179 0.0322>,<at 0.4863 0.4656>,<at 0.6537 0.1027>]}
       GreenCurve {keys [<at 0 0.3438 RCIM_Cubic>,<at 0.2179 0.245>,<at 0.4863 0.875>,<at 0.6537 0.2987>]}
       RedCurve {keys [<at 0 0 RCIM_Cubic>,<at 0.2179 0.021>,<at 0.4863 0.4466>,<at 0.6537 0.0924>]}
      }})
    R2 Grey:LinearColor=&ColorFromCurve(ColorCurve=DataInterface() {propertyValues {
       AlphaCurve {keys [<at 0 4>,<at 0.1907 1>,<at 0.3261 0.05>,<at 0.5495 0.2>,<at 0.7007 0.2>]}
       BlueCurve {keys [<at 0 0.0692 RCIM_Cubic>,<at 0.2285 0.1336>,<at 0.4863 0.9274>,<at 0.6405 0.132>]}
       GreenCurve {keys [<at 0 0.0781 RCIM_Cubic>,<at 0.2285 0.1336>,<at 0.4863 0.9274>,<at 0.6405 0.132>]}
       RedCurve {keys [<at 0 0.0767 RCIM_Cubic>,<at 0.2285 0.1406>,<at 0.4863 0.9274>,<at 0.6405 0.1354>]}
      }})
    R2 Lime:LinearColor=&ColorFromCurve(ColorCurve=DataInterface() {propertyValues {
       AlphaCurve {keys [<at 0 6>,<at 0.1907 1>,<at 0.3261 0.05>,<at 0.5495 0.2>]}
       BlueCurve {keys [<at 0 0 RCIM_Cubic>,<at 0.2179 0.0459>,<at 0.4863 0.9274>,<at 0.6299 0.0724>]}
       GreenCurve {keys [<at 0 0.6875 RCIM_Cubic>,<at 0.2179 0.4635>,<at 0.4863 0.9274>,<at 0.6299 0.7708>]}
       RedCurve {keys [<at 0 0.3447 RCIM_Cubic>,<at 0.2179 0.2553>,<at 0.4863 0.9274>,<at 0.6299 0.4226>]}
      }})
    R2 Orange:LinearColor=&ColorFromCurve(ColorCurve=DataInterface() {propertyValues {
       AlphaCurve {keys [<at 0 7>,<at 0.1907 1>,<at 0.3261 0.05>,<at 0.5495 0.2>,<at 0.7007 0.2>]}
       BlueCurve {keys [<at 0 0 RCIM_Cubic>,<at 0.2364 0.0598>,<at 0.4863 0.4263>,<at 0.6217 0.1191>]}
       GreenCurve {keys [<at 0 0.2834 RCIM_Cubic>,<at 0.2364 0.2792>,<at 0.4863 0.5719>,<at 0.6217 0.3121>]}
       RedCurve {keys [<at 0 0.8125 RCIM_Cubic>,<at 0.2364 0.6889>,<at 0.4863 0.8438>,<at 0.6217 0.6724>]}
      }})
    R2 Pink:LinearColor=&ColorFromCurve(ColorCurve=DataInterface() {propertyValues {
       AlphaCurve {keys [<at 0 7>,<at 0.1907 1>,<at 0.3261 0.05>,<at 0.5495 0.2>,<at 0.7007 0.2>]}
       BlueCurve {keys [<at 0 0.3419 RCIM_Cubic>,<at 0.2707 0.3564>,<at 0.4863 0.7216>,<at 0.6141 0.5634>]}
       GreenCurve {keys [<at 0 0.0612 RCIM_Cubic>,<at 0.2707 0.1489>,<at 0.4863 0.5789>,<at 0.6141 0.2536>]}
       RedCurve {keys [<at 0 0.5029 RCIM_Cubic>,<at 0.2707 0.5614>,<at 0.4863 0.855>,<at 0.6141 0.8542>]}
      }})
    R2 Purple:LinearColor=&ColorFromCurve(ColorCurve=DataInterface() {propertyValues {
       AlphaCurve {keys [<at 0 7>,<at 0.1907 1>,<at 0.3261 0.05>,<at 0.5495 0.2>,<at 0.7007 0.2>]}
       BlueCurve {keys [<at 0 0.8281 RCIM_Cubic>,<at 0.2919 0.6836>,<at 0.4863 0.8957>,<at 0.6141 0.9565>]}
       GreenCurve {keys [<at 0 0 RCIM_Cubic>,<at 0.2919 0.0563>,<at 0.4863 0.2791>,<at 0.6141 0.1097>]}
       RedCurve {keys [<at 0 0.1924 RCIM_Cubic>,<at 0.2919 0.3349>,<at 0.4863 0.5023>,<at 0.6141 0.3436>]}
      }})
    R2 Saffron:LinearColor=&ColorFromCurve(ColorCurve=DataInterface() {propertyValues {
       AlphaCurve {keys [<at 0 4>,<at 0.1907 1>,<at 0.3261 0.05>,<at 0.5495 0.2>]}
       BlueCurve {keys [<at 0 0 RCIM_Cubic>,<at 0.2697 0.0046>,<at 0.4863 0.3188>,<at 0.6748 0.0798>]}
       GreenCurve {keys [<at 0 0.6105 RCIM_Cubic>,<at 0.2697 0.6881>,<at 0.4863 0.9274>,<at 0.6748 0.6052>]}
       RedCurve {keys [<at 0 0.7969 RCIM_Cubic>,<at 0.2697 0.8967>,<at 0.4863 0.8744>,<at 0.6748 0.7656>]}
      }})
    R2 SkyBlue:LinearColor=&ColorFromCurve(ColorCurve=DataInterface() {propertyValues {
       AlphaCurve {keys [<at 0 7>,<at 0.1907 1>,<at 0.3261 0.05>,<at 0.5495 0.2>]}
       BlueCurve {keys [<at 0 0.5333 RCIM_Cubic>,<at 0.2496 0.5525>,<at 0.4863 0.9323>,<at 0.6141 0.9323>]}
       GreenCurve {keys [<at 0 0.3564 RCIM_Cubic>,<at 0.2496 0.3792>,<at 0.4863 0.8083>,<at 0.6141 0.6988>]}
       RedCurve {keys [<at 0 0 RCIM_Cubic>,<at 0.2496 0.0875>,<at 0.4863 0.5584>,<at 0.6141 0.2282>]}
      }})
    R2 TitaniumWhite:LinearColor=&ColorFromCurve(ColorCurve=DataInterface() {propertyValues {
       AlphaCurve {keys [<at 0 4>,<at 0.1907 1>,<at 0.3261 0.05>,<at 0.5495 0.2>,<at 0.7007 0.2>]}
       BlueCurve {keys [<at 0 0.1846 RCIM_Cubic>,<at 0.2285 0.1336>,<at 0.4863 0.9274>,<at 0.6405 0.3248>]}
       GreenCurve {keys [<at 0 0.2083 RCIM_Cubic>,<at 0.2285 0.1336>,<at 0.4863 0.9274>,<at 0.6405 0.3248>]}
       RedCurve {keys [<at 0 0.2046 RCIM_Cubic>,<at 0.2285 0.1406>,<at 0.4863 0.9274>,<at 0.6405 0.3333>]}
      }})
    R3 Black:LinearColor=&ColorFromCurve(ColorCurve=DataInterface() {propertyValues {
       AlphaCurve {keys [<at 0.2665 3>]}
       BlueCurve {keys [<at 0.6966 0.3043>]}
       GreenCurve {keys [<at 0.6966 0.3043>]}
       RedCurve {keys [<at 0.6966 0.3043>]}
      }})
    R3 BurntSienna:LinearColor=&ColorFromCurve(ColorCurve=DataInterface() {propertyValues {
       AlphaCurve {keys [<at 0.2665 1>]}
       BlueCurve {keys [<at 0.6966 0.2031>]}
       GreenCurve {keys [<at 0.6966 0.2749>]}
       RedCurve {keys [<at 0.6966 0.5>]}
      }})
    R3 Cobalt:LinearColor=&ColorFromCurve(ColorCurve=DataInterface() {propertyValues {
       AlphaCurve {defaultValue=3, keys [{tangentMode=RCTM_User, time=0.8086, value=1}]}
       BlueCurve {keys [<at 0.6966 1.3>]}
       GreenCurve {keys [<at 0.6966 0.1083>]}
       RedCurve {keys [<at 0.6966 0>]}
      }})
    R3 Crimson:LinearColor=&ColorFromCurve(ColorCurve=DataInterface() {propertyValues {AlphaCurve {keys [<at 0.2665 3>]}, BlueCurve {keys [<at 0.6966 0.2>]}, GreenCurve {keys [<at 0.6966 0.2>]}, RedCurve {keys [<at 0.6966 1>]}}})
    R3 DefaultColor:LinearColor=&ColorFromCurve(ColorCurve=DataInterface() {propertyValues {
       AlphaCurve {keys [<at 0.8086 3>]}
       BlueCurve {keys [<at 0.6966 1.2>]}
       GreenCurve {keys [<at 0.6966 0.336>]}
       RedCurve {keys [<at 0.6966 0.12>]}
      }})
    R3 ForestGreen:LinearColor=&ColorFromCurve(ColorCurve=DataInterface() {propertyValues {
       AlphaCurve {keys [<at 0.2665 3>]}
       BlueCurve {keys [<at 0.6966 0.571>]}
       GreenCurve {keys [<at 0.6966 1.5>]}
       RedCurve {keys [<at 0.6966 0.3>]}
      }})
    R3 Grey:LinearColor=&ColorFromCurve(ColorCurve=DataInterface() {propertyValues {
       AlphaCurve {keys [<at 0.2665 3>]}
       BlueCurve {keys [<at 0.6966 0.5>]}
       GreenCurve {keys [<at 0.6966 0.5>]}
       RedCurve {keys [<at 0.6966 0.5>]}
      }})
    R3 Lime:LinearColor=&ColorFromCurve(ColorCurve=DataInterface() {propertyValues {
       AlphaCurve {keys [<at 0.2665 3>]}
       BlueCurve {keys [<at 0.6966 0>]}
       GreenCurve {keys [<at 0.6966 0.9>]}
       RedCurve {keys [<at 0.6966 0.5311>]}
      }})
    R3 Orange:LinearColor=&ColorFromCurve(ColorCurve=DataInterface() {propertyValues {
       AlphaCurve {keys [<at 0.2665 3>]}
       BlueCurve {keys [<at 0.6966 0.0333>]}
       GreenCurve {keys [<at 0.6966 0.3003>]}
       RedCurve {keys [<at 0.6966 0.8>]}
      }})
    R3 Pink:LinearColor=&ColorFromCurve(ColorCurve=DataInterface() {propertyValues {
       AlphaCurve {keys [<at 0.2665 3>]}
       BlueCurve {keys [<at 0.6966 1.2188>]}
       GreenCurve {keys [<at 0.6966 0.2158>]}
       RedCurve {keys [<at 0.6966 1.1488>]}
      }})
    R3 Purple:LinearColor=&ColorFromCurve(ColorCurve=DataInterface() {propertyValues {AlphaCurve {keys [<at 0.2665 3>]}, BlueCurve {keys [<at 0.6966 1>]}, GreenCurve {keys [<at 0.6966 0>]}, RedCurve {keys [<at 0.6966 0.3627>]}}})
    R3 Saffron:LinearColor=&ColorFromCurve(ColorCurve=DataInterface() {propertyValues {AlphaCurve {keys [<at 0.2665 3>]}, BlueCurve {keys [<at 0.6966 0>]}, GreenCurve {keys [<at 0.6966 1.8157>]}, RedCurve {keys [<at 0.6966 2>]}}})
    R3 SkyBlue:LinearColor=&ColorFromCurve(ColorCurve=DataInterface() {propertyValues {
       AlphaCurve {keys [<at 0.2665 3>]}
       BlueCurve {keys [<at 0.6966 1.5>]}
       GreenCurve {keys [<at 0.6966 0.3855>]}
       RedCurve {keys [<at 0.6966 0>]}
      }})
    R3 TitaniumWhite:LinearColor=&ColorFromCurve(ColorCurve=DataInterface() {propertyValues {
       AlphaCurve {keys [<at 0.4952 1>]}
       BlueCurve {keys [<at 0.6966 2.5>]}
       GreenCurve {keys [<at 0.6966 2.5>]}
       RedCurve {keys [<at 0.6966 2.5>]}
      }})
    R4 Black:LinearColor=&ColorFromCurve(ColorCurve=DataInterface() {propertyValues {
       AlphaCurve {keys [<at 0.0115 0.9906>,<at 0.3154 1>]}
       BlueCurve {keys [<at 0.3014 0>]}
       GreenCurve {keys [<at 0.3014 0>]}
       RedCurve {keys [<at 0.3014 0>]}
      }})
    R4 BurntSienna:LinearColor=&ColorFromCurve(ColorCurve=DataInterface() {propertyValues {
       AlphaCurve {keys [<at 0.0115 0.9906>,<at 0.3154 1>,<at 0.5095 1>]}
       BlueCurve {keys [<at 0.3014 0>]}
       GreenCurve {keys [<at 0.3014 0>]}
       RedCurve {keys [<at 0.3014 0>]}
      }})
    R4 Cobalt:LinearColor=&ColorFromCurve(ColorCurve=DataInterface() {propertyValues {
       AlphaCurve {keys [<at 0.0115 0.9906>,<at 0.3154 1>]}
       BlueCurve {keys [<at 0.3014 0>]}
       GreenCurve {keys [<at 0.3014 0>]}
       RedCurve {keys [<at 0.3014 0>]}
      }})
    R4 Crimson:LinearColor=&ColorFromCurve(ColorCurve=DataInterface() {propertyValues {
       AlphaCurve {keys [<at 0.0115 0.9906>,<at 0.3154 1>]}
       BlueCurve {keys [<at 0.3014 0>]}
       GreenCurve {keys [<at 0.3014 0>]}
       RedCurve {keys [<at 0.3014 0>]}
      }})
    R4 DefaultColor:LinearColor=&ColorFromCurve(ColorCurve=DataInterface() {propertyValues {AlphaCurve {keys [<at 0.0115 1>]}, BlueCurve {keys [<at 0.3014 0>]}, GreenCurve {keys [<at 0.3014 0>]}, RedCurve {keys [<at 0.3014 0>]}}})
    R4 ForestGreen:LinearColor=&ColorFromCurve(ColorCurve=DataInterface() {propertyValues {
       AlphaCurve {keys [<at 0.0115 0.9906>,<at 0.3154 1>]}
       BlueCurve {keys [<at 0.3014 0>]}
       GreenCurve {keys [<at 0.3014 0>]}
       RedCurve {keys [<at 0.3014 0>]}
      }})
    R4 Grey:LinearColor=&ColorFromCurve(ColorCurve=DataInterface() {propertyValues {
       AlphaCurve {keys [<at 0.0115 0.9906>,<at 0.3154 1>]}
       BlueCurve {keys [<at 0.3014 0>]}
       GreenCurve {keys [<at 0.3014 0>]}
       RedCurve {keys [<at 0.3014 0>]}
      }})
    R4 Lime:LinearColor=&ColorFromCurve(ColorCurve=DataInterface() {propertyValues {
       AlphaCurve {keys [<at 0.0115 0.9906>,<at 0.3154 1>]}
       BlueCurve {keys [<at 0.3014 0>]}
       GreenCurve {keys [<at 0.3014 0>]}
       RedCurve {keys [<at 0.3014 0>]}
      }})
    R4 Orange:LinearColor=&ColorFromCurve(ColorCurve=DataInterface() {propertyValues {
       AlphaCurve {keys [<at 0.0115 0.9906>,<at 0.3154 1>]}
       BlueCurve {keys [<at 0.3014 0>]}
       GreenCurve {keys [<at 0.3014 0>]}
       RedCurve {keys [<at 0.3014 0>]}
      }})
    R4 Pink:LinearColor=&ColorFromCurve(ColorCurve=DataInterface() {propertyValues {
       AlphaCurve {keys [<at 0.0115 0.9906>,<at 0.3154 1>]}
       BlueCurve {keys [<at 0.3014 0>]}
       GreenCurve {keys [<at 0.3014 0>]}
       RedCurve {keys [<at 0.3014 0>]}
      }})
    R4 Purple:LinearColor=&ColorFromCurve(ColorCurve=DataInterface() {propertyValues {
       AlphaCurve {keys [<at 0.0115 0.9906>,<at 0.3154 1>]}
       BlueCurve {keys [<at 0.3014 0>]}
       GreenCurve {keys [<at 0.3014 0>]}
       RedCurve {keys [<at 0.3014 0>]}
      }})
    R4 Saffron:LinearColor=&ColorFromCurve(ColorCurve=DataInterface() {propertyValues {
       AlphaCurve {keys [<at 0.0115 0.9906>,<at 0.3154 1>]}
       BlueCurve {keys [<at 0.3014 0>]}
       GreenCurve {keys [<at 0.3014 0>]}
       RedCurve {keys [<at 0.3014 0>]}
      }})
    R4 SkyBlue:LinearColor=&ColorFromCurve(ColorCurve=DataInterface() {propertyValues {
       AlphaCurve {keys [<at 0.0115 0.9906>,<at 0.3154 1>]}
       BlueCurve {keys [<at 0.3014 0>]}
       GreenCurve {keys [<at 0.3014 0>]}
       RedCurve {keys [<at 0.3014 0>]}
      }})
    R4 TitaniumWhite:LinearColor=&ColorFromCurve(ColorCurve=DataInterface() {propertyValues {
       AlphaCurve {keys [<at 0.0115 0.9906>,<at 0.1682 1>]}
       BlueCurve {keys [<at 0.3509 0.2708>,<at 0.4158 0>]}
       GreenCurve {keys [<at 0.3509 0.2708>,<at 0.4158 0>]}
       RedCurve {keys [<at 0.3509 0.2708>,<at 0.4158 0>]}
      }})
    R5 Black:LinearColor=<rgba 0.0469 0.0469 0.0469 0.6>
    R5 BurntSienna:LinearColor=<rgba 0.1042 0.0322 0.0128 0.7>
    R5 Cobalt:LinearColor=<rgba 0.0312 0.0401 1 0.6>
    R5 Crimson:LinearColor=<rgba 0.4635 0.0266 0.0384 0.6>
    R5 DefaultColor:LinearColor=<rgba 1 1 1 0.5>
    R5 ForestGreen:LinearColor=<rgba 0.0704 0.901 0.1124 0.4>
    R5 Grey:LinearColor=<rgba 0.1667 0.1667 0.1667 0.4>
    R5 Lime:LinearColor=<rgba 0.2938 0.5104 0.0771 0.25>
    R5 Orange:LinearColor=<rgba 0.8261 0.2993 0.0172 0.3>
    R5 Pink:LinearColor=<rgba 0.651 0.1763 0.4784 0.4>
    R5 Purple:LinearColor=<rgba 0.2704 0.1155 0.7917 0.3>
    R5 Saffron:LinearColor=<rgba 0.8906 0.6824 0 0.15>
    R5 SkyBlue:LinearColor=<rgba 0.0754 0.3822 0.5365 0.4>
    R5 TitaniumWhite:LinearColor=<rgba 0.5729 0.5729 0.5729 0.3>
    Use R1 Painted Overrides=true
    Use R2 Painted Overrides=true
    Use R3 Painted Overrides=true
    Use R4 Painted Overrides=true
    Use R5 Painted Overrides=true
   ) {module=VC_PaintedVariantsOverride001}
   &FN_DisableTireFX(
    Mask Index Attribute:NiagaraInt32=$Particles.PartitionParticles.Partition
    Mask Index FL=2
    Mask Index FR=3
    Mask Index RL=0
    Mask Index RR=1
   )
   &VC_WheelWrapScale(Wheel Index Attribute:NiagaraInt32=&Add_Integer(A=$Particles.PartitionParticles.Partition, B=2))
   &ParticleState(Kill Particles When Lifetime Has Elapsed=false)
   &UpdateMeshOrientation(Rotation Coordinate Space=::NewEnumerator2, Rotation Vector=<0 1 0>) {enabled=false}
   &SkeletalMeshLocation(
    Bone / Socket Index=&Add_Integer(A=$Particles.PartitionParticles.Partition, B=4)
    Bone / Socket Sampling Mode=::NewEnumerator5
    MaskSpawnGroup=false
    Mesh Sampling Type=::NewEnumerator6
    Skeletal Mesh=$System.SkeletalMesh
   ) {module=SkeletalMeshLocation001}
   &SolveForcesAndVelocity(Clamp Velocity=false, Limit Acceleration=false)
   /SetVariables(
    Particles.R1_Color_Scaled:LinearColor=&MultiplyLinearColor(
     Linear Color=$Particles.R1 Color
     Scale Factor=&MakeVector4FromVectorAndFloat("Vector (RGB)"=&VectorFromFloat(Value=1), "Float (Alpha)"=1)
    )
    Particles.R3_Color_Scaled:LinearColor=&MultiplyLinearColor(
     Linear Color=$Particles.R3 Color
     Scale Factor=&MakeVector4FromVectorAndFloat("Vector (RGB)"=&VectorFromFloat(Value=0.7), "Float (Alpha)"=1)
    )
    Particles.R4_Color_Scaled:LinearColor=&MultiplyLinearColor(
     Linear Color=$Particles.R4 Color
     Scale Factor=&MakeVector4FromVectorAndFloat("Vector (RGB)"=&VectorFromFloat(Value=10), "Float (Alpha)"=1)
    )
    Particles.R5_Color_Scaled:LinearColor=&MultiplyLinearColor(
     Linear Color=$Particles.R5 Color
     Scale Factor=&MakeVector4FromVectorAndFloat("Vector (RGB)"=&VectorFromFloat(Value=1), "Float (Alpha)"=1)
    )
   ) {isSetParameters=true}
   /SetVariables(Particles.RibbonGlow:LinearColor=<rgba 0 0.5029 1 1>, Particles.RibbonGlow 2:LinearColor=<rgba 0 0.0648 1 1>, Particles.Scale Glow 2=1) {isSetParameters=true}
  ]
  rendererClasses [&NiagaraMeshRendererProperties]
  renderers [
   {rendererClass=&NiagaraMeshRendererProperties, rendererIndex=0}
   {rendererClass=&NiagaraMeshRendererProperties, rendererIndex=1}
   {rendererClass=&NiagaraMeshRendererProperties, rendererIndex=2}
   {rendererClass=&NiagaraMeshRendererProperties, rendererIndex=3}
   {rendererClass=&NiagaraMeshRendererProperties, rendererIndex=4}
   {rendererClass=&NiagaraMeshRendererProperties, rendererIndex=5}
  ]
 }
 WheelWrap {
  emitterUpdateScript [
   &EmitterState(Inactive Response=::NewEnumerator1, Life Cycle Mode=::NewEnumerator1, Loop Duration=1, UseLoopDelay=false)
   &SpawnBurst_Instantaneous(Spawn Count=2, Use Spawn Probability=false, UseLoopCountLimit=false)
  ]
  particleSpawnScript [
   &VC_PaintedVariantsOverride(Use R1 Painted Overrides=true, Use R2 Painted Overrides=true) {enabled=false}
   &InitializeParticle(
    Color Mode=::NewEnumerator0
    Lifetime=$Emitter.CurrentLoopDuration
    Mesh Scale Mode=::NewEnumerator4
    Sprite Rotation Mode=::NewEnumerator2
    Sprite Size Mode=::NewEnumerator3
    UsePositionOffset=false
    Write Material Random=true
   )
   &PartitionParticles()
   &Color(Color=<rgba 0.026 0.026 0.5 3>)
   &SkeletalMeshLocation(
    Bone / Socket Index=&Add_Integer(A=$Particles.PartitionParticles.Partition, B=4)
    Bone / Socket Sampling Mode=::NewEnumerator5
    Interpolate Mesh Sampling=false
    MaskSpawnGroup=false
    Mesh Sampling Type=::NewEnumerator6
    Skeletal Mesh=$System.SkeletalMesh
   )
  ]
  particleUpdateScript [
   &VC_PaintedVariantsOverride(
    R1 Black:LinearColor=&ColorFromCurve(ColorCurve=DataInterface() {propertyValues {
       AlphaCurve {keys [<at 0.2819 7>]}
       BlueCurve {keys [<at 1 0.1094 RCIM_Cubic>]}
       GreenCurve {keys [<at 1 0.1094 RCIM_Cubic>]}
       RedCurve {keys [<at 1 0.1094 RCIM_Cubic>]}
      }})
    R1 BurntSienna:LinearColor=&ColorFromCurve(ColorCurve=DataInterface() {propertyValues {
       AlphaCurve {keys [<at 0.2819 7>]}
       BlueCurve {keys [<at 1 0.0686 RCIM_Cubic>]}
       GreenCurve {keys [<at 1 0.0858 RCIM_Cubic>]}
       RedCurve {keys [<at 1 0.1667 RCIM_Cubic>]}
      }})
    R1 Cobalt:LinearColor=&ColorFromCurve(ColorCurve=DataInterface() {propertyValues {
       AlphaCurve {defaultValue=7, keys [<at 0.2819 7>, <at 0.6302 1>, <at 0.7212 1>, {interpMode=RCIM_Cubic, leaveTangent=-6.7193, time=0.7822, value=1}]}
       BlueCurve {keys [<at 1 0.5 RCIM_Cubic>]}
       GreenCurve {keys [<at 1 0 RCIM_Cubic>]}
       RedCurve {keys [<at 1 0 RCIM_Cubic>]}
      }})
    R1 Crimson:LinearColor=&ColorFromCurve(ColorCurve=DataInterface() {propertyValues {
       AlphaCurve {keys [<at 0.2819 7>, {interpMode=RCIM_Cubic, leaveTangent=-11.9263, time=0.7822, value=1}]}
       BlueCurve {keys [<at 1 0.0131 RCIM_Cubic>]}
       GreenCurve {keys [<at 1 0 RCIM_Cubic>]}
       RedCurve {keys [<at 1 0.2348 RCIM_Cubic>]}
      }})
    R1 DefaultColor=&ColorFromCurve(ColorCurve=DataInterface() {propertyValues {
       AlphaCurve {keys [<at 0.2819 7>]}
       BlueCurve {keys [<at 0.0168 0.12 RCIM_Cubic>]}
       GreenCurve {keys [<at 0.0168 0 RCIM_Cubic>]}
       RedCurve {keys [<at 0.0168 0.0633 RCIM_Cubic>]}
      }})
    R1 ForestGreen:LinearColor=&ColorFromCurve(ColorCurve=DataInterface() {propertyValues {
       AlphaCurve {keys [<at 0.2819 7>, <at 0.7212 1>, {interpMode=RCIM_Cubic, leaveTangent=-6.7757, time=0.7822, value=1}]}
       BlueCurve {keys [<at 1 0.0198 RCIM_Cubic>]}
       GreenCurve {keys [<at 1 0.2031 RCIM_Cubic>]}
       RedCurve {keys [<at 1 0 RCIM_Cubic>]}
      }})
    R1 Grey:LinearColor=&ColorFromCurve(ColorCurve=DataInterface() {propertyValues {
       AlphaCurve {keys [<at 0.2819 1>, <at 0.7212 1>, {interpMode=RCIM_Cubic, leaveTangent=-6.7757, time=0.7822, value=1}]}
       BlueCurve {keys [<at 1 0.3542 RCIM_Cubic>]}
       GreenCurve {keys [<at 1 0.3542 RCIM_Cubic>]}
       RedCurve {keys [<at 1 0.3542 RCIM_Cubic>]}
      }})
    R1 Lime:LinearColor=&ColorFromCurve(ColorCurve=DataInterface() {propertyValues {
       AlphaCurve {keys [<at 0.2819 7>, <at 0.7212 1>, {interpMode=RCIM_Cubic, leaveTangent=-6.7757, time=0.7822, value=1}]}
       BlueCurve {keys [<at 1 0.0865 RCIM_Cubic>]}
       GreenCurve {keys [<at 1 0.2865 RCIM_Cubic>]}
       RedCurve {keys [<at 1 0.1697 RCIM_Cubic>]}
      }})
    R1 Orange:LinearColor=&ColorFromCurve(ColorCurve=DataInterface() {propertyValues {
       AlphaCurve {keys [<at 0.2819 7>, <at 0.7212 1>, {interpMode=RCIM_Cubic, leaveTangent=-6.7757, time=0.7822, value=1}]}
       BlueCurve {keys [<at 1 0.0688 RCIM_Cubic>]}
       GreenCurve {keys [<at 1 0.1085 RCIM_Cubic>]}
       RedCurve {keys [<at 1 0.3478 RCIM_Cubic>]}
      }})
    R1 Pink:LinearColor=&ColorFromCurve(ColorCurve=DataInterface() {propertyValues {
       AlphaCurve {keys [<at 0.2819 7>, <at 0.7212 1>, {interpMode=RCIM_Cubic, leaveTangent=-6.7757, time=0.7822, value=1}]}
       BlueCurve {keys [<at 1 0.2004 RCIM_Cubic>]}
       GreenCurve {keys [<at 1 0 RCIM_Cubic>]}
       RedCurve {keys [<at 1 0.2188 RCIM_Cubic>]}
      }})
    R1 Purple:LinearColor=&ColorFromCurve(ColorCurve=DataInterface() {propertyValues {
       AlphaCurve {keys [<at 0.2819 7>, <at 0.7212 1>, {interpMode=RCIM_Cubic, leaveTangent=-6.7757, time=0.7822, value=1}]}
       BlueCurve {keys [<at 1 0.13 RCIM_Cubic>]}
       GreenCurve {keys [<at 1 0 RCIM_Cubic>]}
       RedCurve {keys [<at 1 0.0515 RCIM_Cubic>]}
      }})
    R1 Saffron:LinearColor=&ColorFromCurve(ColorCurve=DataInterface() {propertyValues {
       AlphaCurve {keys [<at 0.2819 7>, <at 0.7212 1>, {interpMode=RCIM_Cubic, leaveTangent=-6.7757, time=0.7822, value=1}]}
       BlueCurve {keys [<at 1 0 RCIM_Cubic>]}
       GreenCurve {keys [<at 1 0.2395 RCIM_Cubic>]}
       RedCurve {keys [<at 1 0.2812 RCIM_Cubic>]}
      }})
    R1 SkyBlue:LinearColor=&ColorFromCurve(ColorCurve=DataInterface() {propertyValues {
       AlphaCurve {keys [<at 0.2819 7>, <at 0.7212 1>, {interpMode=RCIM_Cubic, leaveTangent=-6.7757, time=0.7822, value=1}]}
       BlueCurve {keys [<at 1 0.3802 RCIM_Cubic>]}
       GreenCurve {keys [<at 1 0.1905 RCIM_Cubic>]}
       RedCurve {keys [<at 1 0 RCIM_Cubic>]}
      }})
    R1 TitaniumWhite:LinearColor=&ColorFromCurve(ColorCurve=DataInterface() {propertyValues {
       AlphaCurve {keys [<at 0.2819 1>]}
       BlueCurve {keys [<at 0.6326 0.25 RCIM_Cubic>]}
       GreenCurve {keys [<at 0.6326 0.25 RCIM_Cubic>]}
       RedCurve {keys [<at 0.6326 0.25 RCIM_Cubic>]}
      }})
    R2 Black:LinearColor=&ColorFromCurve(ColorCurve=DataInterface() {propertyValues {
       AlphaCurve {
        defaultValue=0
        keys [<at 0 0>, {arriveTangent=15.9916, interpMode=RCIM_Cubic, leaveTangent=15.9916, time=0.1339, value=8}, <at 0.6253 10>, <at 0.9771 0>]
       }
       BlueCurve {keys [<at 1 0.1094 RCIM_Cubic>]}
       GreenCurve {keys [<at 1 0.1094 RCIM_Cubic>]}
       RedCurve {keys [<at 1 0.1094 RCIM_Cubic>]}
      }})
    R2 BurntSienna:LinearColor=&ColorFromCurve(ColorCurve=DataInterface() {propertyValues {
       AlphaCurve {
        defaultValue=0
        keys [<at 0 0>, {arriveTangent=15.9916, interpMode=RCIM_Cubic, leaveTangent=15.9916, time=0.1756, value=8}, <at 0.6253 10>, <at 0.9771 0>]
       }
       BlueCurve {keys [<at 1 0.0686 RCIM_Cubic>]}
       GreenCurve {keys [<at 1 0.0858 RCIM_Cubic>]}
       RedCurve {keys [<at 1 0.1667 RCIM_Cubic>]}
      }})
    R2 Cobalt:LinearColor=&ColorFromCurve(ColorCurve=DataInterface() {propertyValues {
       AlphaCurve {defaultValue=0, keys [<at 0 0>, {arriveTangent=15.9916, interpMode=RCIM_Cubic, leaveTangent=15.9916, time=0.1756, value=8}, <at 0.6253 10>]}
       BlueCurve {keys [<at 1 0.3802 RCIM_Cubic>]}
       GreenCurve {keys [<at 1 0.0931 RCIM_Cubic>]}
       RedCurve {keys [<at 1 0.0931 RCIM_Cubic>]}
      }})
    R2 Crimson:LinearColor=&ColorFromCurve(ColorCurve=DataInterface() {propertyValues {
       AlphaCurve {
        defaultValue=0
        keys [<at 0 0>, {arriveTangent=15.9916, interpMode=RCIM_Cubic, leaveTangent=15.9916, time=0.1756, value=8}, <at 0.6253 10>, <at 0.9771 0>]
       }
       BlueCurve {keys [<at 1 0.0131 RCIM_Cubic>]}
       GreenCurve {keys [<at 1 0 RCIM_Cubic>]}
       RedCurve {keys [<at 1 0.2348 RCIM_Cubic>]}
      }})
    R2 DefaultColor:LinearColor=&ColorFromCurve(ColorCurve=DataInterface() {propertyValues {
       AlphaCurve {keys [<at 0 0>, {arriveTangent=15.9916, interpMode=RCIM_Cubic, leaveTangent=15.9916, time=0.1039, value=8}, <at 0.6253 10>]}
       BlueCurve {keys [<at 0.6966 1>]}
       GreenCurve {keys [<at 0.6966 0.5>]}
       RedCurve {keys [<at 0.6966 0>]}
      }})
    R2 ForestGreen:LinearColor=&ColorFromCurve(ColorCurve=DataInterface() {propertyValues {
       AlphaCurve {
        defaultValue=0
        keys [<at 0 0>, {arriveTangent=15.9916, interpMode=RCIM_Cubic, leaveTangent=15.9916, time=0.1756, value=8}, <at 0.6253 10>, <at 0.9771 0>]
       }
       BlueCurve {keys [<at 1 0.0198 RCIM_Cubic>]}
       GreenCurve {keys [<at 1 0.2031 RCIM_Cubic>]}
       RedCurve {keys [<at 1 0 RCIM_Cubic>]}
      }})
    R2 Grey:LinearColor=&ColorFromCurve(ColorCurve=DataInterface() {propertyValues {
       AlphaCurve {
        defaultValue=0
        keys [<at 0 0>, {arriveTangent=15.9916, interpMode=RCIM_Cubic, leaveTangent=15.9916, time=0.1756, value=8}, <at 0.6253 10>, <at 0.9771 0>]
       }
       BlueCurve {keys [<at 1 0.3542 RCIM_Cubic>]}
       GreenCurve {keys [<at 1 0.3542 RCIM_Cubic>]}
       RedCurve {keys [<at 1 0.3542 RCIM_Cubic>]}
      }})
    R2 Lime:LinearColor=&ColorFromCurve(ColorCurve=DataInterface() {propertyValues {
       AlphaCurve {
        defaultValue=0
        keys [<at 0 0>, {arriveTangent=15.9916, interpMode=RCIM_Cubic, leaveTangent=15.9916, time=0.1756, value=8}, <at 0.6253 10>, <at 0.9771 0>]
       }
       BlueCurve {keys [<at 1 0.0865 RCIM_Cubic>]}
       GreenCurve {keys [<at 1 0.2865 RCIM_Cubic>]}
       RedCurve {keys [<at 1 0.1697 RCIM_Cubic>]}
      }})
    R2 Orange:LinearColor=&ColorFromCurve(ColorCurve=DataInterface() {propertyValues {
       AlphaCurve {
        defaultValue=0
        keys [<at 0 0>, {arriveTangent=15.9916, interpMode=RCIM_Cubic, leaveTangent=15.9916, time=0.1756, value=8}, <at 0.6253 10>, <at 0.9771 0>]
       }
       BlueCurve {keys [<at 1 0 RCIM_Cubic>]}
       GreenCurve {keys [<at 1 0.2961 RCIM_Cubic>]}
       RedCurve {keys [<at 1 0.9271 RCIM_Cubic>]}
      }})
    R2 Pink:LinearColor=&ColorFromCurve(ColorCurve=DataInterface() {propertyValues {
       AlphaCurve {
        defaultValue=0
        keys [<at 0 0>, {arriveTangent=15.9916, interpMode=RCIM_Cubic, leaveTangent=15.9916, time=0.1756, value=8}, <at 0.6253 10>, <at 0.9771 0>]
       }
       BlueCurve {keys [<at 1 0.2004 RCIM_Cubic>]}
       GreenCurve {keys [<at 1 0 RCIM_Cubic>]}
       RedCurve {keys [<at 1 0.2188 RCIM_Cubic>]}
      }})
    R2 Purple:LinearColor=&ColorFromCurve(ColorCurve=DataInterface() {propertyValues {
       AlphaCurve {
        defaultValue=0
        keys [<at 0 0>, {arriveTangent=15.9916, interpMode=RCIM_Cubic, leaveTangent=15.9916, time=0.1756, value=8}, <at 0.6253 10>, <at 0.9771 0>]
       }
       BlueCurve {keys [<at 1 0.375 RCIM_Cubic>]}
       GreenCurve {keys [<at 1 0 RCIM_Cubic>]}
       RedCurve {keys [<at 1 0.1484 RCIM_Cubic>]}
      }})
    R2 Saffron:LinearColor=&ColorFromCurve(ColorCurve=DataInterface() {propertyValues {
       AlphaCurve {
        defaultValue=0
        keys [<at 0 0>, {arriveTangent=15.9916, interpMode=RCIM_Cubic, leaveTangent=15.9916, time=0.1756, value=8}, <at 0.6253 10>, <at 0.9771 0>]
       }
       BlueCurve {keys [<at 1 0 RCIM_Cubic>]}
       GreenCurve {keys [<at 1 0.2395 RCIM_Cubic>]}
       RedCurve {keys [<at 1 0.2812 RCIM_Cubic>]}
      }})
    R2 SkyBlue:LinearColor=&ColorFromCurve(ColorCurve=DataInterface() {propertyValues {
       AlphaCurve {
        defaultValue=0
        keys [<at 0 0>, {arriveTangent=15.9916, interpMode=RCIM_Cubic, leaveTangent=15.9916, time=0.1756, value=8}, <at 0.6253 10>, <at 0.9771 0>]
       }
       BlueCurve {keys [<at 1 0.3802 RCIM_Cubic>]}
       GreenCurve {keys [<at 1 0.1905 RCIM_Cubic>]}
       RedCurve {keys [<at 1 0 RCIM_Cubic>]}
      }})
    R2 TitaniumWhite:LinearColor=&ColorFromCurve(ColorCurve=DataInterface() {propertyValues {
       AlphaCurve {
        defaultValue=0
        keys [<at 0 0>, {arriveTangent=15.9916, interpMode=RCIM_Cubic, leaveTangent=15.9916, time=0.1756, value=8}, <at 0.6253 10>, <at 0.9771 0>]
       }
       BlueCurve {keys [<at 1 0.6406 RCIM_Cubic>]}
       GreenCurve {keys [<at 1 0.6406 RCIM_Cubic>]}
       RedCurve {keys [<at 1 0.6406 RCIM_Cubic>]}
      }})
    R3 Black:LinearColor=&ColorFromCurve(ColorCurve=DataInterface() {propertyValues {
       AlphaCurve {keys [<at 0.2665 3>]}
       BlueCurve {keys [<at 0.6966 0.3043>]}
       GreenCurve {keys [<at 0.6966 0.3043>]}
       RedCurve {keys [<at 0.6966 0.3043>]}
      }})
    R3 BurntSienna:LinearColor=&ColorFromCurve(ColorCurve=DataInterface() {propertyValues {
       AlphaCurve {keys [<at 0.2665 1>]}
       BlueCurve {keys [<at 0.6966 0.2729>]}
       GreenCurve {keys [<at 0.6966 0.3693>]}
       RedCurve {keys [<at 0.6966 0.6719>]}
      }})
    R3 Cobalt:LinearColor=&ColorFromCurve(ColorCurve=DataInterface() {propertyValues {
       AlphaCurve {defaultValue=3, keys [{tangentMode=RCTM_User, time=0.8086, value=1}]}
       BlueCurve {keys [<at 0.6966 1.3>]}
       GreenCurve {keys [<at 0.6966 0.1083>]}
       RedCurve {keys [<at 0.6966 0>]}
      }})
    R3 Crimson:LinearColor=&ColorFromCurve(ColorCurve=DataInterface() {propertyValues {AlphaCurve {keys [<at 0.2665 3>]}, BlueCurve {keys [<at 0.6966 0>]}, GreenCurve {keys [<at 0.6966 0>]}, RedCurve {keys [<at 0.6966 1>]}}})
    R3 DefaultColor:LinearColor=&ColorFromCurve(ColorCurve=DataInterface() {propertyValues {AlphaCurve {keys [<at 0.8086 3>]}, BlueCurve {keys [<at 0.6966 1.5>]}, GreenCurve {keys [<at 0.6966 0.3>]}, RedCurve {keys [<at 0.6966 0>]}}})
    R3 ForestGreen:LinearColor=&ColorFromCurve(ColorCurve=DataInterface() {propertyValues {
       AlphaCurve {keys [<at 0.2665 3>]}
       BlueCurve {keys [<at 0.6966 0.3387>]}
       GreenCurve {keys [<at 0.6966 1.5>]}
       RedCurve {keys [<at 0.6966 0>]}
      }})
    R3 Grey:LinearColor=&ColorFromCurve(ColorCurve=DataInterface() {propertyValues {
       AlphaCurve {keys [<at 0.2665 3>]}
       BlueCurve {keys [<at 0.6966 0.5>]}
       GreenCurve {keys [<at 0.6966 0.5>]}
       RedCurve {keys [<at 0.6966 0.5>]}
      }})
    R3 Lime:LinearColor=&ColorFromCurve(ColorCurve=DataInterface() {propertyValues {AlphaCurve {keys [<at 0.2665 3>]}, BlueCurve {keys [<at 0.6966 0>]}, GreenCurve {keys [<at 0.6966 1>]}, RedCurve {keys [<at 0.6966 0.5901>]}}})
    R3 Orange:LinearColor=&ColorFromCurve(ColorCurve=DataInterface() {propertyValues {
       AlphaCurve {keys [<at 0.2665 3>]}
       BlueCurve {keys [<at 0.6966 0.0304>]}
       GreenCurve {keys [<at 0.6966 0.2737>]}
       RedCurve {keys [<at 0.6966 0.7292>]}
      }})
    R3 Pink:LinearColor=&ColorFromCurve(ColorCurve=DataInterface() {propertyValues {
       AlphaCurve {keys [<at 0.2665 3>]}
       BlueCurve {keys [<at 0.6966 1.2188>]}
       GreenCurve {keys [<at 0.6966 0.2158>]}
       RedCurve {keys [<at 0.6966 1.1488>]}
      }})
    R3 Purple:LinearColor=&ColorFromCurve(ColorCurve=DataInterface() {propertyValues {AlphaCurve {keys [<at 0.2665 3>]}, BlueCurve {keys [<at 0.6966 1>]}, GreenCurve {keys [<at 0.6966 0>]}, RedCurve {keys [<at 0.6966 0.3627>]}}})
    R3 Saffron:LinearColor=&ColorFromCurve(ColorCurve=DataInterface() {propertyValues {AlphaCurve {keys [<at 0.2665 3>]}, BlueCurve {keys [<at 0.6966 0>]}, GreenCurve {keys [<at 0.6966 1.8157>]}, RedCurve {keys [<at 0.6966 2>]}}})
    R3 SkyBlue:LinearColor=&ColorFromCurve(ColorCurve=DataInterface() {propertyValues {
       AlphaCurve {keys [<at 0.2665 3>]}
       BlueCurve {keys [<at 0.6966 1.5>]}
       GreenCurve {keys [<at 0.6966 0.3855>]}
       RedCurve {keys [<at 0.6966 0>]}
      }})
    R3 TitaniumWhite:LinearColor=&ColorFromCurve(ColorCurve=DataInterface() {propertyValues {
       AlphaCurve {keys [<at 0.4952 1>]}
       BlueCurve {keys [<at 0.6966 2.5>]}
       GreenCurve {keys [<at 0.6966 2.5>]}
       RedCurve {keys [<at 0.6966 2.5>]}
      }})
    R4 Black:LinearColor=&ColorFromCurve(ColorCurve=DataInterface() {propertyValues {
       AlphaCurve {keys [<at 0.0115 0.9906>,<at 0.3154 1>]}
       BlueCurve {keys [<at 0.3014 0>]}
       GreenCurve {keys [<at 0.3014 0>]}
       RedCurve {keys [<at 0.3014 0>]}
      }})
    R4 BurntSienna:LinearColor=&ColorFromCurve(ColorCurve=DataInterface() {propertyValues {
       AlphaCurve {keys [<at 0.0115 0.9906>,<at 0.3154 1>]}
       BlueCurve {keys [<at 0.3014 0>]}
       GreenCurve {keys [<at 0.3014 0>]}
       RedCurve {keys [<at 0.3014 0>]}
      }})
    R4 Cobalt:LinearColor=&ColorFromCurve(ColorCurve=DataInterface() {propertyValues {
       AlphaCurve {keys [<at 0.0115 0.9906>,<at 0.3154 1>]}
       BlueCurve {keys [<at 0.3014 0>]}
       GreenCurve {keys [<at 0.3014 0>]}
       RedCurve {keys [<at 0.3014 0>]}
      }})
    R4 Crimson:LinearColor=&ColorFromCurve(ColorCurve=DataInterface() {propertyValues {
       AlphaCurve {keys [<at 0.0115 0.9906>,<at 0.3154 1>]}
       BlueCurve {keys [<at 0.3014 0>]}
       GreenCurve {keys [<at 0.3014 0>]}
       RedCurve {keys [<at 0.3014 0>]}
      }})
    R4 DefaultColor:LinearColor=&ColorFromCurve(ColorCurve=DataInterface() {propertyValues {AlphaCurve {keys [<at 0.0115 1>]}, BlueCurve {keys [<at 0.3014 0>]}, GreenCurve {keys [<at 0.3014 0>]}, RedCurve {keys [<at 0.3014 0>]}}})
    R4 ForestGreen:LinearColor=&ColorFromCurve(ColorCurve=DataInterface() {propertyValues {
       AlphaCurve {keys [<at 0.0115 0.9906>,<at 0.3154 1>]}
       BlueCurve {keys [<at 0.3014 0>]}
       GreenCurve {keys [<at 0.3014 0>]}
       RedCurve {keys [<at 0.3014 0>]}
      }})
    R4 Grey:LinearColor=&ColorFromCurve(ColorCurve=DataInterface() {propertyValues {
       AlphaCurve {keys [<at 0.0115 0.9906>,<at 0.3154 1>]}
       BlueCurve {keys [<at 0.3014 0>]}
       GreenCurve {keys [<at 0.3014 0>]}
       RedCurve {keys [<at 0.3014 0>]}
      }})
    R4 Lime:LinearColor=&ColorFromCurve(ColorCurve=DataInterface() {propertyValues {
       AlphaCurve {keys [<at 0.0115 0.9906>,<at 0.3154 1>]}
       BlueCurve {keys [<at 0.3014 0>]}
       GreenCurve {keys [<at 0.3014 0>]}
       RedCurve {keys [<at 0.3014 0>]}
      }})
    R4 Orange:LinearColor=&ColorFromCurve(ColorCurve=DataInterface() {propertyValues {
       AlphaCurve {keys [<at 0.0115 0.9906>,<at 0.3154 1>]}
       BlueCurve {keys [<at 0.3014 0>]}
       GreenCurve {keys [<at 0.3014 0>]}
       RedCurve {keys [<at 0.3014 0>]}
      }})
    R4 Pink:LinearColor=&ColorFromCurve(ColorCurve=DataInterface() {propertyValues {
       AlphaCurve {keys [<at 0.0115 0.9906>,<at 0.3154 1>]}
       BlueCurve {keys [<at 0.3014 0>]}
       GreenCurve {keys [<at 0.3014 0>]}
       RedCurve {keys [<at 0.3014 0>]}
      }})
    R4 Purple:LinearColor=&ColorFromCurve(ColorCurve=DataInterface() {propertyValues {
       AlphaCurve {keys [<at 0.0115 0.9906>,<at 0.3154 1>]}
       BlueCurve {keys [<at 0.3014 0>]}
       GreenCurve {keys [<at 0.3014 0>]}
       RedCurve {keys [<at 0.3014 0>]}
      }})
    R4 Saffron:LinearColor=&ColorFromCurve(ColorCurve=DataInterface() {propertyValues {
       AlphaCurve {keys [<at 0.0115 0.9906>,<at 0.3154 1>]}
       BlueCurve {keys [<at 0.3014 0>]}
       GreenCurve {keys [<at 0.3014 0>]}
       RedCurve {keys [<at 0.3014 0>]}
      }})
    R4 SkyBlue:LinearColor=&ColorFromCurve(ColorCurve=DataInterface() {propertyValues {
       AlphaCurve {keys [<at 0.0115 0.9906>,<at 0.3154 1>]}
       BlueCurve {keys [<at 0.3014 0>]}
       GreenCurve {keys [<at 0.3014 0>]}
       RedCurve {keys [<at 0.3014 0>]}
      }})
    R4 TitaniumWhite:LinearColor=&ColorFromCurve(ColorCurve=DataInterface() {propertyValues {
       AlphaCurve {keys [<at 0.0115 0.9906>,<at 0.1682 1>]}
       BlueCurve {keys [<at 0.3509 0.2708>,<at 0.4158 0>]}
       GreenCurve {keys [<at 0.3509 0.2708>,<at 0.4158 0>]}
       RedCurve {keys [<at 0.3509 0.2708>,<at 0.4158 0>]}
      }})
    R5 Black:LinearColor=&ColorFromCurve(ColorCurve=DataInterface() {propertyValues {
       AlphaCurve {
        defaultValue=0
        keys [
         {
          arriveTangent=13.1575
          arriveTangentWeight=0.0306
          interpMode=RCIM_Cubic
          leaveTangent=13.1575
          leaveTangentWeight=0.2107
          tangentMode=RCTM_User
          time=0.1591
          value=7
         }
         {tangentMode=RCTM_User, time=0.5992, value=7}
        ]
       }
       BlueCurve {keys [<at 0.6966 0.3043>]}
       GreenCurve {keys [<at 0.6966 0.3043>]}
       RedCurve {keys [<at 0.6966 0.3043>]}
      }})
    R5 BurntSienna:LinearColor=&ColorFromCurve(ColorCurve=DataInterface() {propertyValues {
       AlphaCurve {defaultValue=0, keys [{arriveTangent=13.1575, interpMode=RCIM_Cubic, time=0.0863, value=7}, <at 0.5992 7>]}
       BlueCurve {keys [<at 0.6966 0.2729>]}
       GreenCurve {keys [<at 0.6966 0.3693>]}
       RedCurve {keys [<at 0.6966 0.6719>]}
      }})
    R5 Cobalt:LinearColor=&ColorFromCurve(ColorCurve=DataInterface() {propertyValues {
       AlphaCurve {
        defaultValue=0
        keys [
         {arriveTangent=13.1575, interpMode=RCIM_Cubic, leaveTangent=13.1575, tangentMode=RCTM_User, time=0.1591, value=7}
         {tangentMode=RCTM_User, time=0.5992, value=7}
        ]
       }
       BlueCurve {keys [<at 0.6966 2.5>]}
       GreenCurve {keys [<at 0.6966 0.2198>]}
       RedCurve {keys [<at 0.6966 0.125>]}
      }})
    R5 Crimson:LinearColor=&ColorFromCurve(ColorCurve=DataInterface() {propertyValues {
       AlphaCurve {
        defaultValue=0
        keys [
         {arriveTangent=13.1575, interpMode=RCIM_Cubic, leaveTangent=13.1575, tangentMode=RCTM_User, time=0.1591, value=7}
         {tangentMode=RCTM_User, time=0.5992, value=7}
        ]
       }
       BlueCurve {keys [<at 0.6966 0>]}
       GreenCurve {keys [<at 0.6966 0>]}
       RedCurve {keys [<at 0.6966 1>]}
      }})
    R5 DefaultColor:LinearColor=&ColorFromCurve(ColorCurve=DataInterface() {propertyValues {
       AlphaCurve {keys [
         {
          arriveTangent=13.1575
          arriveTangentWeight=0.0306
          interpMode=RCIM_Cubic
          leaveTangent=13.1575
          leaveTangentWeight=0.2538
          tangentMode=RCTM_User
          time=0.1591
          value=7
         }
         {tangentMode=RCTM_User, time=0.5992, value=7}
        ]}
       BlueCurve {keys [<at 0.4821 1>]}
       GreenCurve {keys [<at 0.4821 0.0657>]}
       RedCurve {keys [<at 0.4821 0>]}
      }})
    R5 ForestGreen:LinearColor=&ColorFromCurve(ColorCurve=DataInterface() {propertyValues {
       AlphaCurve {
        defaultValue=0
        keys [
         {arriveTangent=13.1575, interpMode=RCIM_Cubic, leaveTangent=13.1575, tangentMode=RCTM_User, time=0.1591, value=7}
         {tangentMode=RCTM_User, time=0.5992, value=7}
        ]
       }
       BlueCurve {keys [<at 0.6966 0.3387>]}
       GreenCurve {keys [<at 0.6966 1.5>]}
       RedCurve {keys [<at 0.6966 0>]}
      }})
    R5 Grey:LinearColor=&ColorFromCurve(ColorCurve=DataInterface() {propertyValues {
       AlphaCurve {
        defaultValue=0
        keys [
         {arriveTangent=13.1575, interpMode=RCIM_Cubic, leaveTangent=13.1575, tangentMode=RCTM_User, time=0.1591, value=7}
         {tangentMode=RCTM_User, time=0.5992, value=7}
        ]
       }
       BlueCurve {keys [<at 0.6966 0.9375>]}
       GreenCurve {keys [<at 0.6966 0.9375>]}
       RedCurve {keys [<at 0.6966 0.9375>]}
      }})
    R5 Lime:LinearColor=&ColorFromCurve(ColorCurve=DataInterface() {propertyValues {
       AlphaCurve {
        defaultValue=0
        keys [
         {arriveTangent=13.1575, interpMode=RCIM_Cubic, leaveTangent=13.1575, tangentMode=RCTM_User, time=0.1591, value=7}
         {tangentMode=RCTM_User, time=0.5992, value=7}
        ]
       }
       BlueCurve {keys [<at 0.6966 0>]}
       GreenCurve {keys [<at 0.6966 1>]}
       RedCurve {keys [<at 0.6966 0.5901>]}
      }})
    R5 Orange:LinearColor=&ColorFromCurve(ColorCurve=DataInterface() {propertyValues {
       AlphaCurve {
        defaultValue=0
        keys [
         {arriveTangent=13.1575, interpMode=RCIM_Cubic, leaveTangent=13.1575, tangentMode=RCTM_User, time=0.1591, value=7}
         {tangentMode=RCTM_User, time=0.5992, value=7}
        ]
       }
       BlueCurve {keys [<at 0.6966 0.2681>]}
       GreenCurve {keys [<at 0.6966 0.4367>]}
       RedCurve {keys [<at 0.6966 0.9531>]}
      }})
    R5 Pink:LinearColor=&ColorFromCurve(ColorCurve=DataInterface() {propertyValues {
       AlphaCurve {
        defaultValue=0
        keys [
         {arriveTangent=13.1575, interpMode=RCIM_Cubic, leaveTangent=13.1575, tangentMode=RCTM_User, time=0.1591, value=7}
         {tangentMode=RCTM_User, time=0.5992, value=7}
        ]
       }
       BlueCurve {keys [<at 0.6966 1.2188>]}
       GreenCurve {keys [<at 0.6966 0.2158>]}
       RedCurve {keys [<at 0.6966 1.1488>]}
      }})
    R5 Purple:LinearColor=&ColorFromCurve(ColorCurve=DataInterface() {propertyValues {
       AlphaCurve {
        defaultValue=0
        keys [
         {arriveTangent=13.1575, interpMode=RCIM_Cubic, leaveTangent=13.1575, tangentMode=RCTM_User, time=0.1591, value=7}
         {tangentMode=RCTM_User, time=0.5992, value=7}
        ]
       }
       BlueCurve {keys [<at 0.6966 1>]}
       GreenCurve {keys [<at 0.6966 0>]}
       RedCurve {keys [<at 0.6966 0.3627>]}
      }})
    R5 Saffron:LinearColor=&ColorFromCurve(ColorCurve=DataInterface() {propertyValues {
       AlphaCurve {
        defaultValue=0
        keys [
         {arriveTangent=13.1575, interpMode=RCIM_Cubic, leaveTangent=13.1575, tangentMode=RCTM_User, time=0.1591, value=7}
         {tangentMode=RCTM_User, time=0.5992, value=7}
        ]
       }
       BlueCurve {keys [<at 0.6966 0>]}
       GreenCurve {keys [<at 0.6966 1.8157>]}
       RedCurve {keys [<at 0.6966 2>]}
      }})
    R5 SkyBlue:LinearColor=&ColorFromCurve(ColorCurve=DataInterface() {propertyValues {
       AlphaCurve {
        defaultValue=0
        keys [
         {arriveTangent=13.1575, interpMode=RCIM_Cubic, leaveTangent=13.1575, tangentMode=RCTM_User, time=0.1591, value=7}
         {tangentMode=RCTM_User, time=0.5992, value=7}
        ]
       }
       BlueCurve {keys [<at 0.6966 1.5>]}
       GreenCurve {keys [<at 0.6966 0.3855>]}
       RedCurve {keys [<at 0.6966 0>]}
      }})
    R5 TitaniumWhite:LinearColor=&ColorFromCurve(ColorCurve=DataInterface() {propertyValues {
       AlphaCurve {keys [
         {arriveTangent=13.1575, interpMode=RCIM_Cubic, leaveTangent=13.1575, tangentMode=RCTM_User, time=0.1591, value=1}
         {tangentMode=RCTM_User, time=0.5992, value=1}
        ]}
       BlueCurve {keys [<at 0.6966 1>]}
       GreenCurve {keys [<at 0.6966 1>]}
       RedCurve {keys [<at 0.6966 1>]}
      }})
    Use R1 Painted Overrides=true
    Use R2 Painted Overrides=true
    Use R3 Painted Overrides=true
    Use R4 Painted Overrides=true
    Use R5 Painted Overrides=true
   ) {module=VC_PaintedVariantsOverride001}
   &FN_DisableTireFX(
    Mask Index Attribute:NiagaraInt32=$Particles.PartitionParticles.Partition
    Mask Index FL=2
    Mask Index FR=3
    Mask Index RL=0
    Mask Index RR=1
   )
   &VC_WheelWrapScale(Wheel Index Attribute:NiagaraInt32=&Add_Integer(A=$Particles.PartitionParticles.Partition, B=2))
   &ParticleState(Kill Particles When Lifetime Has Elapsed=false)
   &UpdateMeshOrientation(Rotation Coordinate Space=::NewEnumerator2, Rotation Rate=0, Rotation Vector=<0 1 0>)
   &SkeletalMeshLocation(
    Bone / Socket Index=&Add_Integer(A=$Particles.PartitionParticles.Partition, B=4)
    Bone / Socket Sampling Mode=::NewEnumerator5
    MaskSpawnGroup=false
    Mesh Sampling Type=::NewEnumerator6
    Orientation Sampling=::NewEnumerator1
    Skeletal Mesh=$System.SkeletalMesh
   ) {module=SkeletalMeshLocation001}
   &SolveForcesAndVelocity(Clamp Velocity=false, Limit Acceleration=false)
   /SetVariables(Particles.RibbonGlow:LinearColor=<rgba 0 0.5029 1 1>, Particles.RibbonGlow 2:LinearColor=<rgba 0 0.0648 1 1>, Particles.Scale Glow 2=1) {enabled=false, isSetParameters=true}
   /SetVariables(
    Particles.R1_Color_Scaled:LinearColor=&MultiplyLinearColor(
     Linear Color=$Particles.R1 Color
     Scale Factor=&MakeVector4FromVectorAndFloat("Vector (RGB)"=&VectorFromFloat(Value=0.7), "Float (Alpha)"=1)
    )
    Particles.R3_Color_Scaled:LinearColor=&MultiplyLinearColor(
     Linear Color=$Particles.R3 Color
     Scale Factor=&MakeVector4FromVectorAndFloat("Vector (RGB)"=&VectorFromFloat(Value=0.5), "Float (Alpha)"=1)
    )
    Particles.R4_Color_Scaled:LinearColor=&MultiplyLinearColor(
     Linear Color=$Particles.R4 Color
     Scale Factor=&MakeVector4FromVectorAndFloat("Vector (RGB)"=&VectorFromFloat(Value=10), "Float (Alpha)"=1)
    )
    Particles.R5_Color_Scaled:LinearColor=&MultiplyLinearColor(
     Linear Color=$Particles.R5 Color
     Scale Factor=&MakeVector4FromVectorAndFloat("Vector (RGB)"=&VectorFromFloat(Value=1), "Float (Alpha)"=1)
    )
   ) {isSetParameters=true}
   /SetVariables(
    Particles.R1_Color_Scaled_Main:LinearColor=&MultiplyLinearColor(
     Linear Color=$Particles.R2 Color
     Scale Factor=&MakeVector4FromVectorAndFloat("Vector (RGB)"=<1 1 1>, "Float (Alpha)"=1)
    )
    Particles.R3_Color_Scaled_Second:LinearColor=&MultiplyLinearColor(Linear Color=$Particles.R5 Color, Scale Factor=<1 1 1 1>)
   ) {isSetParameters=true}
   &VC_PaintedVariantsOverride(
    R1 Black:LinearColor=&ColorFromCurve(ColorCurve=DataInterface() {propertyValues {
       AlphaCurve {
        defaultValue=0
        keys [<at 0 0>, {arriveTangent=15.9916, interpMode=RCIM_Cubic, leaveTangent=15.9916, time=0.1339, value=8}, <at 0.6253 10>, <at 0.9771 0>]
       }
       BlueCurve {keys [<at 1 0.1094 RCIM_Cubic>]}
       GreenCurve {keys [<at 1 0.1094 RCIM_Cubic>]}
       RedCurve {keys [<at 1 0.1094 RCIM_Cubic>]}
      }})
    R1 BurntSienna:LinearColor=&ColorFromCurve(ColorCurve=DataInterface() {propertyValues {
       AlphaCurve {
        defaultValue=0
        keys [<at 0 0>, {arriveTangent=15.9916, interpMode=RCIM_Cubic, leaveTangent=15.9916, time=0.1756, value=8}, <at 0.6253 10>, <at 0.9771 0>]
       }
       BlueCurve {keys [<at 1 0.0686 RCIM_Cubic>]}
       GreenCurve {keys [<at 1 0.0858 RCIM_Cubic>]}
       RedCurve {keys [<at 1 0.1667 RCIM_Cubic>]}
      }})
    R1 Cobalt:LinearColor=&ColorFromCurve(ColorCurve=DataInterface() {propertyValues {
       AlphaCurve {
        defaultValue=0
        keys [<at 0 0>, {arriveTangent=15.9916, interpMode=RCIM_Cubic, leaveTangent=15.9916, time=0.1756, value=8}, <at 0.6253 10>, <at 0.9771 0>]
       }
       BlueCurve {keys [<at 1 0.3802 RCIM_Cubic>]}
       GreenCurve {keys [<at 1 0.0931 RCIM_Cubic>]}
       RedCurve {keys [<at 1 0.0931 RCIM_Cubic>]}
      }})
    R1 Crimson:LinearColor=&ColorFromCurve(ColorCurve=DataInterface() {propertyValues {
       AlphaCurve {
        defaultValue=0
        keys [<at 0 0>, {arriveTangent=15.9916, interpMode=RCIM_Cubic, leaveTangent=15.9916, time=0.1756, value=8}, <at 0.6253 10>, <at 0.9771 0>]
       }
       BlueCurve {keys [<at 1 0.0131 RCIM_Cubic>]}
       GreenCurve {keys [<at 1 0 RCIM_Cubic>]}
       RedCurve {keys [<at 1 0.2348 RCIM_Cubic>]}
      }})
    R1 DefaultColor=&ColorFromCurve(ColorCurve=DataInterface() {propertyValues {
       AlphaCurve {keys [<at 0 0>, {arriveTangent=15.9916, interpMode=RCIM_Cubic, leaveTangent=15.9916, time=0.1039, value=8}, <at 0.6253 10>, <at 0.9771 0>]}
       BlueCurve {keys [<at 0.6966 1>]}
       GreenCurve {keys [<at 0.6966 0.5>]}
       RedCurve {keys [<at 0.6966 0>]}
      }})
    R1 ForestGreen:LinearColor=&ColorFromCurve(ColorCurve=DataInterface() {propertyValues {
       AlphaCurve {
        defaultValue=0
        keys [<at 0 0>, {arriveTangent=15.9916, interpMode=RCIM_Cubic, leaveTangent=15.9916, time=0.1756, value=8}, <at 0.6253 10>, <at 0.9771 0>]
       }
       BlueCurve {keys [<at 1 0.0198 RCIM_Cubic>]}
       GreenCurve {keys [<at 1 0.2031 RCIM_Cubic>]}
       RedCurve {keys [<at 1 0 RCIM_Cubic>]}
      }})
    R1 Grey:LinearColor=&ColorFromCurve(ColorCurve=DataInterface() {propertyValues {
       AlphaCurve {
        defaultValue=0
        keys [<at 0 0>, {arriveTangent=15.9916, interpMode=RCIM_Cubic, leaveTangent=15.9916, time=0.1756, value=8}, <at 0.6253 10>, <at 0.9771 0>]
       }
       BlueCurve {keys [<at 1 0.3542 RCIM_Cubic>]}
       GreenCurve {keys [<at 1 0.3542 RCIM_Cubic>]}
       RedCurve {keys [<at 1 0.3542 RCIM_Cubic>]}
      }})
    R1 Lime:LinearColor=&ColorFromCurve(ColorCurve=DataInterface() {propertyValues {
       AlphaCurve {
        defaultValue=0
        keys [<at 0 0>, {arriveTangent=15.9916, interpMode=RCIM_Cubic, leaveTangent=15.9916, time=0.1756, value=8}, <at 0.6253 10>, <at 0.9771 0>]
       }
       BlueCurve {keys [<at 1 0.0865 RCIM_Cubic>]}
       GreenCurve {keys [<at 1 0.2865 RCIM_Cubic>]}
       RedCurve {keys [<at 1 0.1697 RCIM_Cubic>]}
      }})
    R1 Orange:LinearColor=&ColorFromCurve(ColorCurve=DataInterface() {propertyValues {
       AlphaCurve {
        defaultValue=0
        keys [<at 0 0>, {arriveTangent=15.9916, interpMode=RCIM_Cubic, leaveTangent=15.9916, time=0.1756, value=8}, <at 0.6253 10>, <at 0.9771 0>]
       }
       BlueCurve {keys [<at 1 0 RCIM_Cubic>]}
       GreenCurve {keys [<at 1 0.2961 RCIM_Cubic>]}
       RedCurve {keys [<at 1 0.9271 RCIM_Cubic>]}
      }})
    R1 Pink:LinearColor=&ColorFromCurve(ColorCurve=DataInterface() {propertyValues {
       AlphaCurve {
        defaultValue=0
        keys [<at 0 0>, {arriveTangent=15.9916, interpMode=RCIM_Cubic, leaveTangent=15.9916, time=0.1756, value=8}, <at 0.6253 10>, <at 0.9771 0>]
       }
       BlueCurve {keys [<at 1 0.2004 RCIM_Cubic>]}
       GreenCurve {keys [<at 1 0 RCIM_Cubic>]}
       RedCurve {keys [<at 1 0.2188 RCIM_Cubic>]}
      }})
    R1 Purple:LinearColor=&ColorFromCurve(ColorCurve=DataInterface() {propertyValues {
       AlphaCurve {
        defaultValue=0
        keys [<at 0 0>, {arriveTangent=15.9916, interpMode=RCIM_Cubic, leaveTangent=15.9916, time=0.1756, value=8}, <at 0.6253 10>, <at 0.9771 0>]
       }
       BlueCurve {keys [<at 1 0.375 RCIM_Cubic>]}
       GreenCurve {keys [<at 1 0 RCIM_Cubic>]}
       RedCurve {keys [<at 1 0.1484 RCIM_Cubic>]}
      }})
    R1 Saffron:LinearColor=&ColorFromCurve(ColorCurve=DataInterface() {propertyValues {
       AlphaCurve {
        defaultValue=0
        keys [<at 0 0>, {arriveTangent=15.9916, interpMode=RCIM_Cubic, leaveTangent=15.9916, time=0.1756, value=8}, <at 0.6253 10>, <at 0.9771 0>]
       }
       BlueCurve {keys [<at 1 0 RCIM_Cubic>]}
       GreenCurve {keys [<at 1 0.2395 RCIM_Cubic>]}
       RedCurve {keys [<at 1 0.2812 RCIM_Cubic>]}
      }})
    R1 SkyBlue:LinearColor=&ColorFromCurve(ColorCurve=DataInterface() {propertyValues {
       AlphaCurve {
        defaultValue=0
        keys [<at 0 0>, {arriveTangent=15.9916, interpMode=RCIM_Cubic, leaveTangent=15.9916, time=0.1756, value=8}, <at 0.6253 10>, <at 0.9771 0>]
       }
       BlueCurve {keys [<at 1 0.3802 RCIM_Cubic>]}
       GreenCurve {keys [<at 1 0.1905 RCIM_Cubic>]}
       RedCurve {keys [<at 1 0 RCIM_Cubic>]}
      }})
    R1 TitaniumWhite:LinearColor=&ColorFromCurve(ColorCurve=DataInterface() {propertyValues {
       AlphaCurve {
        defaultValue=0
        keys [<at 0 0>, {arriveTangent=15.9916, interpMode=RCIM_Cubic, leaveTangent=15.9916, time=0.1756, value=8}, <at 0.6253 10>, <at 0.9771 0>]
       }
       BlueCurve {keys [<at 1 0.6406 RCIM_Cubic>]}
       GreenCurve {keys [<at 1 0.6406 RCIM_Cubic>]}
       RedCurve {keys [<at 1 0.6406 RCIM_Cubic>]}
      }})
    R2 Black:LinearColor=&ColorFromCurve(ColorCurve=DataInterface() {propertyValues {
       AlphaCurve {keys [<at 0 7>,<at 0.2232 1>,<at 0.3261 0.05>,<at 0.5495 0.2>,<at 0.7007 0.2>,<at 0.8474 0 RCIM_Cubic>]}
       BlueCurve {keys [<at 0 0.0194 RCIM_Cubic>,<at 0.1492 0.0333>,<at 0.2681 0.1104>,<at 0.4863 0.9274>,<at 0.6088 0.0305>]}
       GreenCurve {keys [<at 0 0.0194 RCIM_Cubic>,<at 0.1492 0.0333>,<at 0.2681 0.1104>,<at 0.4863 0.9274>,<at 0.6088 0.0305>]}
       RedCurve {keys [<at 0 0.0194 RCIM_Cubic>,<at 0.1492 0.0333>,<at 0.2681 0.1104>,<at 0.4863 0.9274>,<at 0.6088 0.0305>]}
      }})
    R2 BurntSienna:LinearColor=&ColorFromCurve(ColorCurve=DataInterface() {propertyValues {
       AlphaCurve {keys [<at 0 4>,<at 0.1907 1>,<at 0.3261 0.05>,<at 0.593 0.2>,<at 0.8474 0 RCIM_Cubic>]}
       BlueCurve {keys [<at 0 0.008 RCIM_Cubic>,<at 0.2839 0.0202>,<at 0.4863 0.2907>,<at 0.5956 0.008>]}
       GreenCurve {keys [<at 0 0.0203 RCIM_Cubic>,<at 0.2839 0.0323>,<at 0.4863 0.4394>,<at 0.5956 0.0203>]}
       RedCurve {keys [<at 0 0.0648 RCIM_Cubic>,<at 0.2839 0.0762>,<at 0.4863 0.9792>,<at 0.5956 0.0648>]}
      }})
    R2 Cobalt:LinearColor=&ColorFromCurve(ColorCurve=DataInterface() {propertyValues {
       AlphaCurve {keys [<at 0 7>,<at 0.1907 1>,<at 0.3261 0.05>,<at 0.5495 0.6>,<at 0.7007 0.2>,<at 0.8474 0 RCIM_Cubic>]}
       BlueCurve {keys [<at 0 0.6458 RCIM_Cubic>,<at 0.2575 0.9985>,<at 0.4863 1>,<at 0.6115 0.8177>]}
       GreenCurve {keys [<at 0 0.0119 RCIM_Cubic>,<at 0.2575 0.0367>,<at 0.4863 0.2997>,<at 0.6115 0.0527>]}
       RedCurve {keys [<at 0 0 RCIM_Cubic>,<at 0.2575 0.0186>,<at 0.4863 0.2865>,<at 0.6115 0.0383>]}
      }})
    R2 Crimson:LinearColor=&ColorFromCurve(ColorCurve=DataInterface() {propertyValues {
       AlphaCurve {keys [<at 0 7>,<at 0.1907 1>,<at 0.3261 0.05>,<at 0.5495 0.2>,<at 0.7007 0.2>,<at 0.8474 0 RCIM_Cubic>]}
       BlueCurve {keys [<at 0 0 RCIM_Cubic>,<at 0.1492 0.0104>,<at 0.4863 0.2995>,<at 0.6458 0.0894>]}
       GreenCurve {keys [<at 0 0 RCIM_Cubic>,<at 0.1492 0.0104>,<at 0.4863 0.2995>,<at 0.6458 0.0894>]}
       RedCurve {keys [<at 0 0.4323 RCIM_Cubic>,<at 0.1492 0.9992>,<at 0.4863 0.9274>,<at 0.6458 0.993>]}
      }})
    R2 DefaultColor:LinearColor=&ColorFromCurve(ColorCurve=DataInterface() {propertyValues {
       AlphaCurve {keys [<at 0 1>,<at 0.0914 0.2857>,<at 0.308 0.05>,<at 0.5592 0.2>,<at 0.8632 0>]}
       BlueCurve {keys [<at 0.0009 0>,<at 0.2668 -0.2829>,<at 0.5114 0.8849>,<at 0.619 1>,<at 0.682 0.383>]}
       GreenCurve {keys [<at 0.0009 0.0767>,<at 0.2742 -0.1994>,<at 0.5114 0.8849>,<at 0.619 1>,<at 0.682 0.2585>]}
       RedCurve {keys [<at 0.0009 3.2194>,<at 0.0016 3.2194>,<at 0.4479 0.8849>,<at 0.5114 0.8849>,<at 0.619 1>]}
      }})
    R2 ForestGreen:LinearColor=&ColorFromCurve(ColorCurve=DataInterface() {propertyValues {
       AlphaCurve {keys [<at 0 6>,<at 0.1907 1>,<at 0.3261 0.05>,<at 0.5495 0.2>,<at 0.7007 0.2>,<at 0.8474 0 RCIM_Cubic>]}
       BlueCurve {keys [<at 0 0.0172 RCIM_Cubic>,<at 0.2179 0.0322>,<at 0.4863 0.4656>,<at 0.6537 0.1027>]}
       GreenCurve {keys [<at 0 0.3438 RCIM_Cubic>,<at 0.2179 0.245>,<at 0.4863 0.875>,<at 0.6537 0.2987>]}
       RedCurve {keys [<at 0 0 RCIM_Cubic>,<at 0.2179 0.021>,<at 0.4863 0.4466>,<at 0.6537 0.0924>]}
      }})
    R2 Grey:LinearColor=&ColorFromCurve(ColorCurve=DataInterface() {propertyValues {
       AlphaCurve {keys [<at 0 4>,<at 0.1907 1>,<at 0.3261 0.05>,<at 0.5495 0.2>,<at 0.7007 0.2>,<at 0.8474 0 RCIM_Cubic>]}
       BlueCurve {keys [<at 0 0.0692 RCIM_Cubic>,<at 0.2285 0.1336>,<at 0.4863 0.9274>,<at 0.6405 0.132>]}
       GreenCurve {keys [<at 0 0.0781 RCIM_Cubic>,<at 0.2285 0.1336>,<at 0.4863 0.9274>,<at 0.6405 0.132>]}
       RedCurve {keys [<at 0 0.0767 RCIM_Cubic>,<at 0.2285 0.1406>,<at 0.4863 0.9274>,<at 0.6405 0.1354>]}
      }})
    R2 Lime:LinearColor=&ColorFromCurve(ColorCurve=DataInterface() {propertyValues {
       AlphaCurve {keys [<at 0 6>,<at 0.1907 1>,<at 0.3261 0.05>,<at 0.5495 0.2>,<at 0.8122 0 RCIM_Cubic>]}
       BlueCurve {keys [<at 0 0 RCIM_Cubic>,<at 0.2179 0.0459>,<at 0.4863 0.9274>,<at 0.6299 0.0724>]}
       GreenCurve {keys [<at 0 0.6875 RCIM_Cubic>,<at 0.2179 0.4635>,<at 0.4863 0.9274>,<at 0.6299 0.7708>]}
       RedCurve {keys [<at 0 0.3447 RCIM_Cubic>,<at 0.2179 0.2553>,<at 0.4863 0.9274>,<at 0.6299 0.4226>]}
      }})
    R2 Orange:LinearColor=&ColorFromCurve(ColorCurve=DataInterface() {propertyValues {
       AlphaCurve {keys [<at 0 7>,<at 0.1907 1>,<at 0.3261 0.05>,<at 0.5495 0.2>,<at 0.7007 0.2>,<at 0.8474 0 RCIM_Cubic>]}
       BlueCurve {keys [<at 0 0 RCIM_Cubic>,<at 0.2364 0.0598>,<at 0.4863 0.4263>,<at 0.6217 0.1191>]}
       GreenCurve {keys [<at 0 0.2834 RCIM_Cubic>,<at 0.2364 0.2792>,<at 0.4863 0.5719>,<at 0.6217 0.3121>]}
       RedCurve {keys [<at 0 0.8125 RCIM_Cubic>,<at 0.2364 0.6889>,<at 0.4863 0.8438>,<at 0.6217 0.6724>]}
      }})
    R2 Pink:LinearColor=&ColorFromCurve(ColorCurve=DataInterface() {propertyValues {
       AlphaCurve {keys [<at 0 7>,<at 0.1907 1>,<at 0.3261 0.05>,<at 0.5495 0.2>,<at 0.7007 0.2>,<at 0.8474 0 RCIM_Cubic>]}
       BlueCurve {keys [<at 0 0.3419 RCIM_Cubic>,<at 0.2707 0.3564>,<at 0.4863 0.7216>,<at 0.6141 0.5634>]}
       GreenCurve {keys [<at 0 0.0612 RCIM_Cubic>,<at 0.2707 0.1489>,<at 0.4863 0.5789>,<at 0.6141 0.2536>]}
       RedCurve {keys [<at 0 0.5029 RCIM_Cubic>,<at 0.2707 0.5614>,<at 0.4863 0.855>,<at 0.6141 0.8542>]}
      }})
    R2 Purple:LinearColor=&ColorFromCurve(ColorCurve=DataInterface() {propertyValues {
       AlphaCurve {keys [<at 0 7>,<at 0.1907 1>,<at 0.3261 0.05>,<at 0.5495 0.2>,<at 0.7007 0.2>,<at 0.8474 0 RCIM_Cubic>]}
       BlueCurve {keys [<at 0 0.8281 RCIM_Cubic>,<at 0.2919 0.6836>,<at 0.4863 0.8957>,<at 0.6141 0.9565>]}
       GreenCurve {keys [<at 0 0 RCIM_Cubic>,<at 0.2919 0.0563>,<at 0.4863 0.2791>,<at 0.6141 0.1097>]}
       RedCurve {keys [<at 0 0.1924 RCIM_Cubic>,<at 0.2919 0.3349>,<at 0.4863 0.5023>,<at 0.6141 0.3436>]}
      }})
    R2 Saffron:LinearColor=&ColorFromCurve(ColorCurve=DataInterface() {propertyValues {
       AlphaCurve {keys [<at 0 4>,<at 0.1907 1>,<at 0.3261 0.05>,<at 0.5495 0.2>,<at 0.8175 0 RCIM_Cubic>]}
       BlueCurve {keys [<at 0 0 RCIM_Cubic>,<at 0.2697 0.0046>,<at 0.4863 0.3188>,<at 0.6748 0.0798>]}
       GreenCurve {keys [<at 0 0.6105 RCIM_Cubic>,<at 0.2697 0.6881>,<at 0.4863 0.9274>,<at 0.6748 0.6052>]}
       RedCurve {keys [<at 0 0.7969 RCIM_Cubic>,<at 0.2697 0.8967>,<at 0.4863 0.8744>,<at 0.6748 0.7656>]}
      }})
    R2 SkyBlue:LinearColor=&ColorFromCurve(ColorCurve=DataInterface() {propertyValues {
       AlphaCurve {keys [<at 0 7>,<at 0.1907 1>,<at 0.3261 0.05>,<at 0.5495 0.2>,<at 0.8096 0 RCIM_Cubic>]}
       BlueCurve {keys [<at 0 0.5333 RCIM_Cubic>,<at 0.2496 0.5525>,<at 0.4863 0.9323>,<at 0.6141 0.9323>]}
       GreenCurve {keys [<at 0 0.3564 RCIM_Cubic>,<at 0.2496 0.3792>,<at 0.4863 0.8083>,<at 0.6141 0.6988>]}
       RedCurve {keys [<at 0 0 RCIM_Cubic>,<at 0.2496 0.0875>,<at 0.4863 0.5584>,<at 0.6141 0.2282>]}
      }})
    R2 TitaniumWhite:LinearColor=&ColorFromCurve(ColorCurve=DataInterface() {propertyValues {
       AlphaCurve {keys [<at 0 4>,<at 0.1907 1>,<at 0.3261 0.05>,<at 0.5495 0.2>,<at 0.7007 0.2>,<at 0.8175 0 RCIM_Cubic>]}
       BlueCurve {keys [<at 0 0.1846 RCIM_Cubic>,<at 0.2285 0.1336>,<at 0.4863 0.9274>,<at 0.6405 0.3248>]}
       GreenCurve {keys [<at 0 0.2083 RCIM_Cubic>,<at 0.2285 0.1336>,<at 0.4863 0.9274>,<at 0.6405 0.3248>]}
       RedCurve {keys [<at 0 0.2046 RCIM_Cubic>,<at 0.2285 0.1406>,<at 0.4863 0.9274>,<at 0.6405 0.3333>]}
      }})
    R3 Black:LinearColor=&ColorFromCurve(ColorCurve=DataInterface() {propertyValues {
       AlphaCurve {
        defaultValue=0
        keys [
         {tangentMode=RCTM_User, time=0.0672, value=0}
         {
          arriveTangent=13.1575
          arriveTangentWeight=0.0306
          interpMode=RCIM_Cubic
          leaveTangent=13.1575
          leaveTangentWeight=0.2107
          tangentMode=RCTM_User
          time=0.1591
          value=7
         }
         {tangentMode=RCTM_User, time=0.5992, value=7}
         <at 0.7913 0>
        ]
       }
       BlueCurve {keys [<at 0.6966 0.3043>]}
       GreenCurve {keys [<at 0.6966 0.3043>]}
       RedCurve {keys [<at 0.6966 0.3043>]}
      }})
    R3 BurntSienna:LinearColor=&ColorFromCurve(ColorCurve=DataInterface() {propertyValues {
       AlphaCurve {defaultValue=0, keys [{arriveTangent=13.1575, interpMode=RCIM_Cubic, time=0.0863, value=7}, <at 0.5992 7>, <at 0.7913 0>]}
       BlueCurve {keys [<at 0.6966 0.2729>]}
       GreenCurve {keys [<at 0.6966 0.3693>]}
       RedCurve {keys [<at 0.6966 0.6719>]}
      }})
    R3 Cobalt:LinearColor=&ColorFromCurve(ColorCurve=DataInterface() {propertyValues {
       AlphaCurve {
        defaultValue=0
        keys [
         {tangentMode=RCTM_User, time=0.0672, value=0}
         {arriveTangent=13.1575, interpMode=RCIM_Cubic, leaveTangent=13.1575, tangentMode=RCTM_User, time=0.1591, value=7}
         {tangentMode=RCTM_User, time=0.5992, value=7}
         <at 0.7913 0>
        ]
       }
       BlueCurve {keys [<at 0.6966 1.5>]}
       GreenCurve {keys [<at 0.6966 0.0598>]}
       RedCurve {keys [<at 0.6966 0>]}
      }})
    R3 Crimson:LinearColor=&ColorFromCurve(ColorCurve=DataInterface() {propertyValues {
       AlphaCurve {
        defaultValue=0
        keys [
         {tangentMode=RCTM_User, time=0.0672, value=0}
         {arriveTangent=13.1575, interpMode=RCIM_Cubic, leaveTangent=13.1575, tangentMode=RCTM_User, time=0.1591, value=7}
         {tangentMode=RCTM_User, time=0.5992, value=7}
         <at 0.7913 0>
        ]
       }
       BlueCurve {keys [<at 0.6966 0>]}
       GreenCurve {keys [<at 0.6966 0>]}
       RedCurve {keys [<at 0.6966 1>]}
      }})
    R3 DefaultColor:LinearColor=&ColorFromCurve(ColorCurve=DataInterface() {propertyValues {
       AlphaCurve {keys [
         {tangentMode=RCTM_User, time=0, value=0}
         {
          arriveTangent=13.1575
          arriveTangentWeight=0.0306
          interpMode=RCIM_Cubic
          leaveTangent=13.1575
          leaveTangentWeight=0.2538
          tangentMode=RCTM_User
          time=0.1591
          value=7
         }
         {tangentMode=RCTM_User, time=0.5992, value=7}
         <at 0.7913 0>
         <at 0.9206 0>
        ]}
       BlueCurve {keys [<at 0.4821 1>]}
       GreenCurve {keys [<at 0.4821 0.0657>]}
       RedCurve {keys [<at 0.4821 0>]}
      }})
    R3 ForestGreen:LinearColor=&ColorFromCurve(ColorCurve=DataInterface() {propertyValues {
       AlphaCurve {
        defaultValue=0
        keys [
         {tangentMode=RCTM_User, time=0.0672, value=0}
         {arriveTangent=13.1575, interpMode=RCIM_Cubic, leaveTangent=13.1575, tangentMode=RCTM_User, time=0.1591, value=7}
         {tangentMode=RCTM_User, time=0.5992, value=7}
         <at 0.7913 0>
        ]
       }
       BlueCurve {keys [<at 0.6966 0.3387>]}
       GreenCurve {keys [<at 0.6966 1.5>]}
       RedCurve {keys [<at 0.6966 0>]}
      }})
    R3 Grey:LinearColor=&ColorFromCurve(ColorCurve=DataInterface() {propertyValues {
       AlphaCurve {
        defaultValue=0
        keys [
         {tangentMode=RCTM_User, time=0.0672, value=0}
         {arriveTangent=13.1575, interpMode=RCIM_Cubic, leaveTangent=13.1575, tangentMode=RCTM_User, time=0.1591, value=7}
         {tangentMode=RCTM_User, time=0.5992, value=7}
         <at 0.7913 0>
        ]
       }
       BlueCurve {keys [<at 0.6966 0.9375>]}
       GreenCurve {keys [<at 0.6966 0.9375>]}
       RedCurve {keys [<at 0.6966 0.9375>]}
      }})
    R3 Lime:LinearColor=&ColorFromCurve(ColorCurve=DataInterface() {propertyValues {
       AlphaCurve {
        defaultValue=0
        keys [
         {tangentMode=RCTM_User, time=0.0672, value=0}
         {arriveTangent=13.1575, interpMode=RCIM_Cubic, leaveTangent=13.1575, tangentMode=RCTM_User, time=0.1591, value=7}
         {tangentMode=RCTM_User, time=0.5992, value=7}
         <at 0.7913 0>
        ]
       }
       BlueCurve {keys [<at 0.6966 0>]}
       GreenCurve {keys [<at 0.6966 1>]}
       RedCurve {keys [<at 0.6966 0.5901>]}
      }})
    R3 Orange:LinearColor=&ColorFromCurve(ColorCurve=DataInterface() {propertyValues {
       AlphaCurve {
        defaultValue=0
        keys [
         {tangentMode=RCTM_User, time=0, value=0}
         {arriveTangent=13.1575, interpMode=RCIM_Cubic, leaveTangent=13.1575, tangentMode=RCTM_User, time=0.1591, value=7}
         {tangentMode=RCTM_User, time=0.5992, value=7}
         <at 0.7913 0>
        ]
       }
       BlueCurve {keys [<at 0.6966 0.2681>]}
       GreenCurve {keys [<at 0.6966 0.4367>]}
       RedCurve {keys [<at 0.6966 0.9531>]}
      }})
    R3 Pink:LinearColor=&ColorFromCurve(ColorCurve=DataInterface() {propertyValues {
       AlphaCurve {
        defaultValue=0
        keys [
         {tangentMode=RCTM_User, time=0.0672, value=0}
         {arriveTangent=13.1575, interpMode=RCIM_Cubic, leaveTangent=13.1575, tangentMode=RCTM_User, time=0.1591, value=7}
         {tangentMode=RCTM_User, time=0.5992, value=7}
         <at 0.7913 0>
        ]
       }
       BlueCurve {keys [<at 0.6966 1.2188>]}
       GreenCurve {keys [<at 0.6966 0.2158>]}
       RedCurve {keys [<at 0.6966 1.1488>]}
      }})
    R3 Purple:LinearColor=&ColorFromCurve(ColorCurve=DataInterface() {propertyValues {
       AlphaCurve {
        defaultValue=0
        keys [
         {tangentMode=RCTM_User, time=0.0672, value=0}
         {arriveTangent=13.1575, interpMode=RCIM_Cubic, leaveTangent=13.1575, tangentMode=RCTM_User, time=0.1591, value=7}
         {tangentMode=RCTM_User, time=0.5992, value=7}
         <at 0.7913 0>
        ]
       }
       BlueCurve {keys [<at 0.6966 1>]}
       GreenCurve {keys [<at 0.6966 0>]}
       RedCurve {keys [<at 0.6966 0.3627>]}
      }})
    R3 Saffron:LinearColor=&ColorFromCurve(ColorCurve=DataInterface() {propertyValues {
       AlphaCurve {
        defaultValue=0
        keys [
         {tangentMode=RCTM_User, time=0.0672, value=0}
         {arriveTangent=13.1575, interpMode=RCIM_Cubic, leaveTangent=13.1575, tangentMode=RCTM_User, time=0.1591, value=7}
         {tangentMode=RCTM_User, time=0.5992, value=7}
         <at 0.7913 0>
        ]
       }
       BlueCurve {keys [<at 0.6966 0>]}
       GreenCurve {keys [<at 0.6966 1.8157>]}
       RedCurve {keys [<at 0.6966 2>]}
      }})
    R3 SkyBlue:LinearColor=&ColorFromCurve(ColorCurve=DataInterface() {propertyValues {
       AlphaCurve {
        defaultValue=0
        keys [
         {tangentMode=RCTM_User, time=0.0672, value=0}
         {arriveTangent=13.1575, interpMode=RCIM_Cubic, leaveTangent=13.1575, tangentMode=RCTM_User, time=0.1591, value=7}
         {tangentMode=RCTM_User, time=0.5992, value=7}
         <at 0.7913 0>
        ]
       }
       BlueCurve {keys [<at 0.6966 1.5>]}
       GreenCurve {keys [<at 0.6966 0.3855>]}
       RedCurve {keys [<at 0.6966 0>]}
      }})
    R3 TitaniumWhite:LinearColor=&ColorFromCurve(ColorCurve=DataInterface() {propertyValues {
       AlphaCurve {keys [
         {tangentMode=RCTM_User, time=0.0672, value=0}
         {arriveTangent=13.1575, interpMode=RCIM_Cubic, leaveTangent=13.1575, tangentMode=RCTM_User, time=0.1591, value=7}
         {tangentMode=RCTM_User, time=0.5992, value=7}
         <at 0.7913 0>
        ]}
       BlueCurve {keys [<at 0.6966 2.5>]}
       GreenCurve {keys [<at 0.6966 2.5>]}
       RedCurve {keys [<at 0.6966 2.5>]}
      }})
    R4 BurntSienna:LinearColor=<rgba 0.0312 0.0096 0.0039 0.6>
    R4 Cobalt:LinearColor=<rgba 0 0.0115 0.6198 0.7>
    R4 Crimson:LinearColor=<rgba 0.2656 0 0.0072 0.7>
    R4 ForestGreen:LinearColor=<rgba 0 0.2552 0.0129 0.6>
    R4 Grey:LinearColor=<rgba 0.0417 0.0417 0.0417 0.7>
    R4 Lime:LinearColor=<rgba 0.1875 0.375 0 0.6>
    R4 Orange:LinearColor=<rgba 0.5521 0.1706 0 0.45>
    R4 Pink:LinearColor=<rgba 0.3906 0.0468 0.2656 0.5>
    R4 Purple:LinearColor=<rgba 0.1567 0 0.7552 0.5>
    R4 Saffron:LinearColor=<rgba 0.5573 0.4299 0 0.3>
    R4 SkyBlue:LinearColor=<rgba 0.0137 0.2957 0.4375 0.5>
    R4 TitaniumWhite:LinearColor=<rgba 0.224 0.224 0.224 0.6>
    R5 Black:LinearColor=<rgba 0.0469 0.0469 0.0469 0.6>
    R5 BurntSienna:LinearColor=<rgba 0.1042 0.0322 0.0128 0.7>
    R5 Cobalt:LinearColor=<rgba 0.0312 0.0401 1 0.6>
    R5 Crimson:LinearColor=<rgba 0.4635 0.0266 0.0384 0.6>
    R5 DefaultColor:LinearColor=<rgba 1 1 1 0.5>
    R5 ForestGreen:LinearColor=<rgba 0.0704 0.901 0.1124 0.4>
    R5 Grey:LinearColor=<rgba 0.1667 0.1667 0.1667 0.4>
    R5 Lime:LinearColor=<rgba 0.2938 0.5104 0.0771 0.25>
    R5 Orange:LinearColor=<rgba 0.8261 0.2993 0.0172 0.3>
    R5 Pink:LinearColor=<rgba 0.651 0.1763 0.4784 0.4>
    R5 Purple:LinearColor=<rgba 0.2704 0.1155 0.7917 0.3>
    R5 Saffron:LinearColor=<rgba 0.8906 0.6824 0 0.15>
    R5 SkyBlue:LinearColor=<rgba 0.0754 0.3822 0.5365 0.4>
    R5 TitaniumWhite:LinearColor=<rgba 0.5729 0.5729 0.5729 0.3>
    Use R1 Painted Overrides=true
    Use R2 Painted Overrides=true
    Use R3 Painted Overrides=true
    Use R4 Painted Overrides=true
    Use R5 Painted Overrides=true
   ) {enabled=false, module=VC_PaintedVariantsOverride002}
  ]
  rendererClasses [&NiagaraMeshRendererProperties]
  renderers [
   {rendererClass=&NiagaraMeshRendererProperties, rendererIndex=0}
   {rendererClass=&NiagaraMeshRendererProperties, rendererIndex=1}
   {rendererClass=&NiagaraMeshRendererProperties, rendererIndex=2}
   {rendererClass=&NiagaraMeshRendererProperties, rendererIndex=3}
   {rendererClass=&NiagaraMeshRendererProperties, rendererIndex=4}
  ]
 }
}
summary {
 systemName=NS_DriftTrail_BlackFlame1
 emitters [
  {emitterName=Aura_Ribbon, rendererClasses [&NiagaraRibbonRendererProperties]}
  {emitterName=WheelWrap, rendererClasses [&NiagaraMeshRendererProperties]}
  {emitterName=MainRibbon001, rendererClasses [&NiagaraRibbonRendererProperties]}
  {emitterName=VerticalFlames, rendererClasses [&NiagaraMeshRendererProperties]}
  {emitterName=MainRibbon, rendererClasses [&NiagaraRibbonRendererProperties]}
 ]
 userVariables [
  User.bIsFrontendPreview {type=NiagaraBool}
  User.bIsPainted {type=NiagaraBool}
  User.ColorVariantID {type=NiagaraInt32}
  User.offset {type=NiagaraPosition, description="", defaultValue {}}
  User.offset001 {type=Vector3f, description="", defaultValue {struct=Vector3f, value=<0 0 0>}}
  User.PaintedColor {type=LinearColor, description="", defaultValue {struct=LinearColor, value=<rgba 1 1 1 1>}}
  User.PaintedScalar {type=NiagaraFloat, defaultValue=1, description=""}
  User.Power_AlphaPaintedR4 {type=NiagaraFloat, defaultValue=1, description=""}
  User.Power_EmissivePaintedR4 {type=NiagaraFloat, defaultValue=1, description=""}
  User.VehicleFlags {type=NiagaraInt32, defaultValue=806289408}
  User.WaveMaxScale {type=NiagaraFloat, defaultValue=1, description=""}
  User.WavePeriod {type=NiagaraFloat, defaultValue=1, description=""}
  User.Wheel_Radius_Array {
   type=NiagaraDataInterfaceArrayFloat
   defaultValue {
    struct=NiagaraExt_VariableValue_DataInterface
    value {
     dataInterface=&self:NiagaraDataInterfaceArrayFloat_1
     dataInterfaceClass=&NiagaraDataInterfaceArrayFloat
    }
   }
  }
  User.Wheel_Width_Array {
   type=NiagaraDataInterfaceArrayFloat
   defaultValue {
    struct=NiagaraExt_VariableValue_DataInterface
    value {
     dataInterface=&self:NiagaraDataInterfaceArrayFloat_0
     dataInterfaceClass=&NiagaraDataInterfaceArrayFloat
    }
   }
  }
  User.WheelTouchingArray {
   type=NiagaraDataInterfaceArrayFloat
   defaultValue {
    struct=NiagaraExt_VariableValue_DataInterface
    value {
     dataInterface=&self:NiagaraDataInterfaceArrayFloat_2
     dataInterfaceClass=&NiagaraDataInterfaceArrayFloat
    }
   }
  }
 ]
}
